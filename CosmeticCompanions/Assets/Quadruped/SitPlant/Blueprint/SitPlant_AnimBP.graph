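# Animation Blueprint for SitPlant, a quadruped cosmetic companion, covering locomotion, sitting, idle antics and mood-driven poses.
# State machines and additive layers for body, turning, breathing and tail feed a look-at aim offset, with proxy Choosers selecting contextual animations.
# High-fidelity (face-pose layer) and low-fidelity paths switch via bUseLowFidelity with inertialized blending.
# Dead Blending, quadruped slope warping with curve-selected root orientation, and four-leg IK finish the pose.
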AnimBlueprint /CosmeticCompanions/Assets/Quadruped/SitPlant/Blueprint/SitPlant_AnimBP
target_skeleton=/CosmeticCompanions/Assets/Quadruped/SitPlant/Meshes/SitPlant_Quadruped_Skeleton
graphs [
 AllowedToSit_Exit {nodes [
   AllowedToSit_Exit!() -> (then, Context:"Anim Update Context Structure (by ref)", Node:"Anim Node Reference Structure (by ref)") {is=title K2Node_FunctionEntry}
   > Set bIsAllowedToSit(bIsAllowedToSit:bool=true) -> (Output_Get:bool)
   > Return Node()
  ]}
 AllowsCinematicOverride {nodes [
   AllowsCinematicOverride() {is=title K2Node_FunctionEntry}
   > Return Node!(execute, ReturnValue:bool=false)
  ]}
 AnimGraph {nodes [
   n1 = Quadruped Slope Warping!(
    ComponentPose:Component Space Pose Link Structure=Local To Component(LocalPose=Use cached pose 'SavedPose21080'())
    Alpha:float=Get SlopeWarpingAlpha() -> (SlopeWarpingAlpha:float)
    MaxStepHeight:float=50
    bOnlyAffectSkeletonRootBone:bool=true
    MaxSkeletonRootBoneRotationAngle:float=0
   ) -> (Pose:Component Space Pose Link Structure) {
    is=title FortAnimGraphNode_QuadrupedSlopeWarping
    bUseMovementComponentFloorInfoInsteadOfTraces=true
    alphaBoolBlend {blendOption=Linear}
    alphaScaleBiasClamp {bias=1, scale=-1}
    armsFeetDefinitions [
     {numBonesInLimb=3, fKFootBone {boneName=ball_fr_l}, iKFootBone {boneName=IK_Foot_Front_L}}
     {numBonesInLimb=3, fKFootBone {boneName=ball_fr_r}, iKFootBone {boneName=IK_Foot_Front_R}}
    ]
    armsIKFootRootBone {boneName=IK_Foot_Front_Root}
    armsRootBone {boneName=spine_03}
    feetDefinitions [
     {numBonesInLimb=3, fKFootBone {boneName=ball_bk_l}, iKFootBone {boneName=IK_Foot_Rear_L}}
     {numBonesInLimb=3, fKFootBone {boneName=ball_bk_r}, iKFootBone {boneName=IK_Foot_Rear_R}}
    ]
    iKFootRootBone {boneName=IK_Foot_Rear_Root}
    pelvisBone {boneName=hips}
    skeletonRootBone {boneName=root}
   }
   n2 = Component To Local(ComponentPose=@n1)
   n3 = Quadruped Slope Warping(
    ComponentPose=Local To Component(LocalPose=Use cached pose 'SavedPose21080'())
    Alpha=Get SlopeWarpingAlpha()
    MaxStepHeight=50
    bOnlyAffectSkeletonRootBone=False
    MaxSkeletonRootBoneRotationAngle=0
   ) {
    bUseMovementComponentFloorInfoInsteadOfTraces=true
    alphaBoolBlend {blendOption=Linear}
    alphaScaleBiasClamp {bias=1, scale=-1}
    armsFeetDefinitions [
     {numBonesInLimb=3, fKFootBone {boneName=ball_fr_l}, iKFootBone {boneName=IK_Foot_Front_L}}
     {numBonesInLimb=3, fKFootBone {boneName=ball_fr_r}, iKFootBone {boneName=IK_Foot_Front_R}}
    ]
    armsIKFootRootBone {boneName=IK_Foot_Front_Root}
    armsRootBone {boneName=spine_03}
    feetDefinitions [
     {numBonesInLimb=3, fKFootBone {boneName=ball_bk_l}, iKFootBone {boneName=IK_Foot_Rear_L}}
     {numBonesInLimb=3, fKFootBone {boneName=ball_bk_r}, iKFootBone {boneName=IK_Foot_Rear_R}}
    ]
    iKFootRootBone {boneName=IK_Foot_Rear_Root}
    pelvisBone {boneName=hips}
    skeletonRootBone {boneName=root}
   }
   n4 = Component To Local(ComponentPose=@n3)
   n5 = "float > float"(A=GetCurveValue(CurveName=UseRootSlopeOrientation))
   n6 = Leg IK!(
    ComponentPose:Component Space Pose Link Structure=Local To Component(LocalPose=Blend Poses by bool(BlendPose_0=@n2, BlendPose_1=@n4, BlendTime_0=0.3, BlendTime_1=0.3, bActiveValue=@n5) {blendType=Linear})
    Alpha:float=Get IKAlpha() -> (IKAlpha:float)
   ) -> (Pose:Component Space Pose Link Structure) {
    is=title AnimGraphNode_LegIK
    alphaBoolBlend {blendOption=Linear}
    alphaScaleBiasClamp {bias=1, scale=-1}
    legsDefinition [
     {
      bEnableKneeTwistCorrection=true
      bEnableRotationLimit=false
      footBoneForwardAxis=Y
      hingeRotationAxis=None
      minRotationAngle=15
      numBonesInLimb=3
      twistOffsetCurveName=None
      fKFootBone {boneName=ball_bk_l}
      iKFootBone {boneName=IK_Foot_Rear_L}
     }
     {
      bEnableKneeTwistCorrection=true
      bEnableRotationLimit=false
      footBoneForwardAxis=Y
      hingeRotationAxis=None
      minRotationAngle=15
      numBonesInLimb=3
      twistOffsetCurveName=None
      fKFootBone {boneName=ball_bk_r}
      iKFootBone {boneName=IK_Foot_Rear_R}
     }
     {
      bEnableKneeTwistCorrection=true
      bEnableRotationLimit=false
      footBoneForwardAxis=Y
      hingeRotationAxis=None
      minRotationAngle=15
      numBonesInLimb=3
      twistOffsetCurveName=None
      fKFootBone {boneName=ball_fr_r}
      iKFootBone {boneName=IK_Foot_Front_R}
     }
     {
      bEnableKneeTwistCorrection=true
      bEnableRotationLimit=false
      footBoneForwardAxis=Y
      hingeRotationAxis=None
      minRotationAngle=15
      numBonesInLimb=3
      twistOffsetCurveName=None
      fKFootBone {boneName=ball_fr_l}
      iKFootBone {boneName=IK_Foot_Front_L}
     }
    ]
   }
   Output Pose(Result=Component To Local(ComponentPose=@n6))
   AnimGraphNode_StateMachine_0 = IdleAntics() {is=title AnimGraphNode_StateMachine, bAllowConduitEntryStates=true}
   n7 = Apply Additive(Base=Use cached pose 'LocomotionPose'(), Additive=@AnimGraphNode_StateMachine_0, Alpha=Get IdleAnticAlpha() -> (IdleAnticAlpha:double)) {alphaBoolBlend {blendOption=Linear}}
   AnimGraphNode_StateMachine_2 = Tail() {is=title AnimGraphNode_StateMachine, bAllowConduitEntryStates=true}
   n8 = AimOffset Player 'SitPlant_AO'!(
    X:float=Get LookAtYawAdjusted() -> (LookAtYawAdjusted:double)
    Y:float=Get LookAtPitchActual() -> (LookAtPitchActual:float)
    BasePose:Pose Link Structure=Apply Additive(Base=Slot 'DefaultSlot'(Source=@n7), Additive=@AnimGraphNode_StateMachine_2, Alpha=Get TailWagAlpha() -> (TailWagAlpha:float)) {alphaCurveName=Appendage01, alphaBoolBlend {blendOption=Linear}}
    Alpha:float=Get AimOffsetAlpha() -> (AimOffsetAlpha:float)
   ) -> (Pose:Pose Link Structure) {
    alphaCurveName=DisableAimOffset
    blendSpace=/CosmeticCompanions/Assets/Quadruped/SitPlant/Animation/AimOffsets/SitPlant_AO
    alphaBoolBlend {blendOption=Linear}
    alphaScaleBiasClamp {bias=1, scale=-1}
   }
   Save cached pose 'LocomotionPlusAdditivePose'(Pose=@n8) {cacheName=LocomotionPlusAdditivePose}
   AnimGraphNode_StateMachine_5 = Locomotion() {is=title AnimGraphNode_StateMachine, bAllowConduitEntryStates=true}
   AnimGraphNode_StateMachine_6 = BodyAdditives() {is=title AnimGraphNode_StateMachine}
   n9 = Apply Additive(Base=@AnimGraphNode_StateMachine_5, Additive=@AnimGraphNode_StateMachine_6) {alphaBoolBlend {blendOption=Linear}}
   n10 = Blendspace Player 'SitPlant_COR_Loco_Turn_BS'!(X=Get CompanionDirection() -> (CompanionDirection:float)) -> (Pose) {blendSpace=/CosmeticCompanions/Assets/Quadruped/SitPlant/Animation/Core/SitPlant_COR_Loco_Turn_BS}
   n11 = Get BreathingPlayRate() -> (BreathingPlayRate:float)
   Save cached pose 'LocomotionPose'(Pose=Apply Additive(
     Base=Apply Additive(Base=@n9, Additive=@n10) {alphaBoolBlend {blendOption=Linear}}
     Additive="Sequence Player (additive) 'SitPlant_COR_Breathing_ADD'"!(PlayRate:float=@n11) -> (Pose) {playRate=0.75, sequence=/CosmeticCompanions/Assets/Quadruped/SitPlant/Animation/Core/SitPlant_COR_Breathing_ADD}
     Alpha=Get BreathingAlpha() -> (BreathingAlpha:double)
    ) {alphaBoolBlend {blendOption=Linear}}) {cacheName=LocomotionPose}
   AnimGraphNode_StateMachine_1 = LowFidelityBodyAdditives() {is=title AnimGraphNode_StateMachine}
   AnimGraphNode_StateMachine_3 = LowFidelityStateMachine() {is=title AnimGraphNode_StateMachine, bAllowConduitEntryStates=true}
   n12 = Apply Additive(Base=@AnimGraphNode_StateMachine_3, Additive=@AnimGraphNode_StateMachine_1) {alphaBoolBlend {blendOption=Linear}}
   n13 = "Sequence Player (additive) 'SitPlant_ACT_Break_ADD'"() {bLoopAnimation=false, sequence=/CosmeticCompanions/Assets/Quadruped/SitPlant/Animation/Actions/SitPlant_ACT_Break_ADD}
   n14 = Get IdleAnticAlpha()
   n15 = Get BreathingPlayRate()
   n16 = Apply Additive(
    Base=Apply Additive(
     Base=Apply Additive(Base=@n12, Additive=@n13, Alpha=@n14) {alphaBoolBlend {blendOption=Linear}}
     Additive="Sequence Player (additive) 'SitPlant_COR_Breathing_ADD'"(PlayRate=@n15) {playRate=0.75, sequence=&SitPlant_COR_Breathing_ADD}
     Alpha=Get BreathingAlpha()
    ) {alphaBoolBlend {blendOption=Linear}}
    Additive=Blendspace Player 'SitPlant_COR_Loco_Turn_BS'(X=Get CompanionDirection()) {blendSpace=&SitPlant_COR_Loco_Turn_BS}
   ) {alphaBoolBlend {blendOption=Linear}}
   AnimGraphNode_StateMachine_4 = LowFidelity_Tail() {is=title AnimGraphNode_StateMachine, bAllowConduitEntryStates=true}
   n17 = Get TailWagAlpha()
   n18 = AimOffset Player 'SitPlant_AO'!(
    X=Get LookAtYawAdjusted()
    Y=Get LookAtPitchActual()
    BasePose=Slot 'DefaultSlot'(Source=Apply Additive(Base=@n16, Additive=@AnimGraphNode_StateMachine_4, Alpha=@n17) {alphaCurveName=Appendage01, alphaBoolBlend {blendOption=Linear}})
    AlphaCurveName:Name=DisableAimOffset
   ) -> (Pose) {
    alphaInputType=Curve
    blendSpace=&SitPlant_AO
    alphaBoolBlend {blendOption=Linear}
    alphaScaleBiasClamp {bias=1, scale=-1}
   }
   Save cached pose 'LowFidelityPose'(Pose=Modify Curve(SourcePose=@n18, CurveValues_0=1, Alpha=1) {curveNames [LowFidelity], curveValues [0]}) {cacheName=LowFidelityPose}
   AnimGraphNode_ChooserPlayer_3 = Chooser Player: PRA_Companion_FacePose!() -> (Pose:Pose Link Structure) {
    is=title AnimGraphNode_ChooserPlayer
    chooser {
     _structType=/Script/ProxyTable.LookupProxy
     proxy=/CosmeticCompanions/Chooser/PRA_Companion_FacePose
     proxyTable {
      _structType=/Script/ProxyTable.ProxyTableContextProperty
      binding {contextIndex=0, displayName=Chooser Tables, isBoundToRoot=false, propertyBindingChain [ChooserTables]}
     }
    }
    chooserContextDefinition [
     {
      _structType=/Script/Chooser.ContextObjectTypeClass
      class=&/SitPlant_AnimBP.SitPlant_AnimBP_C
      direction=ReadWrite
     }
     {struct=/Script/Chooser.ChooserPlayerSettings, _structType=/Script/Chooser.ContextObjectTypeStruct, direction=Read}
    ]
    defaultSettings {blendOption=Linear}
   }
   n19 = Layered blend per bone(BasePose=Use cached pose 'LocomotionPlusAdditivePose'(), BlendPoses_0=@AnimGraphNode_ChooserPlayer_3) {blendMode=BlendMask, blendMasks [&SitPlant_Quadruped_Skeleton:FaceOff], layerSetup []}
   n20 = Get bIsMoodCombat() -> (bIsMoodCombat:bool)
   n21 = Get bIsMoving() -> (bIsMoving:bool)
   Save cached pose 'HighFidelityPose'(Pose=Blend Poses by bool(
     BlendPose_0=@n19
     BlendPose_1=Use cached pose 'LocomotionPlusAdditivePose'()
     BlendTime_0=0.3
     BlendTime_1=0.3
     bActiveValue=OR Boolean(A=@n21, B=@n20)
    ) {blendType=Linear}) {cacheName=HighFidelityPose}
   n22 = Blend Poses by bool(
    BlendPose_0=Use cached pose 'LowFidelityPose'()
    BlendPose_1=Use cached pose 'HighFidelityPose'()
    BlendTime_0=0.5
    BlendTime_1=0.5
    bActiveValue=Get bUseLowFidelity() -> (bUseLowFidelity:bool)
   ) {
    blendProfile=&SitPlant_Quadruped_Skeleton:TimeBlend_FastFeetSlowBody
    blendType=Linear
    transitionType=Inertialization
   }
   Save cached pose 'SavedPose21080'(Pose=Dead Blending!(Source:Pose Link Structure=Slot 'SequencerDefault'(Source=@n22) {slotName=SequencerDefault}) -> (Pose:Pose Link Structure) {is=title AnimGraphNode_DeadBlending, defaultBlendMode=Linear}) {cacheName=SavedPose21080}
  ]}
 AnimationBlendStackGraph_0 {
  path=AnimGraph.AnimGraphNode_ChooserPlayer_3.AnimationBlendStackGraph_0
  nodes [Output Pose!(Result:Pose Link Structure=Blend Stack Input!() -> (Pose:Pose Link Structure) {is=title AnimGraphNode_BlendStackInput}) {is=title AnimGraphNode_BlendStackResult}]
 }
 IdleAntics {
  entry=Conduit
  path=AnimGraph.AnimGraphNode_StateMachine_0.IdleAntics
  ids=id
  nodes [
   AnimStateConduitNode_0 = Conduit()
   AnimStateNode_0 = Default() {is=title AnimStateNode}
   AnimStateNode_7 = FrontEndSpawnAnticChooser() {is=title AnimStateNode}
   AnimStateNode_9 = AnticChooser() {is=title AnimStateNode}
   AnimStateTransitionNode_0 = AnimStateTransitionNode() {
    crossfadeDuration=0.25
    logicType=TLT_Inertialization
    priorityOrder=100
    blendProfileWrapper {blendProfile=&SitPlant_Quadruped_Skeleton:TimeBlend_FastFeetSlowBody}
   }
   AnimStateTransitionNode_1 = AnimStateTransitionNode() {bAutomaticRuleBasedOnSequencePlayerInState=true}
   AnimStateTransitionNode_14 = AnimStateTransitionNode() {logicType=TLT_Inertialization}
   AnimStateTransitionNode_18 = AnimStateTransitionNode() {logicType=TLT_Inertialization, priorityOrder=2}
   AnimStateTransitionNode_2 = AnimStateTransitionNode()
   AnimStateTransitionNode_4 = AnimStateTransitionNode()
   AnimStateTransitionNode_5 = AnimStateTransitionNode() {bAutomaticRuleBasedOnSequencePlayerInState=true}
  ]
  links [
   {
    dst=Conduit
    rule=AnimGraph.AnimGraphNode_StateMachine_0.IdleAntics.AnimStateTransitionNode_2.Transition
    src=AnticChooser
    via=AnimStateTransitionNode_2
   }
   {
    dst=Conduit
    rule=AnimGraph.AnimGraphNode_StateMachine_0.IdleAntics.AnimStateTransitionNode_5.Transition
    src=AnticChooser
    via=AnimStateTransitionNode_5
   }
   {
    dst=AnticChooser
    rule=AnimGraph.AnimGraphNode_StateMachine_0.IdleAntics.AnimStateTransitionNode_18.Transition
    src=Conduit
    via=AnimStateTransitionNode_18
   }
   {
    dst=Default
    rule=AnimGraph.AnimGraphNode_StateMachine_0.IdleAntics.AnimStateTransitionNode_0.Transition
    src=Conduit
    via=AnimStateTransitionNode_0
   }
   {
    dst=FrontEndSpawnAnticChooser
    rule=AnimGraph.AnimGraphNode_StateMachine_0.IdleAntics.AnimStateTransitionNode_14.Transition
    src=Conduit
    via=AnimStateTransitionNode_14
   }
   {
    dst=Conduit
    rule=AnimGraph.AnimGraphNode_StateMachine_0.IdleAntics.AnimStateTransitionNode_4.Transition
    src=Default
    via=AnimStateTransitionNode_4
   }
   {
    dst=Conduit
    rule=AnimGraph.AnimGraphNode_StateMachine_0.IdleAntics.AnimStateTransitionNode_1.Transition
    src=FrontEndSpawnAnticChooser
    via=AnimStateTransitionNode_1
   }
  ]
 }
 Conduit {path=AnimGraph.AnimGraphNode_StateMachine_0.IdleAntics.AnimStateConduitNode_0.Conduit, nodes [Result(bCanEnterTransition=true)]}
 Default {
  path=AnimGraph.AnimGraphNode_StateMachine_0.IdleAntics.AnimStateNode_0.Default
  nodes [
   AnimGraphNode_ChooserPlayer_1 = Chooser Player: PRA_Companion_AnticsNeutral!() -> (Pose:Pose Link Structure) {
    is=title AnimGraphNode_ChooserPlayer
    chooser {
     _structType=&LookupProxy
     proxy=/CosmeticCompanions/Chooser/PRA_Companion_AnticsNeutral
     proxyTable {
      _structType=&ProxyTableContextProperty
      binding {contextIndex=0, displayName=Chooser Tables, isBoundToRoot=false, propertyBindingChain [ChooserTables]}
     }
    }
    chooserContextDefinition [
     {
      _structType=&ContextObjectTypeClass
      class=&SitPlant_AnimBP_C
      direction=ReadWrite
     }
     {struct=&ChooserPlayerSettings, _structType=&ContextObjectTypeStruct, direction=Read}
    ]
    defaultSettings {blendOption=Linear}
   }
   Output Animation Pose(Result=@AnimGraphNode_ChooserPlayer_1)
  ]
 }
 AnimationBlendStackGraph_0 {
  path=AnimGraph.AnimGraphNode_StateMachine_0.IdleAntics.AnimStateNode_0.Default.AnimGraphNode_ChooserPlayer_1.AnimationBlendStackGraph_0
  nodes [Output Pose(Result=Blend Stack Input())]
 }
 FrontEndSpawnAnticChooser {
  path=AnimGraph.AnimGraphNode_StateMachine_0.IdleAntics.AnimStateNode_7.FrontEndSpawnAnticChooser
  nodes [
   n1 = Get OnSpawnedChooser() -> (OnSpawnedChooser:Async Sequence Chooser Structure)
   Output Animation Pose(Result=Sequence Player!(Sequence:Anim Sequence Base Object Reference=Break Async Sequence Chooser!(AsyncSequenceChooser:"Async Sequence Chooser Structure (by ref)"=@n1) -> (
      bIsLoading:bool=~
      bIsMontage:bool=~
      bIsReady:bool=~
      AnimSequence:Anim Sequence Object Reference
      AnimMontage:Anim Montage Object Reference
     ) {is=title K2Node_BreakStruct}.AnimSequence) -> (Pose) {bLoopAnimation=false}) {
    node {stateEntryFunction {functionName=OnStateEntry_3}, stateExitFunction {functionName=OnStateExit_0}}
    stateEntryFunction {bSelfContext=true, memberGuid=#1, memberName=IsFrontEndSpawn_False}
    stateExitFunction {bSelfContext=true, memberGuid=#2, memberName=AllowedToSit_Exit}
   }
  ]
 }
 AnticChooser {
  path=AnimGraph.AnimGraphNode_StateMachine_0.IdleAntics.AnimStateNode_9.AnticChooser
  nodes [Output Animation Pose(Result=Sequence Player() {bLoopAnimation=false}) {
    node {stateEntryFunction {functionName=OnStateEntry}, stateExitFunction {functionName=OnStateExit_0}}
    stateEntryFunction {bSelfContext=true, memberGuid=#3, memberName=AnticPlaying_Enter}
    stateExitFunction {bSelfContext=true, memberGuid=#4, memberName=AnticPlaying_Exit}
   }]
 }
 Transition {path=AnimGraph.AnimGraphNode_StateMachine_0.IdleAntics.AnimStateTransitionNode_0.Transition, nodes [Result(bCanEnterTransition=true)]}
 Transition {path=AnimGraph.AnimGraphNode_StateMachine_0.IdleAntics.AnimStateTransitionNode_1.Transition, nodes [Result()]}
 Transition {
  path=AnimGraph.AnimGraphNode_StateMachine_0.IdleAntics.AnimStateTransitionNode_14.Transition
  nodes [Result(bCanEnterTransition=Break Async Sequence Chooser(AsyncSequenceChooser=Get OnSpawnedChooser()) -> (bIsLoading=~, bIsMontage=~, bIsReady=~).bIsReady)]
 }
 Transition {
  path=AnimGraph.AnimGraphNode_StateMachine_0.IdleAntics.AnimStateTransitionNode_18.Transition
  nodes [Result(bCanEnterTransition=Get bIsAllowAnticToPlay() -> (bIsAllowAnticToPlay:bool))]
 }
 Transition {
  path=AnimGraph.AnimGraphNode_StateMachine_0.IdleAntics.AnimStateTransitionNode_2.Transition
  nodes [Result(bCanEnterTransition=NOT Boolean(A=Get bIsCanPlayAntic() -> (bIsCanPlayAntic:bool)))]
 }
 Transition {path=AnimGraph.AnimGraphNode_StateMachine_0.IdleAntics.AnimStateTransitionNode_4.Transition, nodes [Result(bCanEnterTransition=true)]}
 Transition {path=AnimGraph.AnimGraphNode_StateMachine_0.IdleAntics.AnimStateTransitionNode_5.Transition, nodes [Result()]}
 LowFidelityBodyAdditives {
  entry=BodyAdditiveTempState
  path=AnimGraph.AnimGraphNode_StateMachine_1.LowFidelityBodyAdditives
  ids=id
  nodes [
   AnimStateNode_0 = BodyAdditiveTempState() {is=title AnimStateNode}
   AnimStateNode_1 = Worried() {is=title AnimStateNode}
   AnimStateNode_2 = Combat() {is=title AnimStateNode}
   AnimStateTransitionNode_0 = AnimStateTransitionNode() {crossfadeDuration=0.35, logicType=TLT_Inertialization}
   AnimStateTransitionNode_1 = AnimStateTransitionNode() {crossfadeDuration=0.35, logicType=TLT_Inertialization}
   AnimStateTransitionNode_2 = AnimStateTransitionNode() {logicType=TLT_Inertialization, priorityOrder=2}
   AnimStateTransitionNode_3 = AnimStateTransitionNode() {crossfadeDuration=0.35}
  ]
  links [
   {
    dst=Combat
    rule=AnimGraph.AnimGraphNode_StateMachine_1.LowFidelityBodyAdditives.AnimStateTransitionNode_2.Transition
    src=BodyAdditiveTempState
    via=AnimStateTransitionNode_2
   }
   {
    dst=Worried
    rule=AnimGraph.AnimGraphNode_StateMachine_1.LowFidelityBodyAdditives.AnimStateTransitionNode_0.Transition
    src=BodyAdditiveTempState
    via=AnimStateTransitionNode_0
   }
   {
    dst=BodyAdditiveTempState
    rule=AnimGraph.AnimGraphNode_StateMachine_1.LowFidelityBodyAdditives.AnimStateTransitionNode_3.Transition
    src=Combat
    via=AnimStateTransitionNode_3
   }
   {
    dst=BodyAdditiveTempState
    rule=AnimGraph.AnimGraphNode_StateMachine_1.LowFidelityBodyAdditives.AnimStateTransitionNode_1.Transition
    src=Worried
    via=AnimStateTransitionNode_1
   }
  ]
 }
 BodyAdditiveTempState {
  path=AnimGraph.AnimGraphNode_StateMachine_1.LowFidelityBodyAdditives.AnimStateNode_0.BodyAdditiveTempState
  nodes [Output Animation Pose(Result=Additive Identity Pose())]
 }
 Worried {
  path=AnimGraph.AnimGraphNode_StateMachine_1.LowFidelityBodyAdditives.AnimStateNode_1.Worried
  nodes [Output Animation Pose(Result=Two Way Blend(
     A=Additive Identity Pose()
     B="Sequence Player (additive) 'SitPlant_COR_Worried_Pose_ADD'"() {bLoopAnimation=false, sequence=/CosmeticCompanions/Assets/Quadruped/SitPlant/Animation/Core/SitPlant_COR_Worried_Pose_ADD}
     Alpha=Get WorriedAddPoseAlpha() -> (WorriedAddPoseAlpha:double)
    ) {blendNode {alphaBoolBlend {blendOption=Linear}}})]
 }
 Combat {
  path=AnimGraph.AnimGraphNode_StateMachine_1.LowFidelityBodyAdditives.AnimStateNode_2.Combat
  nodes [Output Animation Pose(Result="Sequence Player (additive) 'SitPlant_COR_Combat_Pose_ADD'"() {bLoopAnimation=false, sequence=/CosmeticCompanions/Assets/Quadruped/SitPlant/Animation/Core/SitPlant_COR_Combat_Pose_ADD})]
 }
 Transition {
  path=AnimGraph.AnimGraphNode_StateMachine_1.LowFidelityBodyAdditives.AnimStateTransitionNode_0.Transition
  nodes [Result(bCanEnterTransition=Get bIsMoodWorried() -> (bIsMoodWorried:bool))]
 }
 Transition {
  path=AnimGraph.AnimGraphNode_StateMachine_1.LowFidelityBodyAdditives.AnimStateTransitionNode_1.Transition
  nodes [Result(bCanEnterTransition=NOT Boolean(A=Get bIsMoodWorried()))]
 }
 Transition {
  path=AnimGraph.AnimGraphNode_StateMachine_1.LowFidelityBodyAdditives.AnimStateTransitionNode_2.Transition
  nodes [Result(bCanEnterTransition=Get bIsMoodCombat())]
 }
 Transition {
  path=AnimGraph.AnimGraphNode_StateMachine_1.LowFidelityBodyAdditives.AnimStateTransitionNode_3.Transition
  nodes [Result(bCanEnterTransition=NOT Boolean(A=Get bIsMoodCombat()))]
 }
 Tail {
  entry=Conduit
  path=AnimGraph.AnimGraphNode_StateMachine_2.Tail
  ids=id
  nodes [
   AnimStateConduitNode_0 = Conduit()
   AnimStateNode_0 = Default()
   AnimStateNode_2 = Combat()
   AnimStateTransitionNode_0 = AnimStateTransitionNode() {priorityOrder=100}
   AnimStateTransitionNode_1 = AnimStateTransitionNode()
   AnimStateTransitionNode_4 = AnimStateTransitionNode() {crossfadeDuration=0.5, priorityOrder=2}
   AnimStateTransitionNode_5 = AnimStateTransitionNode()
  ]
  links [
   {
    dst=Conduit
    rule=AnimGraph.AnimGraphNode_StateMachine_2.Tail.AnimStateTransitionNode_5.Transition
    src=Combat
    via=AnimStateTransitionNode_5
   }
   {
    dst=Combat
    rule=AnimGraph.AnimGraphNode_StateMachine_2.Tail.AnimStateTransitionNode_4.Transition
    src=Conduit
    via=AnimStateTransitionNode_4
   }
   {
    dst=Default
    rule=AnimGraph.AnimGraphNode_StateMachine_2.Tail.AnimStateTransitionNode_0.Transition
    src=Conduit
    via=AnimStateTransitionNode_0
   }
   {
    dst=Conduit
    rule=AnimGraph.AnimGraphNode_StateMachine_2.Tail.AnimStateTransitionNode_1.Transition
    src=Default
    via=AnimStateTransitionNode_1
   }
  ]
 }
 Conduit {path=AnimGraph.AnimGraphNode_StateMachine_2.Tail.AnimStateConduitNode_0.Conduit, nodes [Result(bCanEnterTransition=true)]}
 Default {
  path=AnimGraph.AnimGraphNode_StateMachine_2.Tail.AnimStateNode_0.Default
  nodes [
   n1 = Blendspace Player 'SitPlant_COR_Tail_BS'!(X=Get TailWagSpeed() -> (TailWagSpeed:float)) -> (Pose) {blendSpace=/CosmeticCompanions/Assets/Quadruped/SitPlant/Animation/Core/SitPlant_COR_Tail_BS, groupName=TailWag, method=SyncGroup}
   Output Animation Pose(Result=Blend Poses by bool(
     BlendPose_0="Sequence Player (additive) 'SitPlant_COR_Tail_Sit_ADD'"() {sequence=/CosmeticCompanions/Assets/Quadruped/SitPlant/Animation/Core/SitPlant_COR_Tail_Sit_ADD}
     BlendPose_1=@n1
     BlendTime_0=0.3
     BlendTime_1=0.3
     bActiveValue=Get bIsSitting() -> (bIsSitting:bool)
    ) {blendType=Linear, transitionType=Inertialization})
  ]
 }
 Combat {
  path=AnimGraph.AnimGraphNode_StateMachine_2.Tail.AnimStateNode_2.Combat
  nodes [Output Animation Pose(Result="Sequence Player (additive) 'SitPlant_COR_Tail_Angry_ADD'"() {sequence=/CosmeticCompanions/Assets/Quadruped/SitPlant/Animation/Core/SitPlant_COR_Tail_Angry_ADD})]
 }
 Transition {path=AnimGraph.AnimGraphNode_StateMachine_2.Tail.AnimStateTransitionNode_0.Transition, nodes [Result(bCanEnterTransition=true)]}
 Transition {path=AnimGraph.AnimGraphNode_StateMachine_2.Tail.AnimStateTransitionNode_1.Transition, nodes [Result(bCanEnterTransition=true)]}
 Transition {
  path=AnimGraph.AnimGraphNode_StateMachine_2.Tail.AnimStateTransitionNode_4.Transition
  nodes [Result(bCanEnterTransition=Get bIsDelayedMoodCombat() -> (bIsDelayedMoodCombat:bool))]
 }
 Transition {path=AnimGraph.AnimGraphNode_StateMachine_2.Tail.AnimStateTransitionNode_5.Transition, nodes [Result(bCanEnterTransition=true)]}
 LowFidelityStateMachine {
  entry=Conduit
  path=AnimGraph.AnimGraphNode_StateMachine_3.LowFidelityStateMachine
  ids=id
  nodes [
   AnimStateConduitNode_0 = Conduit()
   AnimStateNode_0 = Idle() {bAlwaysResetOnEntry=true}
   AnimStateNode_1 = ReactLoops() {is=title AnimStateNode}
   AnimStateNode_2 = ReactToIdle() {is=title AnimStateNode}
   AnimStateNode_3 = Locomotion() {is=title AnimStateNode, bAlwaysResetOnEntry=true}
   AnimStateTransitionNode_0 = AnimStateTransitionNode() {crossfadeDuration=0.4, logicType=TLT_Inertialization, priorityOrder=100}
   AnimStateTransitionNode_12 = AnimStateTransitionNode() {logicType=TLT_Inertialization}
   AnimStateTransitionNode_13 = AnimStateTransitionNode() {bAutomaticRuleBasedOnSequencePlayerInState=true}
   AnimStateTransitionNode_14 = AnimStateTransitionNode() {logicType=TLT_Inertialization, priorityOrder=2}
   AnimStateTransitionNode_15 = AnimStateTransitionNode()
   AnimStateTransitionNode_16 = AnimStateTransitionNode()
   AnimStateTransitionNode_2 = AnimStateTransitionNode()
   AnimStateTransitionNode_4 = AnimStateTransitionNode()
   AnimStateTransitionNode_7 = AnimStateTransitionNode() {logicType=TLT_Inertialization}
   AnimStateTransitionNode_8 = AnimStateTransitionNode() {
    crossfadeDuration=0.4
    priorityOrder=3
    blendProfileWrapper {blendProfile=&SitPlant_Quadruped_Skeleton:TimeBlend_FastFeetSlowBody}
   }
  ]
  links [
   {
    dst=Idle
    rule=AnimGraph.AnimGraphNode_StateMachine_3.LowFidelityStateMachine.AnimStateTransitionNode_0.Transition
    src=Conduit
    via=AnimStateTransitionNode_0
   }
   {
    dst=Locomotion
    rule=AnimGraph.AnimGraphNode_StateMachine_3.LowFidelityStateMachine.AnimStateTransitionNode_8.Transition
    src=Conduit
    via=AnimStateTransitionNode_8
   }
   {
    dst=ReactLoops
    rule=AnimGraph.AnimGraphNode_StateMachine_3.LowFidelityStateMachine.AnimStateTransitionNode_7.Transition
    src=Conduit
    via=AnimStateTransitionNode_7
   }
   {
    dst=ReactToIdle
    rule=AnimGraph.AnimGraphNode_StateMachine_3.LowFidelityStateMachine.AnimStateTransitionNode_14.Transition
    src=Conduit
    via=AnimStateTransitionNode_14
   }
   {
    dst=Conduit
    rule=AnimGraph.AnimGraphNode_StateMachine_3.LowFidelityStateMachine.AnimStateTransitionNode_2.Transition
    src=Idle
    via=AnimStateTransitionNode_2
   }
   {
    dst=Conduit
    rule=AnimGraph.AnimGraphNode_StateMachine_3.LowFidelityStateMachine.AnimStateTransitionNode_4.Transition
    src=Locomotion
    via=AnimStateTransitionNode_4
   }
   {
    dst=Conduit
    rule=AnimGraph.AnimGraphNode_StateMachine_3.LowFidelityStateMachine.AnimStateTransitionNode_15.Transition
    src=ReactLoops
    via=AnimStateTransitionNode_15
   }
   {
    dst=ReactToIdle
    rule=AnimGraph.AnimGraphNode_StateMachine_3.LowFidelityStateMachine.AnimStateTransitionNode_12.Transition
    src=ReactLoops
    via=AnimStateTransitionNode_12
   }
   {
    dst=Conduit
    rule=AnimGraph.AnimGraphNode_StateMachine_3.LowFidelityStateMachine.AnimStateTransitionNode_13.Transition
    src=ReactToIdle
    via=AnimStateTransitionNode_13
   }
   {
    dst=Conduit
    rule=AnimGraph.AnimGraphNode_StateMachine_3.LowFidelityStateMachine.AnimStateTransitionNode_16.Transition
    src=ReactToIdle
    via=AnimStateTransitionNode_16
   }
  ]
 }
 Conduit {path=AnimGraph.AnimGraphNode_StateMachine_3.LowFidelityStateMachine.AnimStateConduitNode_0.Conduit, nodes [Result(bCanEnterTransition=true)]}
 Idle {
  path=AnimGraph.AnimGraphNode_StateMachine_3.LowFidelityStateMachine.AnimStateNode_0.Idle
  nodes [
   AnimGraphNode_StateMachine_1 = Idle() {is=title AnimGraphNode_StateMachine, bAllowConduitEntryStates=true}
   Output Animation Pose(Result=Modify Curve(SourcePose=@AnimGraphNode_StateMachine_1, CurveValues_0=1, Alpha=1) {curveNames [idle], curveValues [0]}) {
    node {stateExitFunction {functionName=OnExit_SitState}}
    stateExitFunction {bSelfContext=true, memberGuid=#5, memberName=OnExit_SitState}
   }
  ]
 }
 Idle {
  entry=Conduit
  path=AnimGraph.AnimGraphNode_StateMachine_3.LowFidelityStateMachine.AnimStateNode_0.Idle.AnimGraphNode_StateMachine_1.Idle
  ids=id
  nodes [
   Alias() {is=title AnimStateAliasNode, also {aliases [SitPose,StandToSit]}}
   AnimStateConduitNode_0 = Conduit()
   AnimStateNode_0 = StandToSit() {is=title AnimStateNode}
   AnimStateNode_1 = SitPose() {is=title AnimStateNode}
   AnimStateNode_2 = StandingPose() {is=title AnimStateNode}
   AnimStateNode_3 = SitToStand() {is=title AnimStateNode}
   AnimStateTransitionNode_0 = AnimStateTransitionNode() {bAutomaticRuleBasedOnSequencePlayerInState=true, logicType=TLT_Inertialization}
   AnimStateTransitionNode_1 = AnimStateTransitionNode() {logicType=TLT_Inertialization, priorityOrder=2}
   AnimStateTransitionNode_2 = AnimStateTransitionNode() {logicType=TLT_Inertialization, priorityOrder=100}
   AnimStateTransitionNode_3 = AnimStateTransitionNode() {logicType=TLT_Inertialization}
   AnimStateTransitionNode_4 = AnimStateTransitionNode()
   AnimStateTransitionNode_6 = AnimStateTransitionNode() {bAutomaticRuleBasedOnSequencePlayerInState=true, crossfadeDuration=0.3, logicType=TLT_Inertialization}
   AnimStateTransitionNode_8 = AnimStateTransitionNode() {logicType=TLT_Inertialization}
  ]
  links [
   {
    dst=SitToStand
    rule=AnimGraph.AnimGraphNode_StateMachine_3.LowFidelityStateMachine.AnimStateNode_0.Idle.AnimGraphNode_StateMachine_1.Idle.AnimStateTransitionNode_8.AnimationTransitionGraph_1
    src=Alias
    via=AnimStateTransitionNode_8
   }
   {
    dst=SitPose
    rule=AnimGraph.AnimGraphNode_StateMachine_3.LowFidelityStateMachine.AnimStateNode_0.Idle.AnimGraphNode_StateMachine_1.Idle.AnimStateTransitionNode_3.AnimationTransitionGraph_1
    src=Conduit
    via=AnimStateTransitionNode_3
   }
   {
    dst=StandToSit
    rule=AnimGraph.AnimGraphNode_StateMachine_3.LowFidelityStateMachine.AnimStateNode_0.Idle.AnimGraphNode_StateMachine_1.Idle.AnimStateTransitionNode_1.AnimationTransitionGraph_1
    src=Conduit
    via=AnimStateTransitionNode_1
   }
   {
    dst=StandingPose
    rule=AnimGraph.AnimGraphNode_StateMachine_3.LowFidelityStateMachine.AnimStateNode_0.Idle.AnimGraphNode_StateMachine_1.Idle.AnimStateTransitionNode_2.AnimationTransitionGraph_1
    src=Conduit
    via=AnimStateTransitionNode_2
   }
   {
    dst=StandingPose
    rule=AnimGraph.AnimGraphNode_StateMachine_3.LowFidelityStateMachine.AnimStateNode_0.Idle.AnimGraphNode_StateMachine_1.Idle.AnimStateTransitionNode_6.AnimationTransitionGraph_1
    src=SitToStand
    via=AnimStateTransitionNode_6
   }
   {
    dst=SitPose
    rule=AnimGraph.AnimGraphNode_StateMachine_3.LowFidelityStateMachine.AnimStateNode_0.Idle.AnimGraphNode_StateMachine_1.Idle.AnimStateTransitionNode_0.AnimationTransitionGraph_1
    src=StandToSit
    via=AnimStateTransitionNode_0
   }
   {
    dst=Conduit
    rule=AnimGraph.AnimGraphNode_StateMachine_3.LowFidelityStateMachine.AnimStateNode_0.Idle.AnimGraphNode_StateMachine_1.Idle.AnimStateTransitionNode_4.AnimationTransitionGraph_1
    src=StandingPose
    via=AnimStateTransitionNode_4
   }
  ]
 }
 Conduit {
  path=AnimGraph.AnimGraphNode_StateMachine_3.LowFidelityStateMachine.AnimStateNode_0.Idle.AnimGraphNode_StateMachine_1.Idle.AnimStateConduitNode_0.Conduit
  nodes [Result(bCanEnterTransition=true)]
 }
 StandToSit {
  path=AnimGraph.AnimGraphNode_StateMachine_3.LowFidelityStateMachine.AnimStateNode_0.Idle.AnimGraphNode_StateMachine_1.Idle.AnimStateNode_0.StandToSit
  nodes [Output Animation Pose(Result=Sequence Player 'SitPlant_COR_Sit_In'() {bLoopAnimation=false, sequence=/CosmeticCompanions/Assets/Quadruped/SitPlant/Animation/Core/SitPlant_COR_Sit_In}) {
    node {stateEntryFunction {functionName=OnEnterSitState}}
    stateEntryFunction {bSelfContext=true, memberGuid=#6, memberName=IsSit_Enter}
   }]
 }
 SitPose {
  path=AnimGraph.AnimGraphNode_StateMachine_3.LowFidelityStateMachine.AnimStateNode_0.Idle.AnimGraphNode_StateMachine_1.Idle.AnimStateNode_1.SitPose
  nodes [Output Animation Pose(Result=Sequence Player 'SitPlant_COR_Sit_Pose'() {bLoopAnimation=false, sequence=/CosmeticCompanions/Assets/Quadruped/SitPlant/Animation/Core/SitPlant_COR_Sit_Pose}) {
    node {stateEntryFunction {functionName=OnStateEntry}, stateExitFunction {functionName=CanPlayAntic}}
    stateEntryFunction {bSelfContext=true, memberGuid=#7, memberName=CanPlayAntic_Enter}
    stateExitFunction {bSelfContext=true, memberGuid=#7, memberName=CanPlayAntic_Enter}
   }]
 }
 StandingPose {
  path=AnimGraph.AnimGraphNode_StateMachine_3.LowFidelityStateMachine.AnimStateNode_0.Idle.AnimGraphNode_StateMachine_1.Idle.AnimStateNode_2.StandingPose
  nodes [Output Animation Pose(Result=Blend Poses by bool(
     BlendPose_0=Sequence Player 'SitPlant_xDEV_React_Angry_02_Loop_OLD'() {sequence=/CosmeticCompanions/Assets/Quadruped/SitPlant/Animation/Developer/SitPlant_xDEV_React_Angry_02_Loop_OLD}
     BlendPose_1=Sequence Player 'SitPlant_COR_Stand_Pose'() {bLoopAnimation=false, sequence=/CosmeticCompanions/Assets/Quadruped/SitPlant/Animation/Core/SitPlant_COR_Stand_Pose}
     BlendTime_0=0.3
     BlendTime_1=0.4
     bActiveValue=Get bIsMoodCombat()
    ) {blendType=Linear, transitionType=Inertialization})]
 }
 SitToStand {
  path=AnimGraph.AnimGraphNode_StateMachine_3.LowFidelityStateMachine.AnimStateNode_0.Idle.AnimGraphNode_StateMachine_1.Idle.AnimStateNode_3.SitToStand
  nodes [Output Animation Pose(Result=Sequence Player 'SitPlant_COR_Sit_Out'() {bLoopAnimation=false, sequence=/CosmeticCompanions/Assets/Quadruped/SitPlant/Animation/Core/SitPlant_COR_Sit_Out}) {
    node {stateEntryFunction {functionName=OnExitSitState}}
    stateEntryFunction {bSelfContext=true, memberGuid=#5, memberName=OnExit_SitState}
   }]
 }
 AnimationTransitionGraph_1 {
  path=AnimGraph.AnimGraphNode_StateMachine_3.LowFidelityStateMachine.AnimStateNode_0.Idle.AnimGraphNode_StateMachine_1.Idle.AnimStateTransitionNode_0.AnimationTransitionGraph_1
  nodes [Result()]
 }
 AnimationTransitionGraph_1 {
  path=AnimGraph.AnimGraphNode_StateMachine_3.LowFidelityStateMachine.AnimStateNode_0.Idle.AnimGraphNode_StateMachine_1.Idle.AnimStateTransitionNode_1.AnimationTransitionGraph_1
  nodes [Result(bCanEnterTransition=Get bIsStandToSit() -> (bIsStandToSit:bool))]
 }
 AnimationTransitionGraph_1 {
  path=AnimGraph.AnimGraphNode_StateMachine_3.LowFidelityStateMachine.AnimStateNode_0.Idle.AnimGraphNode_StateMachine_1.Idle.AnimStateTransitionNode_2.AnimationTransitionGraph_1
  nodes [Result(bCanEnterTransition=true)]
 }
 AnimationTransitionGraph_1 {
  path=AnimGraph.AnimGraphNode_StateMachine_3.LowFidelityStateMachine.AnimStateNode_0.Idle.AnimGraphNode_StateMachine_1.Idle.AnimStateTransitionNode_3.AnimationTransitionGraph_1
  nodes [Result(bCanEnterTransition=Get bIsSitting())]
 }
 AnimationTransitionGraph_1 {
  path=AnimGraph.AnimGraphNode_StateMachine_3.LowFidelityStateMachine.AnimStateNode_0.Idle.AnimGraphNode_StateMachine_1.Idle.AnimStateTransitionNode_4.AnimationTransitionGraph_1
  nodes [Result(bCanEnterTransition=true)]
 }
 AnimationTransitionGraph_1 {
  path=AnimGraph.AnimGraphNode_StateMachine_3.LowFidelityStateMachine.AnimStateNode_0.Idle.AnimGraphNode_StateMachine_1.Idle.AnimStateTransitionNode_6.AnimationTransitionGraph_1
  nodes [Result()]
 }
 AnimationTransitionGraph_1 {
  path=AnimGraph.AnimGraphNode_StateMachine_3.LowFidelityStateMachine.AnimStateNode_0.Idle.AnimGraphNode_StateMachine_1.Idle.AnimStateTransitionNode_8.AnimationTransitionGraph_1
  nodes [Result(bCanEnterTransition=Get bIsSitToStand() -> (bIsSitToStand:bool))]
 }
 ReactLoops {
  path=AnimGraph.AnimGraphNode_StateMachine_3.LowFidelityStateMachine.AnimStateNode_1.ReactLoops
  nodes [
   AnimGraphNode_ChooserPlayer_1 = Chooser Player: PRA_Companion_SpecialIdle!() -> (Pose:Pose Link Structure) {
    is=title AnimGraphNode_ChooserPlayer
    chooser {
     _structType=&LookupProxy
     proxy=/CosmeticCompanions/Chooser/PRA_Companion_SpecialIdle
     proxyTable {
      _structType=&ProxyTableContextProperty
      binding {contextIndex=0, displayName=Chooser Tables, isBoundToRoot=false, propertyBindingChain [ChooserTables]}
     }
    }
    chooserContextDefinition [
     {
      _structType=&ContextObjectTypeClass
      class=&SitPlant_AnimBP_C
      direction=ReadWrite
     }
     {struct=&ChooserPlayerSettings, _structType=&ContextObjectTypeStruct, direction=Read}
    ]
    defaultSettings {blendOption=Linear}
   }
   Output Animation Pose(Result=@AnimGraphNode_ChooserPlayer_1)
  ]
 }
 AnimationBlendStackGraph_0 {
  path=AnimGraph.AnimGraphNode_StateMachine_3.LowFidelityStateMachine.AnimStateNode_1.ReactLoops.AnimGraphNode_ChooserPlayer_1.AnimationBlendStackGraph_0
  nodes [Output Pose(Result=Blend Stack Input())]
 }
 ReactToIdle {
  path=AnimGraph.AnimGraphNode_StateMachine_3.LowFidelityStateMachine.AnimStateNode_2.ReactToIdle
  nodes [
   AnimGraphNode_ChooserPlayer_1 = Chooser Player: PRA_Companion_ReactToIdle!() -> (Pose:Pose Link Structure) {
    is=title AnimGraphNode_ChooserPlayer
    chooser {
     _structType=&LookupProxy
     proxy=/CosmeticCompanions/Chooser/PRA_Companion_ReactToIdle
     proxyTable {
      _structType=&ProxyTableContextProperty
      binding {contextIndex=0, displayName=Chooser Tables, isBoundToRoot=false, propertyBindingChain [ChooserTables]}
     }
    }
    chooserContextDefinition [
     {
      _structType=&ContextObjectTypeClass
      class=&SitPlant_AnimBP_C
      direction=ReadWrite
     }
     {struct=&ChooserPlayerSettings, _structType=&ContextObjectTypeStruct, direction=Read}
    ]
    defaultSettings {blendOption=Linear}
   }
   Output Animation Pose(Result=Modify Curve(SourcePose=@AnimGraphNode_ChooserPlayer_1, CurveValues_0=1, Alpha=1) {curveNames [idle], curveValues [0]}) {
    node {stateEntryFunction {functionName=SpecialIdle_StateEntry}, stateExitFunction {functionName=SpecialIdle_StateExit}}
    stateEntryFunction {bSelfContext=true, memberGuid=#8, memberName=SpecialIdle_StateEntry}
    stateExitFunction {bSelfContext=true, memberGuid=#9, memberName=SpecialIdle_StateExit}
   }
  ]
 }
 AnimationBlendStackGraph_0 {
  path=AnimGraph.AnimGraphNode_StateMachine_3.LowFidelityStateMachine.AnimStateNode_2.ReactToIdle.AnimGraphNode_ChooserPlayer_1.AnimationBlendStackGraph_0
  nodes [Output Pose(Result=Blend Stack Input())]
 }
 Locomotion {
  path=AnimGraph.AnimGraphNode_StateMachine_3.LowFidelityStateMachine.AnimStateNode_3.Locomotion
  nodes [
   n1 = Sequence Player 'SitPlant_COR_Loco_Walk'!(PlayRate:float=Get WalkPlayRate() -> (WalkPlayRate:double)) -> (Pose) {groupName=Locomotion, method=SyncGroup, sequence=/CosmeticCompanions/Assets/Quadruped/SitPlant/Animation/Core/SitPlant_COR_Loco_Walk}
   n2 = Sequence Player 'SitPlant_COR_Loco_Run'!(PlayRate:float=Get RunPlayRate() -> (RunPlayRate:double)) -> (Pose) {groupName=Locomotion, method=SyncGroup, sequence=/CosmeticCompanions/Assets/Quadruped/SitPlant/Animation/Core/SitPlant_COR_Loco_Run}
   n3 = Sequence Player 'SitPlant_COR_Loco_Sprint'!(PlayRate:float=1) -> (Pose) {groupName=Locomotion, method=SyncGroup, sequence=/CosmeticCompanions/Assets/Quadruped/SitPlant/Animation/Core/SitPlant_COR_Loco_Sprint}
   n4 = Sequence Player 'SitPlant_COR_Loco_Trot'!(PlayRate:float=Get TrotPlayRate() -> (TrotPlayRate:double)) -> (Pose) {groupName=Locomotion, method=SyncGroup, sequence=/CosmeticCompanions/Assets/Quadruped/SitPlant/Animation/Core/SitPlant_COR_Loco_Trot}
   n5 = "Blend Poses (ELocomotionMovementType)"!(
    BlendPose_0:Pose Link Structure
    BlendPose_1:Pose Link Structure=@n3
    BlendPose_2:Pose Link Structure=@n2
    BlendPose_3:Pose Link Structure=@n4
    BlendPose_4:Pose Link Structure=Modify Curve(SourcePose=@n1, CurveValues_0=1, Alpha=1) {curveNames [Walk], curveValues [0]}
    BlendTime_0:float=0.2
    BlendTime_1:float=0.2
    BlendTime_2:float=0.2
    BlendTime_3:float=0.2
    BlendTime_4:float=0.2
    ActiveEnumValue:Byte=Get CurrentLocomotionMovementType() -> (CurrentLocomotionMovementType:ELocomotionMovementType Enum=Walk)
   ) -> (Pose:Pose Link Structure) {
    is=title AnimGraphNode_BlendListByEnum
    blendPose [
     {linkId=-1, sourceLinkId=-1}
     {linkId=-1, sourceLinkId=-1}
     {linkId=-1, sourceLinkId=-1}
     {linkId=-1, sourceLinkId=-1}
     {linkId=-1, sourceLinkId=-1}
    ]
    blendTime [0.1,0.1,0.1,0.1,0.1]
   }
   Output Animation Pose(Result=Modify Curve(SourcePose=@n5, CurveValues_0=1, Alpha=1) {curveNames [Locomotion], curveValues [0]})
  ]
 }
 Transition {
  path=AnimGraph.AnimGraphNode_StateMachine_3.LowFidelityStateMachine.AnimStateTransitionNode_0.Transition
  nodes [Result(bCanEnterTransition=true)]
 }
 Transition {
  path=AnimGraph.AnimGraphNode_StateMachine_3.LowFidelityStateMachine.AnimStateTransitionNode_12.Transition
  nodes [Result(bCanEnterTransition=Get bIsSpecialIdleExit() -> (bIsSpecialIdleExit:bool))]
 }
 Transition {path=AnimGraph.AnimGraphNode_StateMachine_3.LowFidelityStateMachine.AnimStateTransitionNode_13.Transition, nodes [Result()]}
 Transition {
  path=AnimGraph.AnimGraphNode_StateMachine_3.LowFidelityStateMachine.AnimStateTransitionNode_14.Transition
  nodes [Result(bCanEnterTransition=Get bIsReactionHappyPlaying() -> (bIsReactionHappyPlaying:bool))]
 }
 Transition {
  path=AnimGraph.AnimGraphNode_StateMachine_3.LowFidelityStateMachine.AnimStateTransitionNode_15.Transition
  nodes [Result(bCanEnterTransition=Get bIsMoving())]
 }
 Transition {
  path=AnimGraph.AnimGraphNode_StateMachine_3.LowFidelityStateMachine.AnimStateTransitionNode_16.Transition
  nodes [Result(bCanEnterTransition=Get bIsMoving())]
 }
 Transition {
  path=AnimGraph.AnimGraphNode_StateMachine_3.LowFidelityStateMachine.AnimStateTransitionNode_2.Transition
  nodes [Result(bCanEnterTransition=true)]
 }
 Transition {
  path=AnimGraph.AnimGraphNode_StateMachine_3.LowFidelityStateMachine.AnimStateTransitionNode_4.Transition
  nodes [Result(bCanEnterTransition=NOT Boolean(A=Get bIsMoving()))]
 }
 Transition {
  path=AnimGraph.AnimGraphNode_StateMachine_3.LowFidelityStateMachine.AnimStateTransitionNode_7.Transition
  nodes [Result(bCanEnterTransition=Get bIsSpecialIdleEnter() -> (bIsSpecialIdleEnter:bool))]
 }
 Transition {
  path=AnimGraph.AnimGraphNode_StateMachine_3.LowFidelityStateMachine.AnimStateTransitionNode_8.Transition
  nodes [Result(bCanEnterTransition=Get bIsMoving())]
 }
 LowFidelity_Tail {
  entry=TailWag
  path=AnimGraph.AnimGraphNode_StateMachine_4.LowFidelity_Tail
  ids=id
  nodes [AnimStateNode_0 = TailWag() {is=title AnimStateNode}]
 }
 TailWag {
  path=AnimGraph.AnimGraphNode_StateMachine_4.LowFidelity_Tail.AnimStateNode_0.TailWag
  nodes [
   n1 = "Sequence Player (additive) 'SitPlant_COR_Tail_Bold_ADD'"() {groupName=TailWag, method=SyncGroup, sequence=/CosmeticCompanions/Assets/Quadruped/SitPlant/Animation/Core/SitPlant_COR_Tail_Bold_ADD}
   n2 = Layered blend per bone(
    BasePose="Sequence Player (additive) 'SitPlant_COR_Base_Pose_ADD'"() {bLoopAnimation=false, sequence=/CosmeticCompanions/Assets/Quadruped/SitPlant/Animation/Core/SitPlant_COR_Base_Pose_ADD}
    BlendPoses_0=@n1
   ) {blendMode=BlendMask, blendMasks [&SitPlant_Quadruped_Skeleton:TailMask], layerSetup []}
   n3 = "Sequence Player (additive) 'SitPlant_COR_Tail_Bold_ADD'"() {groupName=TailWag, method=SyncGroup, sequence=&SitPlant_COR_Tail_Bold_ADD}
   Output Animation Pose(Result=Blend Poses by bool(BlendPose_0=@n2, BlendPose_1=@n3, BlendTime_0=0.3, BlendTime_1=0.3, bActiveValue=Get bIsSitting()) {blendType=Linear, transitionType=Inertialization})
  ]
 }
 Locomotion {
  entry=Conduit
  path=AnimGraph.AnimGraphNode_StateMachine_5.Locomotion
  ids=id
  nodes [
   IsTurning() {is=title AnimStateAliasNode, also {aliases [Idle, Locomotion, LocomotionStop]}}
   AnimStateConduitNode_0 = Conduit()
   AnimStateNode_0 = Idle() {is=title AnimStateNode, bAlwaysResetOnEntry=true}
   AnimStateNode_1 = EmoteLoops() {is=title AnimStateNode}
   AnimStateNode_2 = ReactToIdle()
   AnimStateNode_3 = Locomotion() {
    bAlwaysResetOnEntry=true
    stateEntered {guid=#10}
    stateFullyBlended {guid=#11}
    stateLeft {guid=#12}
   }
   AnimStateNode_4 = LocomotionStop() {
    is=title AnimStateNode
    bAlwaysResetOnEntry=true
    stateEntered {guid=#13}
    stateFullyBlended {guid=#14}
    stateLeft {guid=#15}
   }
   AnimStateNode_5 = Start() {is=title AnimStateNode}
   AnimStateTransitionNode_0 = AnimStateTransitionNode() {blendMode=Cubic, crossfadeDuration=0.3, logicType=TLT_Inertialization, priorityOrder=100}
   AnimStateTransitionNode_1 = AnimStateTransitionNode() {blendMode=Linear, crossfadeDuration=0.15, logicType=TLT_Inertialization}
   AnimStateTransitionNode_10 = AnimStateTransitionNode() {
    logicType=TLT_Inertialization
    priorityOrder=100
    blendProfileWrapper {blendProfile=&SitPlant_Quadruped_Skeleton:TimeBlend_FastFeetSlowBody}
    transitionEnd {guid=#16}
    transitionInterrupt {guid=#17}
    transitionStart {guid=#18}
   }
   AnimStateTransitionNode_11 = AnimStateTransitionNode() {automaticRuleTriggerTime=0, bAutomaticRuleBasedOnSequencePlayerInState=true, blendMode=Linear, crossfadeDuration=0.1}
   AnimStateTransitionNode_12 = AnimStateTransitionNode() {logicType=TLT_Inertialization}
   AnimStateTransitionNode_13 = AnimStateTransitionNode() {bAutomaticRuleBasedOnSequencePlayerInState=true}
   AnimStateTransitionNode_14 = AnimStateTransitionNode() {logicType=TLT_Inertialization, priorityOrder=2}
   AnimStateTransitionNode_15 = AnimStateTransitionNode()
   AnimStateTransitionNode_16 = AnimStateTransitionNode()
   AnimStateTransitionNode_2 = AnimStateTransitionNode()
   AnimStateTransitionNode_3 = AnimStateTransitionNode() {blendMode=Linear, crossfadeDuration=0.1, logicType=TLT_Inertialization, priorityOrder=10}
   AnimStateTransitionNode_4 = AnimStateTransitionNode()
   AnimStateTransitionNode_5 = AnimStateTransitionNode() {bAutomaticRuleBasedOnSequencePlayerInState=true}
   AnimStateTransitionNode_6 = AnimStateTransitionNode()
   AnimStateTransitionNode_7 = AnimStateTransitionNode() {logicType=TLT_Inertialization}
   AnimStateTransitionNode_8 = AnimStateTransitionNode() {
    blendMode=Linear
    crossfadeDuration=0.1
    priorityOrder=3
    blendProfileWrapper {blendProfile=&SitPlant_Quadruped_Skeleton:TimeBlend_FastFeetSlowBody}
   }
  ]
  links [
   {
    dst=EmoteLoops
    rule=AnimGraph.AnimGraphNode_StateMachine_5.Locomotion.AnimStateTransitionNode_7.Transition
    src=Conduit
    via=AnimStateTransitionNode_7
   }
   {
    dst=Idle
    rule=AnimGraph.AnimGraphNode_StateMachine_5.Locomotion.AnimStateTransitionNode_0.Transition
    src=Conduit
    via=AnimStateTransitionNode_0
   }
   {
    dst=Locomotion
    rule=AnimGraph.AnimGraphNode_StateMachine_5.Locomotion.AnimStateTransitionNode_8.Transition
    src=Conduit
    via=AnimStateTransitionNode_8
   }
   {
    dst=ReactToIdle
    rule=AnimGraph.AnimGraphNode_StateMachine_5.Locomotion.AnimStateTransitionNode_14.Transition
    src=Conduit
    via=AnimStateTransitionNode_14
   }
   {
    dst=Conduit
    rule=AnimGraph.AnimGraphNode_StateMachine_5.Locomotion.AnimStateTransitionNode_15.Transition
    src=EmoteLoops
    via=AnimStateTransitionNode_15
   }
   {
    dst=ReactToIdle
    rule=AnimGraph.AnimGraphNode_StateMachine_5.Locomotion.AnimStateTransitionNode_12.Transition
    src=EmoteLoops
    via=AnimStateTransitionNode_12
   }
   {
    dst=Conduit
    rule=AnimGraph.AnimGraphNode_StateMachine_5.Locomotion.AnimStateTransitionNode_2.Transition
    src=Idle
    via=AnimStateTransitionNode_2
   }
   {
    dst=Start
    rule=AnimGraph.AnimGraphNode_StateMachine_5.Locomotion.AnimStateTransitionNode_3.Transition
    src=IsTurning
    via=AnimStateTransitionNode_3
   }
   {
    dst=Conduit
    rule=AnimGraph.AnimGraphNode_StateMachine_5.Locomotion.AnimStateTransitionNode_4.Transition
    src=Locomotion
    via=AnimStateTransitionNode_4
   }
   {
    dst=LocomotionStop
    rule=AnimGraph.AnimGraphNode_StateMachine_5.Locomotion.AnimStateTransitionNode_1.Transition
    src=Locomotion
    via=AnimStateTransitionNode_1
   }
   {
    dst=Conduit
    rule=AnimGraph.AnimGraphNode_StateMachine_5.Locomotion.AnimStateTransitionNode_5.Transition
    src=LocomotionStop
    via=AnimStateTransitionNode_5
   }
   {
    dst=Conduit
    rule=AnimGraph.AnimGraphNode_StateMachine_5.Locomotion.AnimStateTransitionNode_6.Transition
    src=LocomotionStop
    via=AnimStateTransitionNode_6
   }
   {
    dst=Locomotion
    rule=AnimGraph.AnimGraphNode_StateMachine_5.Locomotion.AnimStateTransitionNode_10.Transition
    src=LocomotionStop
    via=AnimStateTransitionNode_10
   }
   {
    dst=Conduit
    rule=AnimGraph.AnimGraphNode_StateMachine_5.Locomotion.AnimStateTransitionNode_13.Transition
    src=ReactToIdle
    via=AnimStateTransitionNode_13
   }
   {
    dst=Conduit
    rule=AnimGraph.AnimGraphNode_StateMachine_5.Locomotion.AnimStateTransitionNode_16.Transition
    src=ReactToIdle
    via=AnimStateTransitionNode_16
   }
   {
    dst=Locomotion
    rule=AnimGraph.AnimGraphNode_StateMachine_5.Locomotion.AnimStateTransitionNode_11.Transition
    src=Start
    via=AnimStateTransitionNode_11
   }
  ]
 }
 Conduit {path=AnimGraph.AnimGraphNode_StateMachine_5.Locomotion.AnimStateConduitNode_0.Conduit, nodes [Result(bCanEnterTransition=true)]}
 Idle {
  path=AnimGraph.AnimGraphNode_StateMachine_5.Locomotion.AnimStateNode_0.Idle
  nodes [
   AnimGraphNode_StateMachine_1 = Idle() {is=title AnimGraphNode_StateMachine, bAllowConduitEntryStates=true}
   Output Animation Pose(Result=Modify Curve(SourcePose=@AnimGraphNode_StateMachine_1, CurveValues_0=1, Alpha=1) {curveNames [idle], curveValues [0]}) {
    node {stateExitFunction {functionName=OnExit_SitState}}
    stateExitFunction {bSelfContext=true, memberGuid=#5, memberName=OnExit_SitState}
   }
  ]
 }
 Idle {
  entry=Conduit
  path=AnimGraph.AnimGraphNode_StateMachine_5.Locomotion.AnimStateNode_0.Idle.AnimGraphNode_StateMachine_1.Idle
  ids=id
  nodes [
   Alias() {also {aliases [SitPose,StandToSit]}}
   AnimStateConduitNode_0 = Conduit()
   AnimStateNode_0 = StandToSit()
   AnimStateNode_1 = SitPose()
   AnimStateNode_2 = StandingPose()
   AnimStateNode_3 = SitToStand()
   AnimStateTransitionNode_0 = AnimStateTransitionNode() {automaticRuleTriggerTime=0, bAutomaticRuleBasedOnSequencePlayerInState=true, logicType=TLT_Inertialization}
   AnimStateTransitionNode_1 = AnimStateTransitionNode() {logicType=TLT_Inertialization, priorityOrder=2}
   AnimStateTransitionNode_2 = AnimStateTransitionNode() {logicType=TLT_Inertialization, priorityOrder=100}
   AnimStateTransitionNode_3 = AnimStateTransitionNode() {logicType=TLT_Inertialization}
   AnimStateTransitionNode_4 = AnimStateTransitionNode()
   AnimStateTransitionNode_6 = AnimStateTransitionNode() {bAutomaticRuleBasedOnSequencePlayerInState=true, crossfadeDuration=0.3, logicType=TLT_Inertialization}
   AnimStateTransitionNode_8 = AnimStateTransitionNode() {logicType=TLT_Inertialization}
  ]
  links [
   {
    dst=SitToStand
    rule=AnimGraph.AnimGraphNode_StateMachine_5.Locomotion.AnimStateNode_0.Idle.AnimGraphNode_StateMachine_1.Idle.AnimStateTransitionNode_8.AnimationTransitionGraph_1
    src=Alias
    via=AnimStateTransitionNode_8
   }
   {
    dst=SitPose
    rule=AnimGraph.AnimGraphNode_StateMachine_5.Locomotion.AnimStateNode_0.Idle.AnimGraphNode_StateMachine_1.Idle.AnimStateTransitionNode_3.AnimationTransitionGraph_1
    src=Conduit
    via=AnimStateTransitionNode_3
   }
   {
    dst=StandToSit
    rule=AnimGraph.AnimGraphNode_StateMachine_5.Locomotion.AnimStateNode_0.Idle.AnimGraphNode_StateMachine_1.Idle.AnimStateTransitionNode_1.AnimationTransitionGraph_1
    src=Conduit
    via=AnimStateTransitionNode_1
   }
   {
    dst=StandingPose
    rule=AnimGraph.AnimGraphNode_StateMachine_5.Locomotion.AnimStateNode_0.Idle.AnimGraphNode_StateMachine_1.Idle.AnimStateTransitionNode_2.AnimationTransitionGraph_1
    src=Conduit
    via=AnimStateTransitionNode_2
   }
   {
    dst=StandingPose
    rule=AnimGraph.AnimGraphNode_StateMachine_5.Locomotion.AnimStateNode_0.Idle.AnimGraphNode_StateMachine_1.Idle.AnimStateTransitionNode_6.AnimationTransitionGraph_1
    src=SitToStand
    via=AnimStateTransitionNode_6
   }
   {
    dst=SitPose
    rule=AnimGraph.AnimGraphNode_StateMachine_5.Locomotion.AnimStateNode_0.Idle.AnimGraphNode_StateMachine_1.Idle.AnimStateTransitionNode_0.AnimationTransitionGraph_1
    src=StandToSit
    via=AnimStateTransitionNode_0
   }
   {
    dst=Conduit
    rule=AnimGraph.AnimGraphNode_StateMachine_5.Locomotion.AnimStateNode_0.Idle.AnimGraphNode_StateMachine_1.Idle.AnimStateTransitionNode_4.AnimationTransitionGraph_1
    src=StandingPose
    via=AnimStateTransitionNode_4
   }
  ]
 }
 Conduit {
  path=AnimGraph.AnimGraphNode_StateMachine_5.Locomotion.AnimStateNode_0.Idle.AnimGraphNode_StateMachine_1.Idle.AnimStateConduitNode_0.Conduit
  nodes [Result(bCanEnterTransition=true)]
 }
 StandToSit {
  path=AnimGraph.AnimGraphNode_StateMachine_5.Locomotion.AnimStateNode_0.Idle.AnimGraphNode_StateMachine_1.Idle.AnimStateNode_0.StandToSit
  nodes [Output Animation Pose(Result=Sequence Player 'SitPlant_COR_Sit_In'() {bLoopAnimation=false, sequence=&SitPlant_COR_Sit_In}) {
    node {stateEntryFunction {functionName=OnEnterSitState}}
    stateEntryFunction {bSelfContext=true, memberGuid=#6, memberName=IsSit_Enter}
   }]
 }
 SitPose {
  path=AnimGraph.AnimGraphNode_StateMachine_5.Locomotion.AnimStateNode_0.Idle.AnimGraphNode_StateMachine_1.Idle.AnimStateNode_1.SitPose
  nodes [Output Animation Pose(Result=Sequence Player 'SitPlant_COR_Sit_Pose'() {bLoopAnimation=false, sequence=&SitPlant_COR_Sit_Pose}) {
    node {stateEntryFunction {functionName=OnStateEntry}, stateExitFunction {functionName=CanPlayAntic}}
    stateEntryFunction {bSelfContext=true, memberGuid=#7, memberName=CanPlayAntic_Enter}
    stateExitFunction {bSelfContext=true, memberGuid=#7, memberName=CanPlayAntic_Enter}
   }]
 }
 StandingPose {
  path=AnimGraph.AnimGraphNode_StateMachine_5.Locomotion.AnimStateNode_0.Idle.AnimGraphNode_StateMachine_1.Idle.AnimStateNode_2.StandingPose
  nodes [
   AnimGraphNode_StateMachine_0 = StandingStateMachine() {is=title AnimGraphNode_StateMachine, bAllowConduitEntryStates=true}
   Output Animation Pose(Result=@AnimGraphNode_StateMachine_0) {
    node {stateEntryFunction {functionName=CanPlayAntic_Enter}, stateExitFunction {functionName=CanPlayAntic_Exit}}
    stateEntryFunction {bSelfContext=true, memberGuid=#7, memberName=CanPlayAntic_Enter}
    stateExitFunction {bSelfContext=true, memberGuid=#19, memberName=CanPlayAntic_Exit}
   }
  ]
 }
 StandingStateMachine {
  entry=Conduit
  path=AnimGraph.AnimGraphNode_StateMachine_5.Locomotion.AnimStateNode_0.Idle.AnimGraphNode_StateMachine_1.Idle.AnimStateNode_2.StandingPose.AnimGraphNode_StateMachine_0.StandingStateMachine
  ids=id
  nodes [
   AnimStateConduitNode_0 = Conduit()
   AnimStateNode_0 = Idle() {is=title AnimStateNode}
   AnimStateNode_1 = AngryIdleIntro() {is=title AnimStateNode}
   AnimStateNode_2 = AngryIdle() {is=title AnimStateNode}
   AnimStateNode_3 = AngryIdleExit() {is=title AnimStateNode}
   AnimStateTransitionNode_0 = AnimStateTransitionNode() {logicType=TLT_Inertialization, priorityOrder=100}
   AnimStateTransitionNode_1 = AnimStateTransitionNode() {bAllowInertializationForSelfTransitions=true}
   AnimStateTransitionNode_2 = AnimStateTransitionNode() {bAllowInertializationForSelfTransitions=true, crossfadeDuration=0.1, logicType=TLT_Inertialization}
   AnimStateTransitionNode_3 = AnimStateTransitionNode() {
    automaticRuleTriggerTime=0
    bAllowInertializationForSelfTransitions=true
    bAutomaticRuleBasedOnSequencePlayerInState=true
    logicType=TLT_Inertialization
   }
   AnimStateTransitionNode_4 = AnimStateTransitionNode() {bAllowInertializationForSelfTransitions=true, crossfadeDuration=0.1, logicType=TLT_Inertialization}
   AnimStateTransitionNode_5 = AnimStateTransitionNode() {automaticRuleTriggerTime=0, bAllowInertializationForSelfTransitions=true, bAutomaticRuleBasedOnSequencePlayerInState=true}
  ]
  links [
   {
    dst=AngryIdleExit
    rule=AnimGraph.AnimGraphNode_StateMachine_5.Locomotion.AnimStateNode_0.Idle.AnimGraphNode_StateMachine_1.Idle.AnimStateNode_2.StandingPose.AnimGraphNode_StateMachine_0.StandingStateMachine.AnimStateTransitionNode_4.Transition
    src=AngryIdle
    via=AnimStateTransitionNode_4
   }
   {
    dst=Conduit
    rule=AnimGraph.AnimGraphNode_StateMachine_5.Locomotion.AnimStateNode_0.Idle.AnimGraphNode_StateMachine_1.Idle.AnimStateNode_2.StandingPose.AnimGraphNode_StateMachine_0.StandingStateMachine.AnimStateTransitionNode_5.Transition
    src=AngryIdleExit
    via=AnimStateTransitionNode_5
   }
   {
    dst=AngryIdle
    rule=AnimGraph.AnimGraphNode_StateMachine_5.Locomotion.AnimStateNode_0.Idle.AnimGraphNode_StateMachine_1.Idle.AnimStateNode_2.StandingPose.AnimGraphNode_StateMachine_0.StandingStateMachine.AnimStateTransitionNode_3.Transition
    src=AngryIdleIntro
    via=AnimStateTransitionNode_3
   }
   {
    dst=AngryIdleIntro
    rule=AnimGraph.AnimGraphNode_StateMachine_5.Locomotion.AnimStateNode_0.Idle.AnimGraphNode_StateMachine_1.Idle.AnimStateNode_2.StandingPose.AnimGraphNode_StateMachine_0.StandingStateMachine.AnimStateTransitionNode_2.Transition
    src=Conduit
    via=AnimStateTransitionNode_2
   }
   {
    dst=Idle
    rule=AnimGraph.AnimGraphNode_StateMachine_5.Locomotion.AnimStateNode_0.Idle.AnimGraphNode_StateMachine_1.Idle.AnimStateNode_2.StandingPose.AnimGraphNode_StateMachine_0.StandingStateMachine.AnimStateTransitionNode_0.Transition
    src=Conduit
    via=AnimStateTransitionNode_0
   }
   {
    dst=Conduit
    rule=AnimGraph.AnimGraphNode_StateMachine_5.Locomotion.AnimStateNode_0.Idle.AnimGraphNode_StateMachine_1.Idle.AnimStateNode_2.StandingPose.AnimGraphNode_StateMachine_0.StandingStateMachine.AnimStateTransitionNode_1.Transition
    src=Idle
    via=AnimStateTransitionNode_1
   }
  ]
 }
 Conduit {
  path=AnimGraph.AnimGraphNode_StateMachine_5.Locomotion.AnimStateNode_0.Idle.AnimGraphNode_StateMachine_1.Idle.AnimStateNode_2.StandingPose.AnimGraphNode_StateMachine_0.StandingStateMachine.AnimStateConduitNode_0.Conduit
  nodes [Result(bCanEnterTransition=true)]
 }
 Idle {
  path=AnimGraph.AnimGraphNode_StateMachine_5.Locomotion.AnimStateNode_0.Idle.AnimGraphNode_StateMachine_1.Idle.AnimStateNode_2.StandingPose.AnimGraphNode_StateMachine_0.StandingStateMachine.AnimStateNode_0.Idle
  nodes [Output Animation Pose(Result=Sequence Player 'SitPlant_COR_Stand_Pose'() {bLoopAnimation=false, sequence=&SitPlant_COR_Stand_Pose})]
 }
 AngryIdleIntro {
  path=AnimGraph.AnimGraphNode_StateMachine_5.Locomotion.AnimStateNode_0.Idle.AnimGraphNode_StateMachine_1.Idle.AnimStateNode_2.StandingPose.AnimGraphNode_StateMachine_0.StandingStateMachine.AnimStateNode_1.AngryIdleIntro
  nodes [Output Animation Pose(Result=Sequence Player 'SitPlant_xDEV_Start_Test'() {bLoopAnimation=false, sequence=/CosmeticCompanions/Assets/Quadruped/SitPlant/Animation/Developer/SitPlant_xDEV_Start_Test})]
 }
 AngryIdle {
  path=AnimGraph.AnimGraphNode_StateMachine_5.Locomotion.AnimStateNode_0.Idle.AnimGraphNode_StateMachine_1.Idle.AnimStateNode_2.StandingPose.AnimGraphNode_StateMachine_0.StandingStateMachine.AnimStateNode_2.AngryIdle
  nodes [Output Animation Pose(Result=Sequence Player 'SitPlant_xDEV_React_Angry_02_Loop_OLD'() {sequence=&SitPlant_xDEV_React_Angry_02_Loop_OLD})]
 }
 AngryIdleExit {
  path=AnimGraph.AnimGraphNode_StateMachine_5.Locomotion.AnimStateNode_0.Idle.AnimGraphNode_StateMachine_1.Idle.AnimStateNode_2.StandingPose.AnimGraphNode_StateMachine_0.StandingStateMachine.AnimStateNode_3.AngryIdleExit
  nodes [Output Animation Pose(Result=Sequence Player 'SitPlant_COR_Angry_Idle_Out'() {bLoopAnimation=false, sequence=/CosmeticCompanions/Assets/Quadruped/SitPlant/Animation/Core/SitPlant_COR_Angry_Idle_Out})]
 }
 Transition {
  path=AnimGraph.AnimGraphNode_StateMachine_5.Locomotion.AnimStateNode_0.Idle.AnimGraphNode_StateMachine_1.Idle.AnimStateNode_2.StandingPose.AnimGraphNode_StateMachine_0.StandingStateMachine.AnimStateTransitionNode_0.Transition
  nodes [Result(bCanEnterTransition=true)]
 }
 Transition {
  path=AnimGraph.AnimGraphNode_StateMachine_5.Locomotion.AnimStateNode_0.Idle.AnimGraphNode_StateMachine_1.Idle.AnimStateNode_2.StandingPose.AnimGraphNode_StateMachine_0.StandingStateMachine.AnimStateTransitionNode_1.Transition
  nodes [Result(bCanEnterTransition=true)]
 }
 Transition {
  path=AnimGraph.AnimGraphNode_StateMachine_5.Locomotion.AnimStateNode_0.Idle.AnimGraphNode_StateMachine_1.Idle.AnimStateNode_2.StandingPose.AnimGraphNode_StateMachine_0.StandingStateMachine.AnimStateTransitionNode_2.Transition
  nodes [Result(bCanEnterTransition=Get bIsDelayedMoodCombat())]
 }
 Transition {
  path=AnimGraph.AnimGraphNode_StateMachine_5.Locomotion.AnimStateNode_0.Idle.AnimGraphNode_StateMachine_1.Idle.AnimStateNode_2.StandingPose.AnimGraphNode_StateMachine_0.StandingStateMachine.AnimStateTransitionNode_3.Transition
  nodes [Result()]
 }
 Transition {
  path=AnimGraph.AnimGraphNode_StateMachine_5.Locomotion.AnimStateNode_0.Idle.AnimGraphNode_StateMachine_1.Idle.AnimStateNode_2.StandingPose.AnimGraphNode_StateMachine_0.StandingStateMachine.AnimStateTransitionNode_4.Transition
  nodes [Result(bCanEnterTransition=NOT Boolean(A=Get bIsMoodCombat()))]
 }
 Transition {
  path=AnimGraph.AnimGraphNode_StateMachine_5.Locomotion.AnimStateNode_0.Idle.AnimGraphNode_StateMachine_1.Idle.AnimStateNode_2.StandingPose.AnimGraphNode_StateMachine_0.StandingStateMachine.AnimStateTransitionNode_5.Transition
  nodes [Result()]
 }
 SitToStand {
  path=AnimGraph.AnimGraphNode_StateMachine_5.Locomotion.AnimStateNode_0.Idle.AnimGraphNode_StateMachine_1.Idle.AnimStateNode_3.SitToStand
  nodes [Output Animation Pose(Result=Sequence Player 'SitPlant_COR_Sit_Out'() {bLoopAnimation=false, sequence=&SitPlant_COR_Sit_Out}) {
    node {stateEntryFunction {functionName=OnExitSitState}}
    stateEntryFunction {bSelfContext=true, memberGuid=#5, memberName=OnExit_SitState}
   }]
 }
 AnimationTransitionGraph_1 {
  path=AnimGraph.AnimGraphNode_StateMachine_5.Locomotion.AnimStateNode_0.Idle.AnimGraphNode_StateMachine_1.Idle.AnimStateTransitionNode_0.AnimationTransitionGraph_1
  nodes [Result()]
 }
 AnimationTransitionGraph_1 {
  path=AnimGraph.AnimGraphNode_StateMachine_5.Locomotion.AnimStateNode_0.Idle.AnimGraphNode_StateMachine_1.Idle.AnimStateTransitionNode_1.AnimationTransitionGraph_1
  nodes [Result(bCanEnterTransition=Get bIsStandToSit())]
 }
 AnimationTransitionGraph_1 {
  path=AnimGraph.AnimGraphNode_StateMachine_5.Locomotion.AnimStateNode_0.Idle.AnimGraphNode_StateMachine_1.Idle.AnimStateTransitionNode_2.AnimationTransitionGraph_1
  nodes [Result(bCanEnterTransition=true)]
 }
 AnimationTransitionGraph_1 {
  path=AnimGraph.AnimGraphNode_StateMachine_5.Locomotion.AnimStateNode_0.Idle.AnimGraphNode_StateMachine_1.Idle.AnimStateTransitionNode_3.AnimationTransitionGraph_1
  nodes [Result(bCanEnterTransition=Get bIsSitting())]
 }
 AnimationTransitionGraph_1 {
  path=AnimGraph.AnimGraphNode_StateMachine_5.Locomotion.AnimStateNode_0.Idle.AnimGraphNode_StateMachine_1.Idle.AnimStateTransitionNode_4.AnimationTransitionGraph_1
  nodes [Result(bCanEnterTransition=true)]
 }
 AnimationTransitionGraph_1 {
  path=AnimGraph.AnimGraphNode_StateMachine_5.Locomotion.AnimStateNode_0.Idle.AnimGraphNode_StateMachine_1.Idle.AnimStateTransitionNode_6.AnimationTransitionGraph_1
  nodes [Result()]
 }
 AnimationTransitionGraph_1 {
  path=AnimGraph.AnimGraphNode_StateMachine_5.Locomotion.AnimStateNode_0.Idle.AnimGraphNode_StateMachine_1.Idle.AnimStateTransitionNode_8.AnimationTransitionGraph_1
  nodes [Result(bCanEnterTransition=Get bIsSitToStand())]
 }
 EmoteLoops {
  path=AnimGraph.AnimGraphNode_StateMachine_5.Locomotion.AnimStateNode_1.EmoteLoops
  nodes [
   AnimGraphNode_ChooserPlayer_1 = Chooser Player: PRA_Companion_SpecialIdle() {
    chooser {
     _structType=&LookupProxy
     proxy=&PRA_Companion_SpecialIdle
     proxyTable {
      _structType=&ProxyTableContextProperty
      binding {contextIndex=0, displayName=Chooser Tables, isBoundToRoot=false, propertyBindingChain [ChooserTables]}
     }
    }
    chooserContextDefinition [
     {
      _structType=&ContextObjectTypeClass
      class=&SitPlant_AnimBP_C
      direction=ReadWrite
     }
     {struct=&ChooserPlayerSettings, _structType=&ContextObjectTypeStruct, direction=Read}
    ]
    defaultSettings {blendOption=Linear}
   }
   Output Animation Pose(Result=@AnimGraphNode_ChooserPlayer_1)
  ]
 }
 AnimationBlendStackGraph_0 {
  path=AnimGraph.AnimGraphNode_StateMachine_5.Locomotion.AnimStateNode_1.EmoteLoops.AnimGraphNode_ChooserPlayer_1.AnimationBlendStackGraph_0
  nodes [Output Pose(Result=Blend Stack Input())]
 }
 ReactToIdle {
  path=AnimGraph.AnimGraphNode_StateMachine_5.Locomotion.AnimStateNode_2.ReactToIdle
  nodes [
   AnimGraphNode_ChooserPlayer_1 = Chooser Player: PRA_Companion_ReactToIdle() {
    chooser {
     _structType=&LookupProxy
     proxy=&PRA_Companion_ReactToIdle
     proxyTable {
      _structType=&ProxyTableContextProperty
      binding {contextIndex=0, displayName=Chooser Tables, isBoundToRoot=false, propertyBindingChain [ChooserTables]}
     }
    }
    chooserContextDefinition [
     {
      _structType=&ContextObjectTypeClass
      class=&SitPlant_AnimBP_C
      direction=ReadWrite
     }
     {struct=&ChooserPlayerSettings, _structType=&ContextObjectTypeStruct, direction=Read}
    ]
    defaultSettings {blendOption=Linear}
   }
   Output Animation Pose(Result=Modify Curve(SourcePose=@AnimGraphNode_ChooserPlayer_1, CurveValues_0=1, Alpha=1) {curveNames [idle], curveValues [0]}) {
    node {stateEntryFunction {functionName=SpecialIdle_StateEntry}, stateExitFunction {functionName=SpecialIdle_StateExit}}
    stateEntryFunction {bSelfContext=true, memberGuid=#8, memberName=SpecialIdle_StateEntry}
    stateExitFunction {bSelfContext=true, memberGuid=#9, memberName=SpecialIdle_StateExit}
   }
  ]
 }
 AnimationBlendStackGraph_0 {
  path=AnimGraph.AnimGraphNode_StateMachine_5.Locomotion.AnimStateNode_2.ReactToIdle.AnimGraphNode_ChooserPlayer_1.AnimationBlendStackGraph_0
  nodes [Output Pose(Result=Blend Stack Input())]
 }
 Locomotion {
  path=AnimGraph.AnimGraphNode_StateMachine_5.Locomotion.AnimStateNode_3.Locomotion
  nodes [
   n1 = Sequence Player 'SitPlant_COR_Loco_Walk'(PlayRate=Get WalkPlayRate()) {groupName=Locomotion, method=SyncGroup, sequence=&SitPlant_COR_Loco_Walk}
   n2 = Sequence Player 'SitPlant_COR_Loco_Run'(PlayRate=Get RunPlayRate()) {groupName=Locomotion, method=SyncGroup, sequence=&SitPlant_COR_Loco_Run}
   n3 = Sequence Player 'SitPlant_COR_Loco_Sprint'(PlayRate=1) {groupName=Locomotion, method=SyncGroup, sequence=&SitPlant_COR_Loco_Sprint}
   n4 = Sequence Player 'SitPlant_COR_Loco_Trot'(PlayRate=Get TrotPlayRate()) {groupName=Locomotion, method=SyncGroup, sequence=&SitPlant_COR_Loco_Trot}
   n5 = "Blend Poses (ELocomotionMovementType)"(
    BlendPose_1=@n3
    BlendPose_2=@n2
    BlendPose_3=@n4
    BlendPose_4=Modify Curve(SourcePose=@n1, CurveValues_0=1, Alpha=1) {curveNames [Walk], curveValues [0]}
    BlendTime_0=0.2
    BlendTime_1=0.2
    BlendTime_2=0.2
    BlendTime_3=0.2
    BlendTime_4=0.2
    ActiveEnumValue=Get CurrentLocomotionMovementType() -> (CurrentLocomotionMovementType=Walk)
   ) {
    blendPose [
     {linkId=-1, sourceLinkId=-1}
     {linkId=-1, sourceLinkId=-1}
     {linkId=-1, sourceLinkId=-1}
     {linkId=-1, sourceLinkId=-1}
     {linkId=-1, sourceLinkId=-1}
    ]
    blendTime [0.1,0.1,0.1,0.1,0.1]
   }
   Output Animation Pose(Result=Modify Curve(SourcePose=@n5, CurveValues_0=1, Alpha=1) {curveNames [Locomotion], curveValues [0]})
  ]
 }
 LocomotionStop {
  path=AnimGraph.AnimGraphNode_StateMachine_5.Locomotion.AnimStateNode_4.LocomotionStop
  nodes [
   AnimGraphNode_StateMachine_0 = LocomotionStopStateMachine() {is=title AnimGraphNode_StateMachine, bAllowConduitEntryStates=true}
   Output Animation Pose(Result=Modify Curve(SourcePose=@AnimGraphNode_StateMachine_0, CurveValues_0=1, Alpha=1) {curveNames [idle], curveValues [0]})
  ]
 }
 LocomotionStopStateMachine {
  entry=Conduit
  path=AnimGraph.AnimGraphNode_StateMachine_5.Locomotion.AnimStateNode_4.LocomotionStop.AnimGraphNode_StateMachine_0.LocomotionStopStateMachine
  ids=id
  nodes [
   AnimStateConduitNode_0 = Conduit()
   AnimStateNode_1 = SlowStop() {is=title AnimStateNode}
   AnimStateNode_4 = FastStop() {is=title AnimStateNode}
   AnimStateTransitionNode_1 = AnimStateTransitionNode() {
    crossfadeDuration=0.4
    logicType=TLT_Inertialization
    blendProfileWrapper {blendProfile=&SitPlant_Quadruped_Skeleton:TimeBlend_FastFeetSlowBody}
   }
   AnimStateTransitionNode_4 = AnimStateTransitionNode() {priorityOrder=100}
  ]
  links [
   {
    dst=FastStop
    rule=AnimGraph.AnimGraphNode_StateMachine_5.Locomotion.AnimStateNode_4.LocomotionStop.AnimGraphNode_StateMachine_0.LocomotionStopStateMachine.AnimStateTransitionNode_4.Transition
    src=Conduit
    via=AnimStateTransitionNode_4
   }
   {
    dst=SlowStop
    rule=AnimGraph.AnimGraphNode_StateMachine_5.Locomotion.AnimStateNode_4.LocomotionStop.AnimGraphNode_StateMachine_0.LocomotionStopStateMachine.AnimStateTransitionNode_1.Transition
    src=Conduit
    via=AnimStateTransitionNode_1
   }
  ]
 }
 Conduit {
  path=AnimGraph.AnimGraphNode_StateMachine_5.Locomotion.AnimStateNode_4.LocomotionStop.AnimGraphNode_StateMachine_0.LocomotionStopStateMachine.AnimStateConduitNode_0.Conduit
  nodes [Result(bCanEnterTransition=true)]
 }
 SlowStop {
  path=AnimGraph.AnimGraphNode_StateMachine_5.Locomotion.AnimStateNode_4.LocomotionStop.AnimGraphNode_StateMachine_0.LocomotionStopStateMachine.AnimStateNode_1.SlowStop
  nodes [Output Animation Pose(Result=Sequence Player 'SitPlant_COR_Loco_StopLight'() {bLoopAnimation=false, sequence=/CosmeticCompanions/Assets/Quadruped/SitPlant/Animation/Core/SitPlant_COR_Loco_StopLight})]
 }
 FastStop {
  path=AnimGraph.AnimGraphNode_StateMachine_5.Locomotion.AnimStateNode_4.LocomotionStop.AnimGraphNode_StateMachine_0.LocomotionStopStateMachine.AnimStateNode_4.FastStop
  nodes [Output Animation Pose(Result=Sequence Player 'SitPlant_COR_Loco_StopHeavy'() {bLoopAnimation=false, sequence=/CosmeticCompanions/Assets/Quadruped/SitPlant/Animation/Core/SitPlant_COR_Loco_StopHeavy})]
 }
 Transition {
  path=AnimGraph.AnimGraphNode_StateMachine_5.Locomotion.AnimStateNode_4.LocomotionStop.AnimGraphNode_StateMachine_0.LocomotionStopStateMachine.AnimStateTransitionNode_1.Transition
  nodes [Result(bCanEnterTransition="float > float"(A=GetCurveValue(CurveName=Walk)))]
 }
 Transition {
  path=AnimGraph.AnimGraphNode_StateMachine_5.Locomotion.AnimStateNode_4.LocomotionStop.AnimGraphNode_StateMachine_0.LocomotionStopStateMachine.AnimStateTransitionNode_4.Transition
  nodes [Result(bCanEnterTransition=true)]
 }
 Start {
  path=AnimGraph.AnimGraphNode_StateMachine_5.Locomotion.AnimStateNode_5.Start
  nodes [
   n1 = Sequence Player 'SitPlant_xDEV_Start_Test'() {
    bLoopAnimation=false
    groupName=Locomotion
    method=SyncGroup
    sequence=&SitPlant_xDEV_Start_Test
   }
   Output Animation Pose(Result=Modify Curve(SourcePose=@n1, CurveValues_0=1, Alpha=1) {curveNames [walk], curveValues [0]}) {
    node {stateEntryFunction {functionName=OnStateEntry}, stateExitFunction {functionName=OnStateExit}}
    stateEntryFunction {bSelfContext=true, memberGuid=#20, memberName=LocoStart_Entry}
    stateExitFunction {bSelfContext=true, memberGuid=#21, memberName=OnStateExit}
   }
  ]
 }
 Transition {path=AnimGraph.AnimGraphNode_StateMachine_5.Locomotion.AnimStateTransitionNode_0.Transition, nodes [Result(bCanEnterTransition=true)]}
 Transition {
  path=AnimGraph.AnimGraphNode_StateMachine_5.Locomotion.AnimStateTransitionNode_1.Transition
  nodes [
   n1 = Get bIsMoodCombat()
   n2 = Get bIsMoodWorried()
   n3 = NOT Boolean(A=Get bIsMoving())
   n4 = Get bIsCompanionStopping() -> (bIsCompanionStopping:bool)
   Result(bCanEnterTransition=AND Boolean!(A=NOT Boolean(A=@n1), B=NOT Boolean(A=@n2), C:bool=OR Boolean(A=@n4, B=@n3)) -> (ReturnValue))
  ]
 }
 Transition {
  path=AnimGraph.AnimGraphNode_StateMachine_5.Locomotion.AnimStateTransitionNode_10.Transition
  nodes [Result(bCanEnterTransition=Get bIsStopToLocomotion() -> (bIsStopToLocomotion:bool))]
 }
 Transition {path=AnimGraph.AnimGraphNode_StateMachine_5.Locomotion.AnimStateTransitionNode_11.Transition, nodes [Result()]}
 Transition {
  path=AnimGraph.AnimGraphNode_StateMachine_5.Locomotion.AnimStateTransitionNode_12.Transition
  nodes [Result(bCanEnterTransition=Get bIsSpecialIdleExit())]
 }
 Transition {path=AnimGraph.AnimGraphNode_StateMachine_5.Locomotion.AnimStateTransitionNode_13.Transition, nodes [Result()]}
 Transition {
  path=AnimGraph.AnimGraphNode_StateMachine_5.Locomotion.AnimStateTransitionNode_14.Transition
  nodes [Result(bCanEnterTransition=Get bIsReactionHappyPlaying())]
 }
 Transition {
  path=AnimGraph.AnimGraphNode_StateMachine_5.Locomotion.AnimStateTransitionNode_15.Transition
  nodes [Result(bCanEnterTransition=Get bIsMoving())]
 }
 Transition {
  path=AnimGraph.AnimGraphNode_StateMachine_5.Locomotion.AnimStateTransitionNode_16.Transition
  nodes [Result(bCanEnterTransition=Get bIsMoving())]
 }
 Transition {path=AnimGraph.AnimGraphNode_StateMachine_5.Locomotion.AnimStateTransitionNode_2.Transition, nodes [Result(bCanEnterTransition=true)]}
 Transition {
  path=AnimGraph.AnimGraphNode_StateMachine_5.Locomotion.AnimStateTransitionNode_3.Transition
  nodes [
   n1 = NOT Boolean(A=Get bIsMoodWorried())
   n2 = Get bIsAllowTurn() -> (bIsAllowTurn:bool)
   n3 = NOT Boolean(A=Get bIsMoving())
   n4 = Get UpdatedGoalRotation() -> (UpdatedGoalRotation:float)
   n5 = OR Boolean(A="float > float"(A=@n4, B=5), B="float < float"(A=@n4, B=-5))
   n6 = Get bIsMoodCombat()
   Result(bCanEnterTransition=OR Boolean(A=AND Boolean!(A=@n2, B=@n1) -> (ReturnValue), B=AND Boolean!(A=@n6, B=@n3, C:bool=@n5) -> (ReturnValue)))
  ]
 }
 Transition {
  path=AnimGraph.AnimGraphNode_StateMachine_5.Locomotion.AnimStateTransitionNode_4.Transition
  nodes [
   n1 = NOT Boolean(A=Get bIsMoving())
   n2 = Get bIsMoodWorried()
   Result(bCanEnterTransition=OR Boolean(A=Get bIsCombatStop() -> (bIsCombatStop:bool), B=AND Boolean!(A=@n2, B=@n1) -> (ReturnValue)))
  ]
 }
 Transition {path=AnimGraph.AnimGraphNode_StateMachine_5.Locomotion.AnimStateTransitionNode_5.Transition, nodes [Result()]}
 Transition {
  path=AnimGraph.AnimGraphNode_StateMachine_5.Locomotion.AnimStateTransitionNode_6.Transition
  nodes [Result(bCanEnterTransition=Get bIsEarlyExitLocoStop() -> (bIsEarlyExitLocoStop:bool))]
 }
 Transition {
  path=AnimGraph.AnimGraphNode_StateMachine_5.Locomotion.AnimStateTransitionNode_7.Transition
  nodes [Result(bCanEnterTransition=Get bIsSpecialIdleEnter())]
 }
 Transition {
  path=AnimGraph.AnimGraphNode_StateMachine_5.Locomotion.AnimStateTransitionNode_8.Transition
  nodes [Result(bCanEnterTransition=Get bIsMoving())]
 }
 BodyAdditives {
  entry=BodyAdditiveTempState
  path=AnimGraph.AnimGraphNode_StateMachine_6.BodyAdditives
  ids=id
  nodes [
   AnimStateNode_0 = BodyAdditiveTempState()
   AnimStateNode_1 = Worried()
   AnimStateNode_2 = Combat()
   AnimStateTransitionNode_0 = AnimStateTransitionNode() {crossfadeDuration=0.35, logicType=TLT_Inertialization}
   AnimStateTransitionNode_1 = AnimStateTransitionNode() {crossfadeDuration=0.35, logicType=TLT_Inertialization}
   AnimStateTransitionNode_2 = AnimStateTransitionNode() {logicType=TLT_Inertialization, priorityOrder=2}
   AnimStateTransitionNode_3 = AnimStateTransitionNode() {crossfadeDuration=0.35}
  ]
  links [
   {
    dst=Combat
    rule=AnimGraph.AnimGraphNode_StateMachine_6.BodyAdditives.AnimStateTransitionNode_2.Transition
    src=BodyAdditiveTempState
    via=AnimStateTransitionNode_2
   }
   {
    dst=Worried
    rule=AnimGraph.AnimGraphNode_StateMachine_6.BodyAdditives.AnimStateTransitionNode_0.Transition
    src=BodyAdditiveTempState
    via=AnimStateTransitionNode_0
   }
   {
    dst=BodyAdditiveTempState
    rule=AnimGraph.AnimGraphNode_StateMachine_6.BodyAdditives.AnimStateTransitionNode_3.Transition
    src=Combat
    via=AnimStateTransitionNode_3
   }
   {
    dst=BodyAdditiveTempState
    rule=AnimGraph.AnimGraphNode_StateMachine_6.BodyAdditives.AnimStateTransitionNode_1.Transition
    src=Worried
    via=AnimStateTransitionNode_1
   }
  ]
 }
 BodyAdditiveTempState {
  path=AnimGraph.AnimGraphNode_StateMachine_6.BodyAdditives.AnimStateNode_0.BodyAdditiveTempState
  nodes [Output Animation Pose(Result=Additive Identity Pose())]
 }
 Worried {
  path=AnimGraph.AnimGraphNode_StateMachine_6.BodyAdditives.AnimStateNode_1.Worried
  nodes [Output Animation Pose(Result=Two Way Blend(
     A=Additive Identity Pose()
     B="Sequence Player (additive) 'SitPlant_COR_Worried_Pose_ADD'"() {bLoopAnimation=false, sequence=&SitPlant_COR_Worried_Pose_ADD}
     Alpha=Get WorriedAddPoseAlpha()
    ) {blendNode {alphaBoolBlend {blendOption=Linear}}})]
 }
 Combat {
  path=AnimGraph.AnimGraphNode_StateMachine_6.BodyAdditives.AnimStateNode_2.Combat
  nodes [Output Animation Pose(Result="Sequence Player (additive) 'SitPlant_COR_Combat_Pose_ADD'"() {bLoopAnimation=false, sequence=&SitPlant_COR_Combat_Pose_ADD})]
 }
 Transition {
  path=AnimGraph.AnimGraphNode_StateMachine_6.BodyAdditives.AnimStateTransitionNode_0.Transition
  nodes [Result(bCanEnterTransition=Get bIsMoodWorried())]
 }
 Transition {
  path=AnimGraph.AnimGraphNode_StateMachine_6.BodyAdditives.AnimStateTransitionNode_1.Transition
  nodes [Result(bCanEnterTransition=NOT Boolean(A=Get bIsMoodWorried()))]
 }
 Transition {
  path=AnimGraph.AnimGraphNode_StateMachine_6.BodyAdditives.AnimStateTransitionNode_2.Transition
  nodes [Result(bCanEnterTransition=Get bIsDelayedMoodCombat())]
 }
 Transition {
  path=AnimGraph.AnimGraphNode_StateMachine_6.BodyAdditives.AnimStateTransitionNode_3.Transition
  nodes [Result(bCanEnterTransition=NOT Boolean(A=Get bIsMoodCombat()))]
 }
 AnticPlaying_Enter {nodes [
   AnticPlaying_Enter!() -> (then, Context:"Anim Update Context Structure (by ref)", Node:"Anim Node Reference Structure (by ref)") {is=title K2Node_FunctionEntry}
   > Set bIsPlayingAntic(bIsPlayingAntic:bool=true) -> (Output_Get:bool)
   > Return Node!(execute)
  ]}
 AnticPlaying_Exit {nodes [
   AnticPlaying_Exit!() -> (then, Context:"Anim Update Context Structure (by ref)", Node:"Anim Node Reference Structure (by ref)") {is=title K2Node_FunctionEntry}
   > Set bIsPlayingAntic(bIsPlayingAntic=false)
   > Return Node()
  ]}
 CanPlayAntic_Enter {nodes [
   CanPlayAntic_Enter!() -> (then, Context:"Anim Update Context Structure (by ref)", Node:"Anim Node Reference Structure (by ref)") {is=title K2Node_FunctionEntry}
   > Set bIsCanPlayAntic(bIsCanPlayAntic:bool=true) -> (Output_Get:bool)
   > Return Node()
  ]}
 CanPlayAntic_Exit {nodes [
   CanPlayAntic_Exit!() -> (then, Context:"Anim Update Context Structure (by ref)", Node:"Anim Node Reference Structure (by ref)") {is=title K2Node_FunctionEntry}
   > Set bIsCanPlayAntic(bIsCanPlayAntic=false)
   > Return Node()
  ]}
 EventGraph {nodes [
   n1 = Get bIsMoving()
   n2 = "float < float"(A=GetCurveValue(CurveName=Walk), B=0.7)
   n3 = Get bIsCompanionStopping()
   Set IsLocomtionStop_b(IsLocomtionStop_b:bool=OR Boolean(A=AND Boolean(A=@n3, B=@n2), B=NOT Boolean(A=@n1))) -> (Output_Get:bool)
   n4 = Get bIsCompanionStopping()
   n5 = Get CompanionSpeed2D() -> (CompanionSpeed2D:float)
   > Set IsStopToLocomotion_b(IsStopToLocomotion_b:bool=AND Boolean(A="float >= float"(A=@n5, B=100), B=NOT Boolean(A=@n4))) -> (Output_Get:bool)
   Event BlueprintUpdateAnimation()
  ]}
 GetSequencerAnimSlotNames {nodes [
   GetSequencerAnimSlotNames() {is=title K2Node_FunctionEntry}
   > Return Node!(execute, ReturnValue:Array of Names=Make Array("[0]":Name=SequencerDefault) -> (Array:Array of Names))
  ]}
 IsFrontEndSpawn_False {nodes [
   IsFrontEndSpawn_False!() -> (then, Context:"Anim Update Context Structure (by ref)", Node:"Anim Node Reference Structure (by ref)") {is=title K2Node_FunctionEntry}
   > Return Node!(execute)
  ]}
 IsSit_Enter {nodes [
   IsSit_Enter!() -> (then, Context:"Anim Update Context Structure (by ref)", Node:"Anim Node Reference Structure (by ref)") {is=title K2Node_FunctionEntry}
   > Set bIsSitting(bIsSitting:bool=true) -> (Output_Get:bool)
   > Set bIsCanPlayAntic(bIsCanPlayAntic=false)
   > Return Node()
  ]}
 LocoStart_Entry {nodes [
   LocoStart_Entry!() -> (then, Context:"Anim Update Context Structure (by ref)", Node:"Anim Node Reference Structure (by ref)") {is=title K2Node_FunctionEntry}
   > Set bIsLocoStart(bIsLocoStart:bool=true) -> (Output_Get:bool)
   > Return Node()
  ]}
 OnExit_SitState {nodes [
   OnExit_SitState!() -> (then, Context:"Anim Update Context Structure (by ref)", Node:"Anim Node Reference Structure (by ref)") {is=title K2Node_FunctionEntry}
   > Set bIsSitting(bIsSitting=false)
   > Set bIsCanPlayAntic(bIsCanPlayAntic=false)
   > Return Node()
  ]}
 OnStateExit {nodes [
   OnStateExit!() -> (then, Context:"Anim Update Context Structure (by ref)", Node:"Anim Node Reference Structure (by ref)") {is=title K2Node_FunctionEntry}
   > Set UpdatedGoalRotation(UpdatedGoalRotation:float=0) -> (Output_Get:float)
   > Set bIsLocoStart(bIsLocoStart=false)
   > Return Node()
  ]}
 SpecialIdle_StateEntry {nodes [
   SpecialIdle_StateEntry!() -> (then, Context:"Anim Update Context Structure (by ref)", Node:"Anim Node Reference Structure (by ref)") {is=title K2Node_FunctionEntry}
   > Set bIsSpecialIdlePlaying(bIsSpecialIdlePlaying:bool=true) -> (Output_Get:bool)
   > Return Node()
  ]}
 SpecialIdle_StateExit {nodes [
   SpecialIdle_StateExit!() -> (then, Context:"Anim Update Context Structure (by ref)", Node:"Anim Node Reference Structure (by ref)") {is=title K2Node_FunctionEntry}
   > Set bIsSpecialIdlePlaying(bIsSpecialIdlePlaying=false)
   > Return Node()
  ]}
]
variables [
 ChooserTables {type=ProxyTable, default=/CosmeticCompanions/Assets/Quadruped/SitPlant/Animation/Choosers/PRT_SitPlant_Choosers}
 IsLocomtionStop_b {type=boolean}
 IsStopToLocomotion_b {type=boolean}
]
