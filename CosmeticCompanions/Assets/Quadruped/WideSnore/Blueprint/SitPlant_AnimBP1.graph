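# Animation Blueprint for the WideSnore quadruped cosmetic companion, reusing many SitPlant animations.
# Output runs from the Locomotion state machine through additive breathing, a montage slot, chooser-driven additive idle antics, a look-at aim offset and Dead Blending.
# A cached low-fidelity pose has its own idle/locomotion machine with sit and special-idle states.
# The low-fidelity switch, mood and tail additives, slope warping and leg IK are not wired to the output.
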AnimBlueprint /CosmeticCompanions/Assets/Quadruped/WideSnore/Blueprint/SitPlant_AnimBP1
target_skeleton=/CosmeticCompanions/Assets/Quadruped/WideSnore/Meshes/WideSnore_Quadruped_Skeleton
graphs [
 AnimGraph {nodes [
   AnimGraphNode_StateMachine_1 = BodyAdditives() {is=title AnimGraphNode_StateMachine}
   Apply Additive(Additive=@AnimGraphNode_StateMachine_1) {alphaBoolBlend {blendOption=Linear}}
   Apply Additive(Additive=Blendspace Player 'SitPlant_COR_Loco_Turn_BS'!(X=Get CompanionDirection() -> (CompanionDirection:float)) -> (Pose) {blendSpace=/CosmeticCompanions/Assets/Quadruped/SitPlant/Animation/Core/SitPlant_COR_Loco_Turn_BS}) {alphaBoolBlend {blendOption=Linear}}
   Apply Additive(
    Additive="Sequence Player (additive) 'SitPlant_COR_Tail_Bold_ADD'"() {sequence=/CosmeticCompanions/Assets/Quadruped/SitPlant/Animation/Core/SitPlant_COR_Tail_Bold_ADD}
    Alpha=Get Appendage01() -> (Appendage01:float)
   ) {alphaCurveName=Appendage01, alphaBoolBlend {blendOption=Linear}}
   AnimGraphNode_StateMachine_4 = Tail() {is=title AnimGraphNode_StateMachine, bAllowConduitEntryStates=true}
   Apply Additive(Additive=@AnimGraphNode_StateMachine_4, Alpha=Get Appendage01()) {alphaCurveName=Appendage01, alphaBoolBlend {blendOption=Linear}}
   Blend Poses by bool(
    BlendPose_0=Use cached pose 'LowFidelityPose'()
    BlendPose_1=Use cached pose 'HighFidelityPose'()
    BlendTime_0=0.2
    BlendTime_1=0.2
    bActiveValue=Get IsLowFidelity() -> (IsLowFidelity:bool)
   ) {blendType=Linear, transitionType=Inertialization}
   n1 = Quadruped Slope Warping!(ComponentPose:Component Space Pose Link Structure=Local To Component(), Alpha:float=1) -> (Pose:Component Space Pose Link Structure) {
    is=title FortAnimGraphNode_QuadrupedSlopeWarping
    alphaBoolBlend {blendOption=Linear}
    armsFeetDefinitions [
     {numBonesInLimb=3, fKFootBone {boneName=ball_fr_l}, iKFootBone {boneName=IK_Foot_Front_L}}
     {numBonesInLimb=3, fKFootBone {boneName=ball_fr_r}, iKFootBone {boneName=IK_Foot_Front_R}}
    ]
    armsIKFootRootBone {boneName=IK_Foot_Front_Root}
    armsRootBone {boneName=spine_03}
    feetDefinitions [
     {numBonesInLimb=3, fKFootBone {boneName=ball_bk_l}, iKFootBone {boneName=IK_Foot_Rear_L}}
     {numBonesInLimb=3, fKFootBone {boneName=ball_bk_r}, iKFootBone {boneName=IK_Foot_Rear_R}}
    ]
    iKFootRootBone {boneName=IK_Foot_Rear_Root}
    pelvisBone {boneName=hips}
    skeletonRootBone {boneName=root}
   }
   n2 = Leg IK!(ComponentPose:Component Space Pose Link Structure=@n1, Alpha:float=1) -> (Pose:Component Space Pose Link Structure) {
    is=title AnimGraphNode_LegIK
    alphaBoolBlend {blendOption=Linear}
    legsDefinition [
     {
      bEnableKneeTwistCorrection=true
      bEnableRotationLimit=false
      footBoneForwardAxis=Y
      hingeRotationAxis=None
      minRotationAngle=15
      numBonesInLimb=3
      twistOffsetCurveName=None
      fKFootBone {boneName=ball_bk_l}
      iKFootBone {boneName=IK_Foot_Rear_L}
     }
     {
      bEnableKneeTwistCorrection=true
      bEnableRotationLimit=false
      footBoneForwardAxis=Y
      hingeRotationAxis=None
      minRotationAngle=15
      numBonesInLimb=3
      twistOffsetCurveName=None
      fKFootBone {boneName=ball_bk_r}
      iKFootBone {boneName=IK_Foot_Rear_R}
     }
     {
      bEnableKneeTwistCorrection=true
      bEnableRotationLimit=false
      footBoneForwardAxis=Y
      hingeRotationAxis=None
      minRotationAngle=15
      numBonesInLimb=2
      twistOffsetCurveName=None
      fKFootBone {boneName=ball_fr_r}
      iKFootBone {boneName=IK_Foot_Front_R}
     }
     {
      bEnableKneeTwistCorrection=true
      bEnableRotationLimit=false
      footBoneForwardAxis=Y
      hingeRotationAxis=None
      minRotationAngle=15
      numBonesInLimb=2
      twistOffsetCurveName=None
      fKFootBone {boneName=ball_fr_l}
      iKFootBone {boneName=IK_Foot_Front_L}
     }
    ]
   }
   Component To Local(ComponentPose=@n2)
   Layered blend per bone(
    BasePose=Sequence Player 'SitPlant_COR_Loco_Run'() {sequence=/CosmeticCompanions/Assets/Quadruped/SitPlant/Animation/Core/SitPlant_COR_Loco_Run}
    BlendPoses_0=Sequence Player 'SitPlant_xDEV_Idle_Loop'() {sequence=/CosmeticCompanions/Assets/Quadruped/SitPlant/Animation/Developer/SitPlant_xDEV_Idle_Loop}
   ) {blendMode=BlendMask, blendMasks [/CosmeticCompanions/Assets/Quadruped/SitPlant/Meshes/SitPlant_Quadruped_Skeleton:FaceOff], layerSetup []}
   AnimGraphNode_StateMachine_0 = IdleAntics() {is=title AnimGraphNode_StateMachine, bAllowConduitEntryStates=true}
   n3 = Apply Additive(
    Base=Slot 'DefaultSlot'(Source=Use cached pose 'LocomotionPose'())
    Additive=@AnimGraphNode_StateMachine_0
    Alpha=Get IdleAnticAlpha() -> (IdleAnticAlpha:double)
   ) {alphaBoolBlend {blendOption=Linear}}
   n4 = Get LookAtPitchActual() -> (LookAtPitchActual:float)
   n5 = Get LookAtYawAdjusted() -> (LookAtYawAdjusted:double)
   Output Pose(Result=Dead Blending!(Source:Pose Link Structure=AimOffset Player 'WideSnore_AO'!(X:float=@n5, Y:float=@n4, BasePose:Pose Link Structure=@n3, Alpha:float=1) -> (Pose:Pose Link Structure) {blendSpace=/CosmeticCompanions/Assets/Quadruped/WideSnore/Animation/AimOffsets/WideSnore_AO, alphaBoolBlend {blendOption=Linear}}) -> (Pose:Pose Link Structure) {is=title AnimGraphNode_DeadBlending, defaultBlendMode=Linear})
   n6 = Get BreathingPlayRate() -> (BreathingPlayRate:float)
   AnimGraphNode_StateMachine_3 = Locomotion() {is=title AnimGraphNode_StateMachine, bAllowConduitEntryStates=true}
   Save cached pose 'LocomotionPose'(Pose=Apply Additive(
     Base=@AnimGraphNode_StateMachine_3
     Additive="Sequence Player (additive) 'WideSnore_COR_Breathing_ADD'"!(PlayRate:float=@n6) -> (Pose) {sequence=/CosmeticCompanions/Assets/Quadruped/WideSnore/Animation/Core/WideSnore_COR_Breathing_ADD}
     Alpha=Get BreathingAlpha() -> (BreathingAlpha:double)
    ) {alphaBoolBlend {blendOption=Linear}}) {cacheName=LocomotionPose}
   n7 = Get CompanionDirection()
   AnimGraphNode_StateMachine_2 = LowFidelityStateMachine() {is=title AnimGraphNode_StateMachine, bAllowConduitEntryStates=true}
   n8 = AimOffset Player 'WideSnore_AO'(
    X=Get LookAtYawAdjusted()
    Y=Get LookAtPitchActual()
    BasePose=Apply Additive(
     Base=@AnimGraphNode_StateMachine_2
     Additive=Blendspace Player 'SitPlant_COR_Loco_Turn_BS'(X=@n7) {blendSpace=&SitPlant_COR_Loco_Turn_BS}
    ) {alphaBoolBlend {blendOption=Linear}}
    Alpha=1
   ) {blendSpace=&WideSnore_AO, alphaBoolBlend {blendOption=Linear}}
   Save cached pose 'LowFidelityPose'(Pose=Dead Blending(Source=Slot 'DefaultSlot'(Source=@n8)) {defaultBlendMode=Linear}) {cacheName=LowFidelityPose}
   Use cached pose 'LocomotionPlusAdditivePose'()
  ]}
 IdleAntics {
  entry=Conduit
  path=AnimGraph.AnimGraphNode_StateMachine_0.IdleAntics
  ids=id
  nodes [
   AnimStateConduitNode_0 = Conduit()
   AnimStateNode_0 = Default() {is=title AnimStateNode}
   AnimStateNode_7 = FrontEndSpawnAnticChooser() {is=title AnimStateNode}
   AnimStateNode_9 = AnticChooser() {is=title AnimStateNode}
   AnimStateTransitionNode_0 = AnimStateTransitionNode() {crossfadeDuration=0.25, logicType=TLT_Inertialization, priorityOrder=100}
   AnimStateTransitionNode_1 = AnimStateTransitionNode() {bAutomaticRuleBasedOnSequencePlayerInState=true}
   AnimStateTransitionNode_14 = AnimStateTransitionNode() {logicType=TLT_Inertialization}
   AnimStateTransitionNode_18 = AnimStateTransitionNode() {logicType=TLT_Inertialization, priorityOrder=2}
   AnimStateTransitionNode_4 = AnimStateTransitionNode()
   AnimStateTransitionNode_5 = AnimStateTransitionNode() {bAutomaticRuleBasedOnSequencePlayerInState=true}
  ]
  links [
   {
    dst=Conduit
    rule=AnimGraph.AnimGraphNode_StateMachine_0.IdleAntics.AnimStateTransitionNode_5.Transition
    src=AnticChooser
    via=AnimStateTransitionNode_5
   }
   {
    dst=AnticChooser
    rule=AnimGraph.AnimGraphNode_StateMachine_0.IdleAntics.AnimStateTransitionNode_18.Transition
    src=Conduit
    via=AnimStateTransitionNode_18
   }
   {
    dst=Default
    rule=AnimGraph.AnimGraphNode_StateMachine_0.IdleAntics.AnimStateTransitionNode_0.Transition
    src=Conduit
    via=AnimStateTransitionNode_0
   }
   {
    dst=FrontEndSpawnAnticChooser
    rule=AnimGraph.AnimGraphNode_StateMachine_0.IdleAntics.AnimStateTransitionNode_14.Transition
    src=Conduit
    via=AnimStateTransitionNode_14
   }
   {
    dst=Conduit
    rule=AnimGraph.AnimGraphNode_StateMachine_0.IdleAntics.AnimStateTransitionNode_4.Transition
    src=Default
    via=AnimStateTransitionNode_4
   }
   {
    dst=Conduit
    rule=AnimGraph.AnimGraphNode_StateMachine_0.IdleAntics.AnimStateTransitionNode_1.Transition
    src=FrontEndSpawnAnticChooser
    via=AnimStateTransitionNode_1
   }
  ]
 }
 Conduit {path=AnimGraph.AnimGraphNode_StateMachine_0.IdleAntics.AnimStateConduitNode_0.Conduit, nodes [Result(bCanEnterTransition=true)]}
 Default {
  path=AnimGraph.AnimGraphNode_StateMachine_0.IdleAntics.AnimStateNode_0.Default
  nodes [
   "Sequence Player (additive) 'SitPlant_ACT_Break_ADD'"() {bLoopAnimation=false, sequence=/CosmeticCompanions/Assets/Quadruped/SitPlant/Animation/Actions/SitPlant_ACT_Break_ADD}
   AnimGraphNode_ChooserPlayer_1 = Chooser Player: CHT_WideSnore_AnticsNeutral!() -> (Pose:Pose Link Structure) {
    is=title AnimGraphNode_ChooserPlayer
    chooser {
     _structType=/Script/Chooser.EvaluateChooser
     chooser=/CosmeticCompanions/Assets/Quadruped/WideSnore/Animation/Choosers/CHT_WideSnore_AnticsNeutral
    }
    chooserContextDefinition [
     {
      _structType=/Script/Chooser.ContextObjectTypeClass
      class=/CosmeticCompanions/Assets/Quadruped/SitPlant/Blueprint/SitPlant_AnimBP.SitPlant_AnimBP_C
      direction=ReadWrite
     }
     {struct=/Script/Chooser.ChooserPlayerSettings, _structType=/Script/Chooser.ContextObjectTypeStruct, direction=Read}
    ]
    defaultSettings {blendOption=Linear}
   }
   Output Animation Pose(Result=@AnimGraphNode_ChooserPlayer_1)
  ]
 }
 AnimationBlendStackGraph_0 {
  path=AnimGraph.AnimGraphNode_StateMachine_0.IdleAntics.AnimStateNode_0.Default.AnimGraphNode_ChooserPlayer_1.AnimationBlendStackGraph_0
  nodes [Output Pose!(Result:Pose Link Structure=Blend Stack Input!() -> (Pose:Pose Link Structure) {is=title AnimGraphNode_BlendStackInput}) {is=title AnimGraphNode_BlendStackResult}]
 }
 FrontEndSpawnAnticChooser {
  path=AnimGraph.AnimGraphNode_StateMachine_0.IdleAntics.AnimStateNode_7.FrontEndSpawnAnticChooser
  nodes [
   AnimGraphNode_ChooserPlayer_0 = Chooser Player: CHT_WideSnore_AnticsFrontEnd!() -> (Pose:Pose Link Structure) {
    is=title AnimGraphNode_ChooserPlayer
    evaluationFrequency=OnInitialUpdate
    chooser {
     _structType=&EvaluateChooser
     chooser=/CosmeticCompanions/Assets/Quadruped/WideSnore/Animation/Choosers/CHT_WideSnore_AnticsFrontEnd
    }
    chooserContextDefinition [
     {
      _structType=&ContextObjectTypeClass
      class=&SitPlant_AnimBP_C
      direction=ReadWrite
     }
     {struct=&ChooserPlayerSettings, _structType=&ContextObjectTypeStruct, direction=Read}
    ]
    defaultSettings {blendOption=Linear}
   }
   Output Animation Pose(Result=@AnimGraphNode_ChooserPlayer_0) {
    node {stateEntryFunction {functionName=OnStateEntry_3}}
    stateEntryFunction {bSelfContext=true, memberGuid=#1, memberName=IsFrontEndSpawn_False}
   }
  ]
 }
 AnimationBlendStackGraph_0 {
  path=AnimGraph.AnimGraphNode_StateMachine_0.IdleAntics.AnimStateNode_7.FrontEndSpawnAnticChooser.AnimGraphNode_ChooserPlayer_0.AnimationBlendStackGraph_0
  nodes [Output Pose(Result=Blend Stack Input())]
 }
 AnticChooser {
  path=AnimGraph.AnimGraphNode_StateMachine_0.IdleAntics.AnimStateNode_9.AnticChooser
  nodes [
   AnimGraphNode_ChooserPlayer_0 = Chooser Player: CHT_WideSnore_AnticsSelection!() -> (Pose:Pose Link Structure) {
    is=title AnimGraphNode_ChooserPlayer
    chooser {
     _structType=&EvaluateChooser
     chooser=/CosmeticCompanions/Assets/Quadruped/WideSnore/Animation/Choosers/CHT_WideSnore_AnticsSelection
    }
    chooserContextDefinition [
     {
      _structType=&ContextObjectTypeClass
      class=&SitPlant_AnimBP_C
      direction=ReadWrite
     }
     {struct=&ChooserPlayerSettings, _structType=&ContextObjectTypeStruct, direction=Read}
    ]
    defaultSettings {blendOption=Linear}
   }
   Output Animation Pose(Result=@AnimGraphNode_ChooserPlayer_0)
  ]
 }
 AnimationBlendStackGraph_0 {
  path=AnimGraph.AnimGraphNode_StateMachine_0.IdleAntics.AnimStateNode_9.AnticChooser.AnimGraphNode_ChooserPlayer_0.AnimationBlendStackGraph_0
  nodes [Output Pose(Result=Blend Stack Input())]
 }
 Transition {path=AnimGraph.AnimGraphNode_StateMachine_0.IdleAntics.AnimStateTransitionNode_0.Transition, nodes [Result(bCanEnterTransition=true)]}
 Transition {path=AnimGraph.AnimGraphNode_StateMachine_0.IdleAntics.AnimStateTransitionNode_1.Transition, nodes [Result()]}
 Transition {
  path=AnimGraph.AnimGraphNode_StateMachine_0.IdleAntics.AnimStateTransitionNode_14.Transition
  nodes [Result(bCanEnterTransition=Get IsFrontEndSpawnAnticChooser() -> (IsFrontEndSpawnAnticChooser:bool))]
 }
 Transition {
  path=AnimGraph.AnimGraphNode_StateMachine_0.IdleAntics.AnimStateTransitionNode_18.Transition
  nodes [Result(bCanEnterTransition=Break Companion Random Number Timer!(CompanionRandomNumberTimer:"Companion Random Number Timer Structure (by ref)"=Get AnticsTimer() -> (AnticsTimer:Companion Random Number Timer Structure)) -> (ValueInt:int, ValueFloat:float=~, bJustChanged:bool=~) {is=title K2Node_BreakStruct}.bJustChanged)]
 }
 Transition {path=AnimGraph.AnimGraphNode_StateMachine_0.IdleAntics.AnimStateTransitionNode_4.Transition, nodes [Result(bCanEnterTransition=true)]}
 Transition {path=AnimGraph.AnimGraphNode_StateMachine_0.IdleAntics.AnimStateTransitionNode_5.Transition, nodes [Result()]}
 BodyAdditives {
  entry=BodyAdditiveTempState
  path=AnimGraph.AnimGraphNode_StateMachine_1.BodyAdditives
  ids=id
  nodes [
   AnimStateNode_0 = BodyAdditiveTempState() {is=title AnimStateNode}
   AnimStateNode_1 = Worried() {is=title AnimStateNode}
   AnimStateNode_2 = Combat() {is=title AnimStateNode}
   AnimStateTransitionNode_0 = AnimStateTransitionNode() {crossfadeDuration=0.35, logicType=TLT_Inertialization}
   AnimStateTransitionNode_1 = AnimStateTransitionNode() {logicType=TLT_Inertialization}
   AnimStateTransitionNode_2 = AnimStateTransitionNode() {logicType=TLT_Inertialization, priorityOrder=2}
   AnimStateTransitionNode_3 = AnimStateTransitionNode()
  ]
  links [
   {
    dst=Combat
    rule=AnimGraph.AnimGraphNode_StateMachine_1.BodyAdditives.AnimStateTransitionNode_2.Transition
    src=BodyAdditiveTempState
    via=AnimStateTransitionNode_2
   }
   {
    dst=Worried
    rule=AnimGraph.AnimGraphNode_StateMachine_1.BodyAdditives.AnimStateTransitionNode_0.Transition
    src=BodyAdditiveTempState
    via=AnimStateTransitionNode_0
   }
   {
    dst=BodyAdditiveTempState
    rule=AnimGraph.AnimGraphNode_StateMachine_1.BodyAdditives.AnimStateTransitionNode_3.Transition
    src=Combat
    via=AnimStateTransitionNode_3
   }
   {
    dst=BodyAdditiveTempState
    rule=AnimGraph.AnimGraphNode_StateMachine_1.BodyAdditives.AnimStateTransitionNode_1.Transition
    src=Worried
    via=AnimStateTransitionNode_1
   }
  ]
 }
 BodyAdditiveTempState {
  path=AnimGraph.AnimGraphNode_StateMachine_1.BodyAdditives.AnimStateNode_0.BodyAdditiveTempState
  nodes [Output Animation Pose(Result=Additive Identity Pose())]
 }
 Worried {
  path=AnimGraph.AnimGraphNode_StateMachine_1.BodyAdditives.AnimStateNode_1.Worried
  nodes [Output Animation Pose(Result=Two Way Blend(
     A=Additive Identity Pose()
     B="Sequence Player (additive) 'SitPlant_COR_Worried_Pose_ADD'"() {bLoopAnimation=false, sequence=/CosmeticCompanions/Assets/Quadruped/SitPlant/Animation/Core/SitPlant_COR_Worried_Pose_ADD}
     Alpha=Get WorriedAddPoseAlpha() -> (WorriedAddPoseAlpha:double)
    ) {blendNode {alphaBoolBlend {blendOption=Linear}}})]
 }
 Combat {
  path=AnimGraph.AnimGraphNode_StateMachine_1.BodyAdditives.AnimStateNode_2.Combat
  nodes [Output Animation Pose(Result="Sequence Player (additive) 'SitPlant_COR_Combat_Pose_ADD'"() {bLoopAnimation=false, sequence=/CosmeticCompanions/Assets/Quadruped/SitPlant/Animation/Core/SitPlant_COR_Combat_Pose_ADD})]
 }
 Transition {
  path=AnimGraph.AnimGraphNode_StateMachine_1.BodyAdditives.AnimStateTransitionNode_0.Transition
  nodes [Result(bCanEnterTransition=Get IsMoodWorried() -> (IsMoodWorried:bool))]
 }
 Transition {
  path=AnimGraph.AnimGraphNode_StateMachine_1.BodyAdditives.AnimStateTransitionNode_1.Transition
  nodes [Result(bCanEnterTransition=NOT Boolean(A=Get IsMoodWorried()))]
 }
 Transition {
  path=AnimGraph.AnimGraphNode_StateMachine_1.BodyAdditives.AnimStateTransitionNode_2.Transition
  nodes [Result(bCanEnterTransition=Get BodyAdditiveToCombat() -> (BodyAdditiveToCombat:bool))]
 }
 Transition {
  path=AnimGraph.AnimGraphNode_StateMachine_1.BodyAdditives.AnimStateTransitionNode_3.Transition
  nodes [Result(bCanEnterTransition=Get CombatToBodyAdditive() -> (CombatToBodyAdditive:bool))]
 }
 LowFidelityStateMachine {
  entry=Conduit
  path=AnimGraph.AnimGraphNode_StateMachine_2.LowFidelityStateMachine
  ids=id
  nodes [
   AnimStateConduitNode_0 = Conduit()
   AnimStateNode_1 = LowFidelity_Idle() {is=title AnimStateNode}
   AnimStateNode_2 = LowFidelity_Loco() {is=title AnimStateNode}
   AnimStateTransitionNode_0 = AnimStateTransitionNode() {logicType=TLT_Inertialization, priorityOrder=100}
   AnimStateTransitionNode_1 = AnimStateTransitionNode() {logicType=TLT_Inertialization}
   AnimStateTransitionNode_2 = AnimStateTransitionNode()
   AnimStateTransitionNode_3 = AnimStateTransitionNode()
  ]
  links [
   {
    dst=LowFidelity_Idle
    rule=AnimGraph.AnimGraphNode_StateMachine_2.LowFidelityStateMachine.AnimStateTransitionNode_0.Transition
    src=Conduit
    via=AnimStateTransitionNode_0
   }
   {
    dst=LowFidelity_Loco
    rule=AnimGraph.AnimGraphNode_StateMachine_2.LowFidelityStateMachine.AnimStateTransitionNode_1.Transition
    src=Conduit
    via=AnimStateTransitionNode_1
   }
   {
    dst=Conduit
    rule=AnimGraph.AnimGraphNode_StateMachine_2.LowFidelityStateMachine.AnimStateTransitionNode_2.Transition
    src=LowFidelity_Idle
    via=AnimStateTransitionNode_2
   }
   {
    dst=Conduit
    rule=AnimGraph.AnimGraphNode_StateMachine_2.LowFidelityStateMachine.AnimStateTransitionNode_3.Transition
    src=LowFidelity_Loco
    via=AnimStateTransitionNode_3
   }
  ]
 }
 Conduit {path=AnimGraph.AnimGraphNode_StateMachine_2.LowFidelityStateMachine.AnimStateConduitNode_0.Conduit, nodes [Result(bCanEnterTransition=true)]}
 LowFidelity_Idle {
  path=AnimGraph.AnimGraphNode_StateMachine_2.LowFidelityStateMachine.AnimStateNode_1.LowFidelity_Idle
  nodes [
   AnimGraphNode_StateMachine_2 = DefaultIdleStateMachine() {is=title AnimGraphNode_StateMachine, bAllowConduitEntryStates=true}
   Output Animation Pose(Result=Modify Curve(SourcePose=@AnimGraphNode_StateMachine_2, CurveValues_0=1, Alpha=1) {curveNames [idle], curveValues [0]})
  ]
 }
 DefaultIdleStateMachine {
  entry=Conduit
  path=AnimGraph.AnimGraphNode_StateMachine_2.LowFidelityStateMachine.AnimStateNode_1.LowFidelity_Idle.AnimGraphNode_StateMachine_2.DefaultIdleStateMachine
  ids=id
  nodes [
   AnimStateConduitNode_0 = Conduit()
   AnimStateNode_0 = DefaultIdle() {is=title AnimStateNode}
   AnimStateNode_1 = SpecialIdle() {is=title AnimStateNode}
   AnimStateNode_2 = SpecialIdleToDefaultIdle() {is=title AnimStateNode}
   AnimStateTransitionNode_0 = AnimStateTransitionNode() {logicType=TLT_Inertialization, priorityOrder=100}
   AnimStateTransitionNode_1 = AnimStateTransitionNode() {logicType=TLT_Inertialization}
   AnimStateTransitionNode_2 = AnimStateTransitionNode()
   AnimStateTransitionNode_3 = AnimStateTransitionNode() {crossfadeDuration=0.3, logicType=TLT_Inertialization}
   AnimStateTransitionNode_4 = AnimStateTransitionNode() {bAutomaticRuleBasedOnSequencePlayerInState=true, logicType=TLT_Inertialization}
   AnimStateTransitionNode_5 = AnimStateTransitionNode() {logicType=TLT_Inertialization, priorityOrder=2}
  ]
  links [
   {
    dst=DefaultIdle
    rule=AnimGraph.AnimGraphNode_StateMachine_2.LowFidelityStateMachine.AnimStateNode_1.LowFidelity_Idle.AnimGraphNode_StateMachine_2.DefaultIdleStateMachine.AnimStateTransitionNode_0.Transition
    src=Conduit
    via=AnimStateTransitionNode_0
   }
   {
    dst=SpecialIdle
    rule=AnimGraph.AnimGraphNode_StateMachine_2.LowFidelityStateMachine.AnimStateNode_1.LowFidelity_Idle.AnimGraphNode_StateMachine_2.DefaultIdleStateMachine.AnimStateTransitionNode_1.Transition
    src=Conduit
    via=AnimStateTransitionNode_1
   }
   {
    dst=SpecialIdleToDefaultIdle
    rule=AnimGraph.AnimGraphNode_StateMachine_2.LowFidelityStateMachine.AnimStateNode_1.LowFidelity_Idle.AnimGraphNode_StateMachine_2.DefaultIdleStateMachine.AnimStateTransitionNode_5.Transition
    src=Conduit
    via=AnimStateTransitionNode_5
   }
   {
    dst=Conduit
    rule=AnimGraph.AnimGraphNode_StateMachine_2.LowFidelityStateMachine.AnimStateNode_1.LowFidelity_Idle.AnimGraphNode_StateMachine_2.DefaultIdleStateMachine.AnimStateTransitionNode_2.Transition
    src=DefaultIdle
    via=AnimStateTransitionNode_2
   }
   {
    dst=SpecialIdleToDefaultIdle
    rule=AnimGraph.AnimGraphNode_StateMachine_2.LowFidelityStateMachine.AnimStateNode_1.LowFidelity_Idle.AnimGraphNode_StateMachine_2.DefaultIdleStateMachine.AnimStateTransitionNode_3.Transition
    src=SpecialIdle
    via=AnimStateTransitionNode_3
   }
   {
    dst=DefaultIdle
    rule=AnimGraph.AnimGraphNode_StateMachine_2.LowFidelityStateMachine.AnimStateNode_1.LowFidelity_Idle.AnimGraphNode_StateMachine_2.DefaultIdleStateMachine.AnimStateTransitionNode_4.Transition
    src=SpecialIdleToDefaultIdle
    via=AnimStateTransitionNode_4
   }
  ]
 }
 Conduit {
  path=AnimGraph.AnimGraphNode_StateMachine_2.LowFidelityStateMachine.AnimStateNode_1.LowFidelity_Idle.AnimGraphNode_StateMachine_2.DefaultIdleStateMachine.AnimStateConduitNode_0.Conduit
  nodes [Result(bCanEnterTransition=true)]
 }
 DefaultIdle {
  path=AnimGraph.AnimGraphNode_StateMachine_2.LowFidelityStateMachine.AnimStateNode_1.LowFidelity_Idle.AnimGraphNode_StateMachine_2.DefaultIdleStateMachine.AnimStateNode_0.DefaultIdle
  nodes [
   AnimGraphNode_StateMachine_1 = Idle() {is=title AnimGraphNode_StateMachine, bAllowConduitEntryStates=true}
   Output Animation Pose(Result=@AnimGraphNode_StateMachine_1) {
    node {stateExitFunction {functionName=OnExitSitState}}
    stateExitFunction {bSelfContext=true, memberGuid=#2, memberName=OnExit_SitState}
   }
  ]
 }
 Idle {
  entry=Conduit
  path=AnimGraph.AnimGraphNode_StateMachine_2.LowFidelityStateMachine.AnimStateNode_1.LowFidelity_Idle.AnimGraphNode_StateMachine_2.DefaultIdleStateMachine.AnimStateNode_0.DefaultIdle.AnimGraphNode_StateMachine_1.Idle
  ids=id
  nodes [
   Alias() {is=title AnimStateAliasNode, also {aliases [SitPose,StandToSit]}}
   AnimStateConduitNode_0 = Conduit()
   AnimStateNode_0 = StandToSit() {is=title AnimStateNode}
   AnimStateNode_1 = SitPose() {is=title AnimStateNode}
   AnimStateNode_2 = StandingPose() {is=title AnimStateNode}
   AnimStateNode_3 = SitToStand() {is=title AnimStateNode}
   AnimStateTransitionNode_0 = AnimStateTransitionNode() {bAutomaticRuleBasedOnSequencePlayerInState=true, logicType=TLT_Inertialization}
   AnimStateTransitionNode_1 = AnimStateTransitionNode() {logicType=TLT_Inertialization}
   AnimStateTransitionNode_2 = AnimStateTransitionNode() {logicType=TLT_Inertialization, priorityOrder=100}
   AnimStateTransitionNode_4 = AnimStateTransitionNode()
   AnimStateTransitionNode_6 = AnimStateTransitionNode() {bAutomaticRuleBasedOnSequencePlayerInState=true, crossfadeDuration=0.3, logicType=TLT_Inertialization}
   AnimStateTransitionNode_8 = AnimStateTransitionNode() {logicType=TLT_Inertialization}
  ]
  links [
   {
    dst=SitToStand
    rule=AnimGraph.AnimGraphNode_StateMachine_2.LowFidelityStateMachine.AnimStateNode_1.LowFidelity_Idle.AnimGraphNode_StateMachine_2.DefaultIdleStateMachine.AnimStateNode_0.DefaultIdle.AnimGraphNode_StateMachine_1.Idle.AnimStateTransitionNode_8.AnimationTransitionGraph_0
    src=Alias
    via=AnimStateTransitionNode_8
   }
   {
    dst=StandToSit
    rule=AnimGraph.AnimGraphNode_StateMachine_2.LowFidelityStateMachine.AnimStateNode_1.LowFidelity_Idle.AnimGraphNode_StateMachine_2.DefaultIdleStateMachine.AnimStateNode_0.DefaultIdle.AnimGraphNode_StateMachine_1.Idle.AnimStateTransitionNode_1.AnimationTransitionGraph_0
    src=Conduit
    via=AnimStateTransitionNode_1
   }
   {
    dst=StandingPose
    rule=AnimGraph.AnimGraphNode_StateMachine_2.LowFidelityStateMachine.AnimStateNode_1.LowFidelity_Idle.AnimGraphNode_StateMachine_2.DefaultIdleStateMachine.AnimStateNode_0.DefaultIdle.AnimGraphNode_StateMachine_1.Idle.AnimStateTransitionNode_2.AnimationTransitionGraph_0
    src=Conduit
    via=AnimStateTransitionNode_2
   }
   {
    dst=StandingPose
    rule=AnimGraph.AnimGraphNode_StateMachine_2.LowFidelityStateMachine.AnimStateNode_1.LowFidelity_Idle.AnimGraphNode_StateMachine_2.DefaultIdleStateMachine.AnimStateNode_0.DefaultIdle.AnimGraphNode_StateMachine_1.Idle.AnimStateTransitionNode_6.AnimationTransitionGraph_0
    src=SitToStand
    via=AnimStateTransitionNode_6
   }
   {
    dst=SitPose
    rule=AnimGraph.AnimGraphNode_StateMachine_2.LowFidelityStateMachine.AnimStateNode_1.LowFidelity_Idle.AnimGraphNode_StateMachine_2.DefaultIdleStateMachine.AnimStateNode_0.DefaultIdle.AnimGraphNode_StateMachine_1.Idle.AnimStateTransitionNode_0.AnimationTransitionGraph_0
    src=StandToSit
    via=AnimStateTransitionNode_0
   }
   {
    dst=Conduit
    rule=AnimGraph.AnimGraphNode_StateMachine_2.LowFidelityStateMachine.AnimStateNode_1.LowFidelity_Idle.AnimGraphNode_StateMachine_2.DefaultIdleStateMachine.AnimStateNode_0.DefaultIdle.AnimGraphNode_StateMachine_1.Idle.AnimStateTransitionNode_4.AnimationTransitionGraph_0
    src=StandingPose
    via=AnimStateTransitionNode_4
   }
  ]
 }
 Conduit {
  path=AnimGraph.AnimGraphNode_StateMachine_2.LowFidelityStateMachine.AnimStateNode_1.LowFidelity_Idle.AnimGraphNode_StateMachine_2.DefaultIdleStateMachine.AnimStateNode_0.DefaultIdle.AnimGraphNode_StateMachine_1.Idle.AnimStateConduitNode_0.Conduit
  nodes [Result(bCanEnterTransition=true)]
 }
 StandToSit {
  path=AnimGraph.AnimGraphNode_StateMachine_2.LowFidelityStateMachine.AnimStateNode_1.LowFidelity_Idle.AnimGraphNode_StateMachine_2.DefaultIdleStateMachine.AnimStateNode_0.DefaultIdle.AnimGraphNode_StateMachine_1.Idle.AnimStateNode_0.StandToSit
  nodes [Output Animation Pose(Result=Sequence Player 'SitPlant_COR_Sit_In'() {bLoopAnimation=false, sequence=/CosmeticCompanions/Assets/Quadruped/SitPlant/Animation/Core/SitPlant_COR_Sit_In}) {
    node {stateEntryFunction {functionName=OnEnterSitState}}
    stateEntryFunction {bSelfContext=true, memberGuid=#3, memberName=OnEnterSitState}
   }]
 }
 SitPose {
  path=AnimGraph.AnimGraphNode_StateMachine_2.LowFidelityStateMachine.AnimStateNode_1.LowFidelity_Idle.AnimGraphNode_StateMachine_2.DefaultIdleStateMachine.AnimStateNode_0.DefaultIdle.AnimGraphNode_StateMachine_1.Idle.AnimStateNode_1.SitPose
  nodes [Output Animation Pose(Result="Sequence Evaluator (by time) 'SitPlant_COR_Sit_In'"!(ExplicitTime:float=1.06) -> (Pose:Pose Link Structure) {is=title AnimGraphNode_SequenceEvaluator, sequence=&SitPlant_COR_Sit_In})]
 }
 StandingPose {
  path=AnimGraph.AnimGraphNode_StateMachine_2.LowFidelityStateMachine.AnimStateNode_1.LowFidelity_Idle.AnimGraphNode_StateMachine_2.DefaultIdleStateMachine.AnimStateNode_0.DefaultIdle.AnimGraphNode_StateMachine_1.Idle.AnimStateNode_2.StandingPose
  nodes [Output Animation Pose(Result=Blend Poses by bool(
     BlendPose_0=Sequence Player 'SitPlant_xDEV_React_Angry_02_Loop_OLD'() {sequence=/CosmeticCompanions/Assets/Quadruped/SitPlant/Animation/Developer/SitPlant_xDEV_React_Angry_02_Loop_OLD}
     BlendPose_1=Sequence Player 'WideSnore_COR_Stand_Pose'() {bLoopAnimation=false, sequence=/CosmeticCompanions/Assets/Quadruped/WideSnore/Animation/Core/WideSnore_COR_Stand_Pose}
     BlendTime_0=0.3
     BlendTime_1=0.4
     bActiveValue=Get IsMoodCombat() -> (IsMoodCombat:bool)
    ) {blendType=Linear, transitionType=Inertialization})]
 }
 SitToStand {
  path=AnimGraph.AnimGraphNode_StateMachine_2.LowFidelityStateMachine.AnimStateNode_1.LowFidelity_Idle.AnimGraphNode_StateMachine_2.DefaultIdleStateMachine.AnimStateNode_0.DefaultIdle.AnimGraphNode_StateMachine_1.Idle.AnimStateNode_3.SitToStand
  nodes [Output Animation Pose(Result=Sequence Player 'SitPlant_COR_Sit_Out'() {bLoopAnimation=false, sequence=/CosmeticCompanions/Assets/Quadruped/SitPlant/Animation/Core/SitPlant_COR_Sit_Out}) {
    node {stateEntryFunction {functionName=OnExitSitState}}
    stateEntryFunction {bSelfContext=true, memberGuid=#2, memberName=OnExitSitState}
   }]
 }
 AnimationTransitionGraph_0 {
  path=AnimGraph.AnimGraphNode_StateMachine_2.LowFidelityStateMachine.AnimStateNode_1.LowFidelity_Idle.AnimGraphNode_StateMachine_2.DefaultIdleStateMachine.AnimStateNode_0.DefaultIdle.AnimGraphNode_StateMachine_1.Idle.AnimStateTransitionNode_0.AnimationTransitionGraph_0
  nodes [Result()]
 }
 AnimationTransitionGraph_0 {
  path=AnimGraph.AnimGraphNode_StateMachine_2.LowFidelityStateMachine.AnimStateNode_1.LowFidelity_Idle.AnimGraphNode_StateMachine_2.DefaultIdleStateMachine.AnimStateNode_0.DefaultIdle.AnimGraphNode_StateMachine_1.Idle.AnimStateTransitionNode_1.AnimationTransitionGraph_0
  nodes [Result(bCanEnterTransition=Get IsStandToSit() -> (IsStandToSit:bool))]
 }
 AnimationTransitionGraph_0 {
  path=AnimGraph.AnimGraphNode_StateMachine_2.LowFidelityStateMachine.AnimStateNode_1.LowFidelity_Idle.AnimGraphNode_StateMachine_2.DefaultIdleStateMachine.AnimStateNode_0.DefaultIdle.AnimGraphNode_StateMachine_1.Idle.AnimStateTransitionNode_2.AnimationTransitionGraph_0
  nodes [Result(bCanEnterTransition=true)]
 }
 AnimationTransitionGraph_0 {
  path=AnimGraph.AnimGraphNode_StateMachine_2.LowFidelityStateMachine.AnimStateNode_1.LowFidelity_Idle.AnimGraphNode_StateMachine_2.DefaultIdleStateMachine.AnimStateNode_0.DefaultIdle.AnimGraphNode_StateMachine_1.Idle.AnimStateTransitionNode_4.AnimationTransitionGraph_0
  nodes [Result(bCanEnterTransition=true)]
 }
 AnimationTransitionGraph_0 {
  path=AnimGraph.AnimGraphNode_StateMachine_2.LowFidelityStateMachine.AnimStateNode_1.LowFidelity_Idle.AnimGraphNode_StateMachine_2.DefaultIdleStateMachine.AnimStateNode_0.DefaultIdle.AnimGraphNode_StateMachine_1.Idle.AnimStateTransitionNode_6.AnimationTransitionGraph_0
  nodes [Result()]
 }
 AnimationTransitionGraph_0 {
  path=AnimGraph.AnimGraphNode_StateMachine_2.LowFidelityStateMachine.AnimStateNode_1.LowFidelity_Idle.AnimGraphNode_StateMachine_2.DefaultIdleStateMachine.AnimStateNode_0.DefaultIdle.AnimGraphNode_StateMachine_1.Idle.AnimStateTransitionNode_8.AnimationTransitionGraph_0
  nodes [Result(bCanEnterTransition=Get IsSitToStand() -> (IsSitToStand:bool))]
 }
 SpecialIdle {
  path=AnimGraph.AnimGraphNode_StateMachine_2.LowFidelityStateMachine.AnimStateNode_1.LowFidelity_Idle.AnimGraphNode_StateMachine_2.DefaultIdleStateMachine.AnimStateNode_1.SpecialIdle
  nodes [
   AnimGraphNode_ChooserPlayer_1 = Chooser Player: CHT_SitPlant_SpecialIdle!() -> (Pose:Pose Link Structure) {
    is=title AnimGraphNode_ChooserPlayer
    chooser {
     _structType=&EvaluateChooser
     chooser=/CosmeticCompanions/Assets/Quadruped/SitPlant/Animation/Choosers/CHT_SitPlant_SpecialIdle
    }
    chooserContextDefinition [
     {
      _structType=&ContextObjectTypeClass
      class=&SitPlant_AnimBP_C
      direction=ReadWrite
     }
     {struct=&ChooserPlayerSettings, _structType=&ContextObjectTypeStruct, direction=Read}
    ]
    defaultSettings {blendOption=Linear}
   }
   Output Animation Pose(Result=@AnimGraphNode_ChooserPlayer_1)
  ]
 }
 AnimationBlendStackGraph_0 {
  path=AnimGraph.AnimGraphNode_StateMachine_2.LowFidelityStateMachine.AnimStateNode_1.LowFidelity_Idle.AnimGraphNode_StateMachine_2.DefaultIdleStateMachine.AnimStateNode_1.SpecialIdle.AnimGraphNode_ChooserPlayer_1.AnimationBlendStackGraph_0
  nodes [Output Pose(Result=Blend Stack Input())]
 }
 SpecialIdleToDefaultIdle {
  path=AnimGraph.AnimGraphNode_StateMachine_2.LowFidelityStateMachine.AnimStateNode_1.LowFidelity_Idle.AnimGraphNode_StateMachine_2.DefaultIdleStateMachine.AnimStateNode_2.SpecialIdleToDefaultIdle
  nodes [
   AnimGraphNode_StateMachine_0 = TransitionToIdle() {is=title AnimGraphNode_StateMachine, bAllowConduitEntryStates=true}
   Output Animation Pose(Result=@AnimGraphNode_StateMachine_0)
  ]
 }
 TransitionToIdle {
  entry=Conduit
  path=AnimGraph.AnimGraphNode_StateMachine_2.LowFidelityStateMachine.AnimStateNode_1.LowFidelity_Idle.AnimGraphNode_StateMachine_2.DefaultIdleStateMachine.AnimStateNode_2.SpecialIdleToDefaultIdle.AnimGraphNode_StateMachine_0.TransitionToIdle
  ids=id
  nodes [
   AnimStateConduitNode_0 = Conduit()
   AnimStateNode_0 = Combat_To_Idle() {is=title AnimStateNode}
   AnimStateNode_1 = Happy_01_To_Idle() {is=title AnimStateNode}
   AnimStateNode_2 = EmoteDance_To_Idle() {is=title AnimStateNode}
   AnimStateNode_3 = EmoteNonDance_To_Idle() {is=title AnimStateNode}
   AnimStateNode_4 = Happy02_To_Idle() {is=title AnimStateNode}
   AnimStateTransitionNode_0 = AnimStateTransitionNode()
   AnimStateTransitionNode_1 = AnimStateTransitionNode() {priorityOrder=4}
   AnimStateTransitionNode_2 = AnimStateTransitionNode() {logicType=TLT_Inertialization, priorityOrder=3}
   AnimStateTransitionNode_3 = AnimStateTransitionNode() {priorityOrder=2}
   AnimStateTransitionNode_4 = AnimStateTransitionNode() {priorityOrder=100}
  ]
  links [
   {
    dst=Combat_To_Idle
    rule=AnimGraph.AnimGraphNode_StateMachine_2.LowFidelityStateMachine.AnimStateNode_1.LowFidelity_Idle.AnimGraphNode_StateMachine_2.DefaultIdleStateMachine.AnimStateNode_2.SpecialIdleToDefaultIdle.AnimGraphNode_StateMachine_0.TransitionToIdle.AnimStateTransitionNode_0.Transition
    src=Conduit
    via=AnimStateTransitionNode_0
   }
   {
    dst=EmoteDance_To_Idle
    rule=AnimGraph.AnimGraphNode_StateMachine_2.LowFidelityStateMachine.AnimStateNode_1.LowFidelity_Idle.AnimGraphNode_StateMachine_2.DefaultIdleStateMachine.AnimStateNode_2.SpecialIdleToDefaultIdle.AnimGraphNode_StateMachine_0.TransitionToIdle.AnimStateTransitionNode_3.Transition
    src=Conduit
    via=AnimStateTransitionNode_3
   }
   {
    dst=EmoteNonDance_To_Idle
    rule=AnimGraph.AnimGraphNode_StateMachine_2.LowFidelityStateMachine.AnimStateNode_1.LowFidelity_Idle.AnimGraphNode_StateMachine_2.DefaultIdleStateMachine.AnimStateNode_2.SpecialIdleToDefaultIdle.AnimGraphNode_StateMachine_0.TransitionToIdle.AnimStateTransitionNode_2.Transition
    src=Conduit
    via=AnimStateTransitionNode_2
   }
   {
    dst=Happy02_To_Idle
    rule=AnimGraph.AnimGraphNode_StateMachine_2.LowFidelityStateMachine.AnimStateNode_1.LowFidelity_Idle.AnimGraphNode_StateMachine_2.DefaultIdleStateMachine.AnimStateNode_2.SpecialIdleToDefaultIdle.AnimGraphNode_StateMachine_0.TransitionToIdle.AnimStateTransitionNode_4.Transition
    src=Conduit
    via=AnimStateTransitionNode_4
   }
   {
    dst=Happy_01_To_Idle
    rule=AnimGraph.AnimGraphNode_StateMachine_2.LowFidelityStateMachine.AnimStateNode_1.LowFidelity_Idle.AnimGraphNode_StateMachine_2.DefaultIdleStateMachine.AnimStateNode_2.SpecialIdleToDefaultIdle.AnimGraphNode_StateMachine_0.TransitionToIdle.AnimStateTransitionNode_1.Transition
    src=Conduit
    via=AnimStateTransitionNode_1
   }
  ]
 }
 Conduit {
  path=AnimGraph.AnimGraphNode_StateMachine_2.LowFidelityStateMachine.AnimStateNode_1.LowFidelity_Idle.AnimGraphNode_StateMachine_2.DefaultIdleStateMachine.AnimStateNode_2.SpecialIdleToDefaultIdle.AnimGraphNode_StateMachine_0.TransitionToIdle.AnimStateConduitNode_0.Conduit
  nodes [Result(bCanEnterTransition=true)]
 }
 Combat_To_Idle {
  path=AnimGraph.AnimGraphNode_StateMachine_2.LowFidelityStateMachine.AnimStateNode_1.LowFidelity_Idle.AnimGraphNode_StateMachine_2.DefaultIdleStateMachine.AnimStateNode_2.SpecialIdleToDefaultIdle.AnimGraphNode_StateMachine_0.TransitionToIdle.AnimStateNode_0.Combat_To_Idle
  nodes [Output Animation Pose(Result=Sequence Player 'SitPlant_COR_Angry_Idle_Out'() {bLoopAnimation=false, sequence=/CosmeticCompanions/Assets/Quadruped/SitPlant/Animation/Core/SitPlant_COR_Angry_Idle_Out})]
 }
 Happy_01_To_Idle {
  path=AnimGraph.AnimGraphNode_StateMachine_2.LowFidelityStateMachine.AnimStateNode_1.LowFidelity_Idle.AnimGraphNode_StateMachine_2.DefaultIdleStateMachine.AnimStateNode_2.SpecialIdleToDefaultIdle.AnimGraphNode_StateMachine_0.TransitionToIdle.AnimStateNode_1.Happy_01_To_Idle
  nodes [Output Animation Pose(Result=Sequence Player 'SitPlant_RCT_Happy_01_Out'() {bLoopAnimation=false, sequence=/CosmeticCompanions/Assets/Quadruped/SitPlant/Animation/Reacts/SitPlant_RCT_Happy_01_Out})]
 }
 EmoteDance_To_Idle {
  path=AnimGraph.AnimGraphNode_StateMachine_2.LowFidelityStateMachine.AnimStateNode_1.LowFidelity_Idle.AnimGraphNode_StateMachine_2.DefaultIdleStateMachine.AnimStateNode_2.SpecialIdleToDefaultIdle.AnimGraphNode_StateMachine_0.TransitionToIdle.AnimStateNode_2.EmoteDance_To_Idle
  nodes [Output Animation Pose(Result=Sequence Player 'SitPlant_RCT_Dance_Out'() {bLoopAnimation=false, sequence=/CosmeticCompanions/Assets/Quadruped/SitPlant/Animation/Reacts/SitPlant_RCT_Dance_Out})]
 }
 EmoteNonDance_To_Idle {
  path=AnimGraph.AnimGraphNode_StateMachine_2.LowFidelityStateMachine.AnimStateNode_1.LowFidelity_Idle.AnimGraphNode_StateMachine_2.DefaultIdleStateMachine.AnimStateNode_2.SpecialIdleToDefaultIdle.AnimGraphNode_StateMachine_0.TransitionToIdle.AnimStateNode_3.EmoteNonDance_To_Idle
  nodes [Output Animation Pose(Result=Sequence Player 'SitPlant_RCT_NonDance_Out'() {bLoopAnimation=false, sequence=/CosmeticCompanions/Assets/Quadruped/SitPlant/Animation/Reacts/SitPlant_RCT_NonDance_Out})]
 }
 Happy02_To_Idle {
  path=AnimGraph.AnimGraphNode_StateMachine_2.LowFidelityStateMachine.AnimStateNode_1.LowFidelity_Idle.AnimGraphNode_StateMachine_2.DefaultIdleStateMachine.AnimStateNode_2.SpecialIdleToDefaultIdle.AnimGraphNode_StateMachine_0.TransitionToIdle.AnimStateNode_4.Happy02_To_Idle
  nodes [Output Animation Pose(Result=Sequence Player 'SitPlant_RCT_Happy_02_Out'() {bLoopAnimation=false, sequence=/CosmeticCompanions/Assets/Quadruped/SitPlant/Animation/Reacts/SitPlant_RCT_Happy_02_Out})]
 }
 Transition {
  path=AnimGraph.AnimGraphNode_StateMachine_2.LowFidelityStateMachine.AnimStateNode_1.LowFidelity_Idle.AnimGraphNode_StateMachine_2.DefaultIdleStateMachine.AnimStateNode_2.SpecialIdleToDefaultIdle.AnimGraphNode_StateMachine_0.TransitionToIdle.AnimStateTransitionNode_0.Transition
  nodes [Result(bCanEnterTransition=Get IsCombatToIdle() -> (IsCombatToIdle:bool))]
 }
 Transition {
  path=AnimGraph.AnimGraphNode_StateMachine_2.LowFidelityStateMachine.AnimStateNode_1.LowFidelity_Idle.AnimGraphNode_StateMachine_2.DefaultIdleStateMachine.AnimStateNode_2.SpecialIdleToDefaultIdle.AnimGraphNode_StateMachine_0.TransitionToIdle.AnimStateTransitionNode_1.Transition
  nodes [Result(bCanEnterTransition=Get IsReactionHappyCurve() -> (IsReactionHappyCurve:bool))]
 }
 Transition {
  path=AnimGraph.AnimGraphNode_StateMachine_2.LowFidelityStateMachine.AnimStateNode_1.LowFidelity_Idle.AnimGraphNode_StateMachine_2.DefaultIdleStateMachine.AnimStateNode_2.SpecialIdleToDefaultIdle.AnimGraphNode_StateMachine_0.TransitionToIdle.AnimStateTransitionNode_2.Transition
  nodes [Result(bCanEnterTransition=Get IsEmoteNonDanceToIdle() -> (IsEmoteNonDanceToIdle:bool))]
 }
 Transition {
  path=AnimGraph.AnimGraphNode_StateMachine_2.LowFidelityStateMachine.AnimStateNode_1.LowFidelity_Idle.AnimGraphNode_StateMachine_2.DefaultIdleStateMachine.AnimStateNode_2.SpecialIdleToDefaultIdle.AnimGraphNode_StateMachine_0.TransitionToIdle.AnimStateTransitionNode_3.Transition
  nodes [Result(bCanEnterTransition=Get IsEmoteDanceToIdle() -> (IsEmoteDanceToIdle:bool))]
 }
 Transition {
  path=AnimGraph.AnimGraphNode_StateMachine_2.LowFidelityStateMachine.AnimStateNode_1.LowFidelity_Idle.AnimGraphNode_StateMachine_2.DefaultIdleStateMachine.AnimStateNode_2.SpecialIdleToDefaultIdle.AnimGraphNode_StateMachine_0.TransitionToIdle.AnimStateTransitionNode_4.Transition
  nodes [Result(bCanEnterTransition=true)]
 }
 Transition {
  path=AnimGraph.AnimGraphNode_StateMachine_2.LowFidelityStateMachine.AnimStateNode_1.LowFidelity_Idle.AnimGraphNode_StateMachine_2.DefaultIdleStateMachine.AnimStateTransitionNode_0.Transition
  nodes [Result(bCanEnterTransition=true)]
 }
 Transition {
  path=AnimGraph.AnimGraphNode_StateMachine_2.LowFidelityStateMachine.AnimStateNode_1.LowFidelity_Idle.AnimGraphNode_StateMachine_2.DefaultIdleStateMachine.AnimStateTransitionNode_1.Transition
  nodes [Result(bCanEnterTransition=Get IsSpecialIdleEnter() -> (IsSpecialIdleEnter:bool))]
 }
 Transition {
  path=AnimGraph.AnimGraphNode_StateMachine_2.LowFidelityStateMachine.AnimStateNode_1.LowFidelity_Idle.AnimGraphNode_StateMachine_2.DefaultIdleStateMachine.AnimStateTransitionNode_2.Transition
  nodes [Result(bCanEnterTransition=true)]
 }
 Transition {
  path=AnimGraph.AnimGraphNode_StateMachine_2.LowFidelityStateMachine.AnimStateNode_1.LowFidelity_Idle.AnimGraphNode_StateMachine_2.DefaultIdleStateMachine.AnimStateTransitionNode_3.Transition
  nodes [Result(bCanEnterTransition=Get IsSpecialIdleExit() -> (IsSpecialIdleExit:bool))]
 }
 Transition {
  path=AnimGraph.AnimGraphNode_StateMachine_2.LowFidelityStateMachine.AnimStateNode_1.LowFidelity_Idle.AnimGraphNode_StateMachine_2.DefaultIdleStateMachine.AnimStateTransitionNode_4.Transition
  nodes [Result()]
 }
 Transition {
  path=AnimGraph.AnimGraphNode_StateMachine_2.LowFidelityStateMachine.AnimStateNode_1.LowFidelity_Idle.AnimGraphNode_StateMachine_2.DefaultIdleStateMachine.AnimStateTransitionNode_5.Transition
  nodes [Result(bCanEnterTransition=Get IsReactionHappyPlaying() -> (IsReactionHappyPlaying:bool))]
 }
 LowFidelity_Loco {
  path=AnimGraph.AnimGraphNode_StateMachine_2.LowFidelityStateMachine.AnimStateNode_2.LowFidelity_Loco
  nodes [
   n1 = Sequence Player 'SitPlant_COR_Loco_Walk'() {sequence=/CosmeticCompanions/Assets/Quadruped/SitPlant/Animation/Core/SitPlant_COR_Loco_Walk}
   n2 = Sequence Player 'SitPlant_COR_Loco_Run'!(PlayRate:float=1) -> (Pose) {groupName=Locomotion, method=SyncGroup, sequence=&SitPlant_COR_Loco_Run}
   n3 = Blend Poses by bool(
    BlendPose_0=@n2
    BlendPose_1=Blend Poses by bool(
     BlendPose_0=Sequence Player 'SitPlant_COR_Loco_Trot'() {sequence=/CosmeticCompanions/Assets/Quadruped/SitPlant/Animation/Core/SitPlant_COR_Loco_Trot}
     BlendPose_1=Modify Curve(SourcePose=@n1, CurveValues_0=1, Alpha=1) {curveNames [Walk], curveValues [0]}
     BlendTime_1=0.2
     bActiveValue=Get IsLocoTrot() -> (IsLocoTrot:bool)
    ) {blendType=Sinusoidal}
    BlendTime_0=0.2
    BlendTime_1=0.4
    bActiveValue=Get IsLocoRun() -> (IsLocoRun:bool)
   ) {blendType=Sinusoidal}
   n4 = Sequence Player 'SitPlant_COR_Loco_Sprint'!(PlayRate:float=1) -> (Pose) {groupName=Locomotion, method=SyncGroup, sequence=/CosmeticCompanions/Assets/Quadruped/SitPlant/Animation/Core/SitPlant_COR_Loco_Sprint}
   n5 = Get IsSprinting() -> (IsSprinting:bool)
   Output Animation Pose(Result=Modify Curve(
     SourcePose=Blend Poses by bool(BlendPose_0=@n4, BlendPose_1=@n3, BlendTime_0=0.2, BlendTime_1=0.2, bActiveValue=@n5) {blendType=Sinusoidal}
     CurveValues_0=1
     Alpha=1
    ) {curveNames [Locomotion], curveValues [0]})
  ]
 }
 Transition {
  path=AnimGraph.AnimGraphNode_StateMachine_2.LowFidelityStateMachine.AnimStateTransitionNode_0.Transition
  nodes [Result(bCanEnterTransition=true)]
 }
 Transition {
  path=AnimGraph.AnimGraphNode_StateMachine_2.LowFidelityStateMachine.AnimStateTransitionNode_1.Transition
  nodes [Result(bCanEnterTransition=Get IsMoving() -> (IsMoving:bool))]
 }
 Transition {
  path=AnimGraph.AnimGraphNode_StateMachine_2.LowFidelityStateMachine.AnimStateTransitionNode_2.Transition
  nodes [Result(bCanEnterTransition=true)]
 }
 Transition {
  path=AnimGraph.AnimGraphNode_StateMachine_2.LowFidelityStateMachine.AnimStateTransitionNode_3.Transition
  nodes [Result(bCanEnterTransition=true)]
 }
 Locomotion {
  entry=Conduit
  path=AnimGraph.AnimGraphNode_StateMachine_3.Locomotion
  ids=id
  nodes [
   AnimStateConduitNode_0 = Conduit()
   AnimStateNode_0 = Idle() {is=title AnimStateNode, bAlwaysResetOnEntry=true}
   AnimStateNode_2 = LocomotionStart() {is=title AnimStateNode, bAlwaysResetOnEntry=true}
   AnimStateNode_3 = Locomotion() {is=title AnimStateNode, bAlwaysResetOnEntry=true}
   AnimStateNode_4 = LocomotionStop() {is=title AnimStateNode, bAlwaysResetOnEntry=true}
   AnimStateTransitionNode_0 = AnimStateTransitionNode() {crossfadeDuration=0.4, logicType=TLT_Inertialization, priorityOrder=100}
   AnimStateTransitionNode_1 = AnimStateTransitionNode() {
    bDisabled=true
    logicType=TLT_Inertialization
    priorityOrder=2
    blendProfileWrapper {blendProfile=&SitPlant_Quadruped_Skeleton:TimeBlend_FastFeetSlowBody}
   }
   AnimStateTransitionNode_10 = AnimStateTransitionNode()
   AnimStateTransitionNode_11 = AnimStateTransitionNode() {logicType=TLT_Inertialization, priorityOrder=3}
   AnimStateTransitionNode_12 = AnimStateTransitionNode() {logicType=TLT_Inertialization, priorityOrder=100}
   AnimStateTransitionNode_2 = AnimStateTransitionNode()
   AnimStateTransitionNode_3 = AnimStateTransitionNode() {bAutomaticRuleBasedOnSequencePlayerInState=true, logicType=TLT_Inertialization, priorityOrder=100}
   AnimStateTransitionNode_5 = AnimStateTransitionNode() {bAutomaticRuleBasedOnSequencePlayerInState=true, logicType=TLT_Inertialization}
   AnimStateTransitionNode_8 = AnimStateTransitionNode() {crossfadeDuration=0.1, logicType=TLT_Inertialization}
   AnimStateTransitionNode_9 = AnimStateTransitionNode()
  ]
  links [
   {
    dst=Idle
    rule=AnimGraph.AnimGraphNode_StateMachine_3.Locomotion.AnimStateTransitionNode_0.Transition
    src=Conduit
    via=AnimStateTransitionNode_0
   }
   {
    dst=Locomotion
    rule=AnimGraph.AnimGraphNode_StateMachine_3.Locomotion.AnimStateTransitionNode_8.Transition
    src=Conduit
    via=AnimStateTransitionNode_8
   }
   {
    dst=Conduit
    rule=AnimGraph.AnimGraphNode_StateMachine_3.Locomotion.AnimStateTransitionNode_2.Transition
    src=Idle
    via=AnimStateTransitionNode_2
   }
   {
    dst=Idle
    rule=AnimGraph.AnimGraphNode_StateMachine_3.Locomotion.AnimStateTransitionNode_10.Transition
    src=Locomotion
    via=AnimStateTransitionNode_10
   }
   {
    dst=LocomotionStop
    rule=AnimGraph.AnimGraphNode_StateMachine_3.Locomotion.AnimStateTransitionNode_1.Transition
    src=Locomotion
    via=AnimStateTransitionNode_1
   }
   {
    dst=Locomotion
    rule=AnimGraph.AnimGraphNode_StateMachine_3.Locomotion.AnimStateTransitionNode_5.Transition
    src=LocomotionStart
    via=AnimStateTransitionNode_5
   }
   {
    dst=LocomotionStop
    rule=AnimGraph.AnimGraphNode_StateMachine_3.Locomotion.AnimStateTransitionNode_12.Transition
    src=LocomotionStart
    via=AnimStateTransitionNode_12
   }
   {
    dst=Idle
    rule=AnimGraph.AnimGraphNode_StateMachine_3.Locomotion.AnimStateTransitionNode_11.Transition
    src=LocomotionStop
    via=AnimStateTransitionNode_11
   }
   {
    dst=Idle
    rule=AnimGraph.AnimGraphNode_StateMachine_3.Locomotion.AnimStateTransitionNode_3.Transition
    src=LocomotionStop
    via=AnimStateTransitionNode_3
   }
   {
    dst=Locomotion
    rule=AnimGraph.AnimGraphNode_StateMachine_3.Locomotion.AnimStateTransitionNode_9.Transition
    src=LocomotionStop
    via=AnimStateTransitionNode_9
   }
  ]
 }
 Conduit {path=AnimGraph.AnimGraphNode_StateMachine_3.Locomotion.AnimStateConduitNode_0.Conduit, nodes [Result(bCanEnterTransition=true)]}
 Idle {
  path=AnimGraph.AnimGraphNode_StateMachine_3.Locomotion.AnimStateNode_0.Idle
  nodes [
   AnimGraphNode_StateMachine_2 = DefaultIdleStateMachine() {bAllowConduitEntryStates=true}
   Output Animation Pose(Result=Modify Curve(SourcePose=@AnimGraphNode_StateMachine_2, CurveValues_0=1, Alpha=1) {curveNames [idle], curveValues [0]})
  ]
 }
 DefaultIdleStateMachine {
  entry=Conduit
  path=AnimGraph.AnimGraphNode_StateMachine_3.Locomotion.AnimStateNode_0.Idle.AnimGraphNode_StateMachine_2.DefaultIdleStateMachine
  ids=id
  nodes [
   AnimStateConduitNode_0 = Conduit()
   AnimStateNode_0 = DefaultIdle()
   AnimStateNode_1 = SpecialIdle()
   AnimStateNode_2 = SpecialIdleToDefaultIdle()
   AnimStateTransitionNode_0 = AnimStateTransitionNode() {logicType=TLT_Inertialization, priorityOrder=100}
   AnimStateTransitionNode_1 = AnimStateTransitionNode() {logicType=TLT_Inertialization}
   AnimStateTransitionNode_2 = AnimStateTransitionNode()
   AnimStateTransitionNode_3 = AnimStateTransitionNode() {crossfadeDuration=0.3, logicType=TLT_Inertialization}
   AnimStateTransitionNode_4 = AnimStateTransitionNode() {bAutomaticRuleBasedOnSequencePlayerInState=true, logicType=TLT_Inertialization}
   AnimStateTransitionNode_5 = AnimStateTransitionNode() {logicType=TLT_Inertialization, priorityOrder=2}
  ]
  links [
   {
    dst=DefaultIdle
    rule=AnimGraph.AnimGraphNode_StateMachine_3.Locomotion.AnimStateNode_0.Idle.AnimGraphNode_StateMachine_2.DefaultIdleStateMachine.AnimStateTransitionNode_0.Transition
    src=Conduit
    via=AnimStateTransitionNode_0
   }
   {
    dst=SpecialIdle
    rule=AnimGraph.AnimGraphNode_StateMachine_3.Locomotion.AnimStateNode_0.Idle.AnimGraphNode_StateMachine_2.DefaultIdleStateMachine.AnimStateTransitionNode_1.Transition
    src=Conduit
    via=AnimStateTransitionNode_1
   }
   {
    dst=SpecialIdleToDefaultIdle
    rule=AnimGraph.AnimGraphNode_StateMachine_3.Locomotion.AnimStateNode_0.Idle.AnimGraphNode_StateMachine_2.DefaultIdleStateMachine.AnimStateTransitionNode_5.Transition
    src=Conduit
    via=AnimStateTransitionNode_5
   }
   {
    dst=Conduit
    rule=AnimGraph.AnimGraphNode_StateMachine_3.Locomotion.AnimStateNode_0.Idle.AnimGraphNode_StateMachine_2.DefaultIdleStateMachine.AnimStateTransitionNode_2.Transition
    src=DefaultIdle
    via=AnimStateTransitionNode_2
   }
   {
    dst=SpecialIdleToDefaultIdle
    rule=AnimGraph.AnimGraphNode_StateMachine_3.Locomotion.AnimStateNode_0.Idle.AnimGraphNode_StateMachine_2.DefaultIdleStateMachine.AnimStateTransitionNode_3.Transition
    src=SpecialIdle
    via=AnimStateTransitionNode_3
   }
   {
    dst=DefaultIdle
    rule=AnimGraph.AnimGraphNode_StateMachine_3.Locomotion.AnimStateNode_0.Idle.AnimGraphNode_StateMachine_2.DefaultIdleStateMachine.AnimStateTransitionNode_4.Transition
    src=SpecialIdleToDefaultIdle
    via=AnimStateTransitionNode_4
   }
  ]
 }
 Conduit {
  path=AnimGraph.AnimGraphNode_StateMachine_3.Locomotion.AnimStateNode_0.Idle.AnimGraphNode_StateMachine_2.DefaultIdleStateMachine.AnimStateConduitNode_0.Conduit
  nodes [Result(bCanEnterTransition=true)]
 }
 DefaultIdle {
  path=AnimGraph.AnimGraphNode_StateMachine_3.Locomotion.AnimStateNode_0.Idle.AnimGraphNode_StateMachine_2.DefaultIdleStateMachine.AnimStateNode_0.DefaultIdle
  nodes [
   AnimGraphNode_StateMachine_1 = Idle() {is=title AnimGraphNode_StateMachine, bAllowConduitEntryStates=true}
   Output Animation Pose(Result=@AnimGraphNode_StateMachine_1) {
    node {stateExitFunction {functionName=OnExitSitState}}
    stateExitFunction {bSelfContext=true, memberGuid=#2, memberName=OnExit_SitState}
   }
  ]
 }
 Idle {
  entry=Conduit
  path=AnimGraph.AnimGraphNode_StateMachine_3.Locomotion.AnimStateNode_0.Idle.AnimGraphNode_StateMachine_2.DefaultIdleStateMachine.AnimStateNode_0.DefaultIdle.AnimGraphNode_StateMachine_1.Idle
  ids=id
  nodes [
   Alias() {also {aliases [SitPose]}}
   AnimStateConduitNode_0 = Conduit()
   AnimStateNode_0 = StandToSit()
   AnimStateNode_1 = SitPose()
   AnimStateNode_2 = StandingPose()
   AnimStateNode_3 = SitToStand()
   AnimStateTransitionNode_0 = AnimStateTransitionNode() {bAutomaticRuleBasedOnSequencePlayerInState=true, logicType=TLT_Inertialization}
   AnimStateTransitionNode_1 = AnimStateTransitionNode() {bDisabled=true, logicType=TLT_Inertialization}
   AnimStateTransitionNode_2 = AnimStateTransitionNode() {logicType=TLT_Inertialization, priorityOrder=100}
   AnimStateTransitionNode_3 = AnimStateTransitionNode() {crossfadeDuration=0.3, logicType=TLT_Inertialization}
   AnimStateTransitionNode_4 = AnimStateTransitionNode()
   AnimStateTransitionNode_5 = AnimStateTransitionNode() {crossfadeDuration=0.3, logicType=TLT_Inertialization}
   AnimStateTransitionNode_6 = AnimStateTransitionNode() {bAutomaticRuleBasedOnSequencePlayerInState=true, crossfadeDuration=0.3, logicType=TLT_Inertialization}
   AnimStateTransitionNode_8 = AnimStateTransitionNode() {bDisabled=true, logicType=TLT_Inertialization}
  ]
  links [
   {
    dst=SitToStand
    rule=AnimGraph.AnimGraphNode_StateMachine_3.Locomotion.AnimStateNode_0.Idle.AnimGraphNode_StateMachine_2.DefaultIdleStateMachine.AnimStateNode_0.DefaultIdle.AnimGraphNode_StateMachine_1.Idle.AnimStateTransitionNode_8.AnimationTransitionGraph_0
    src=Alias
    via=AnimStateTransitionNode_8
   }
   {
    dst=StandingPose
    rule=AnimGraph.AnimGraphNode_StateMachine_3.Locomotion.AnimStateNode_0.Idle.AnimGraphNode_StateMachine_2.DefaultIdleStateMachine.AnimStateNode_0.DefaultIdle.AnimGraphNode_StateMachine_1.Idle.AnimStateTransitionNode_5.Transition
    src=Alias
    via=AnimStateTransitionNode_5
   }
   {
    dst=SitPose
    rule=AnimGraph.AnimGraphNode_StateMachine_3.Locomotion.AnimStateNode_0.Idle.AnimGraphNode_StateMachine_2.DefaultIdleStateMachine.AnimStateNode_0.DefaultIdle.AnimGraphNode_StateMachine_1.Idle.AnimStateTransitionNode_3.Transition
    src=Conduit
    via=AnimStateTransitionNode_3
   }
   {
    dst=StandToSit
    rule=AnimGraph.AnimGraphNode_StateMachine_3.Locomotion.AnimStateNode_0.Idle.AnimGraphNode_StateMachine_2.DefaultIdleStateMachine.AnimStateNode_0.DefaultIdle.AnimGraphNode_StateMachine_1.Idle.AnimStateTransitionNode_1.AnimationTransitionGraph_0
    src=Conduit
    via=AnimStateTransitionNode_1
   }
   {
    dst=StandingPose
    rule=AnimGraph.AnimGraphNode_StateMachine_3.Locomotion.AnimStateNode_0.Idle.AnimGraphNode_StateMachine_2.DefaultIdleStateMachine.AnimStateNode_0.DefaultIdle.AnimGraphNode_StateMachine_1.Idle.AnimStateTransitionNode_2.AnimationTransitionGraph_0
    src=Conduit
    via=AnimStateTransitionNode_2
   }
   {
    dst=StandingPose
    rule=AnimGraph.AnimGraphNode_StateMachine_3.Locomotion.AnimStateNode_0.Idle.AnimGraphNode_StateMachine_2.DefaultIdleStateMachine.AnimStateNode_0.DefaultIdle.AnimGraphNode_StateMachine_1.Idle.AnimStateTransitionNode_6.AnimationTransitionGraph_0
    src=SitToStand
    via=AnimStateTransitionNode_6
   }
   {
    dst=SitPose
    rule=AnimGraph.AnimGraphNode_StateMachine_3.Locomotion.AnimStateNode_0.Idle.AnimGraphNode_StateMachine_2.DefaultIdleStateMachine.AnimStateNode_0.DefaultIdle.AnimGraphNode_StateMachine_1.Idle.AnimStateTransitionNode_0.AnimationTransitionGraph_0
    src=StandToSit
    via=AnimStateTransitionNode_0
   }
   {
    dst=Conduit
    rule=AnimGraph.AnimGraphNode_StateMachine_3.Locomotion.AnimStateNode_0.Idle.AnimGraphNode_StateMachine_2.DefaultIdleStateMachine.AnimStateNode_0.DefaultIdle.AnimGraphNode_StateMachine_1.Idle.AnimStateTransitionNode_4.AnimationTransitionGraph_0
    src=StandingPose
    via=AnimStateTransitionNode_4
   }
  ]
 }
 Conduit {
  path=AnimGraph.AnimGraphNode_StateMachine_3.Locomotion.AnimStateNode_0.Idle.AnimGraphNode_StateMachine_2.DefaultIdleStateMachine.AnimStateNode_0.DefaultIdle.AnimGraphNode_StateMachine_1.Idle.AnimStateConduitNode_0.Conduit
  nodes [Result(bCanEnterTransition=true)]
 }
 StandToSit {
  path=AnimGraph.AnimGraphNode_StateMachine_3.Locomotion.AnimStateNode_0.Idle.AnimGraphNode_StateMachine_2.DefaultIdleStateMachine.AnimStateNode_0.DefaultIdle.AnimGraphNode_StateMachine_1.Idle.AnimStateNode_0.StandToSit
  nodes [Output Animation Pose(Result=Sequence Player 'SitPlant_COR_Sit_In'() {bLoopAnimation=false, sequence=&SitPlant_COR_Sit_In}) {
    node {stateEntryFunction {functionName=OnEnterSitState}}
    stateEntryFunction {bSelfContext=true, memberGuid=#3, memberName=OnEnter_SitState}
   }]
 }
 SitPose {
  path=AnimGraph.AnimGraphNode_StateMachine_3.Locomotion.AnimStateNode_0.Idle.AnimGraphNode_StateMachine_2.DefaultIdleStateMachine.AnimStateNode_0.DefaultIdle.AnimGraphNode_StateMachine_1.Idle.AnimStateNode_1.SitPose
  nodes [
   "Sequence Evaluator (by time) 'SitPlant_COR_Sit_In'"(ExplicitTime=1.06) {sequence=&SitPlant_COR_Sit_In}
   Output Animation Pose(Result=Sequence Player 'WideSnore_COR_Sit_Pose'() {bLoopAnimation=false, sequence=/CosmeticCompanions/Assets/Quadruped/WideSnore/Animation/Core/WideSnore_COR_Sit_Pose})
  ]
 }
 StandingPose {
  path=AnimGraph.AnimGraphNode_StateMachine_3.Locomotion.AnimStateNode_0.Idle.AnimGraphNode_StateMachine_2.DefaultIdleStateMachine.AnimStateNode_0.DefaultIdle.AnimGraphNode_StateMachine_1.Idle.AnimStateNode_2.StandingPose
  nodes [
   Blend Poses by bool(
    BlendPose_0=Sequence Player 'SitPlant_xDEV_React_Angry_02_Loop_OLD'() {sequence=&SitPlant_xDEV_React_Angry_02_Loop_OLD}
    BlendTime_0=0.3
    BlendTime_1=0.4
    bActiveValue=Get IsMoodCombat()
   ) {blendType=Linear, transitionType=Inertialization}
   Output Animation Pose(Result=Sequence Player 'WideSnore_COR_Stand_Pose'() {bLoopAnimation=false, sequence=&WideSnore_COR_Stand_Pose})
  ]
 }
 SitToStand {
  path=AnimGraph.AnimGraphNode_StateMachine_3.Locomotion.AnimStateNode_0.Idle.AnimGraphNode_StateMachine_2.DefaultIdleStateMachine.AnimStateNode_0.DefaultIdle.AnimGraphNode_StateMachine_1.Idle.AnimStateNode_3.SitToStand
  nodes [Output Animation Pose(Result=Sequence Player 'SitPlant_COR_Sit_Out'() {bLoopAnimation=false, sequence=&SitPlant_COR_Sit_Out}) {
    node {stateEntryFunction {functionName=OnExitSitState}}
    stateEntryFunction {bSelfContext=true, memberGuid=#2, memberName=OnExit_SitState}
   }]
 }
 AnimationTransitionGraph_0 {
  path=AnimGraph.AnimGraphNode_StateMachine_3.Locomotion.AnimStateNode_0.Idle.AnimGraphNode_StateMachine_2.DefaultIdleStateMachine.AnimStateNode_0.DefaultIdle.AnimGraphNode_StateMachine_1.Idle.AnimStateTransitionNode_0.AnimationTransitionGraph_0
  nodes [Result()]
 }
 AnimationTransitionGraph_0 {
  path=AnimGraph.AnimGraphNode_StateMachine_3.Locomotion.AnimStateNode_0.Idle.AnimGraphNode_StateMachine_2.DefaultIdleStateMachine.AnimStateNode_0.DefaultIdle.AnimGraphNode_StateMachine_1.Idle.AnimStateTransitionNode_1.AnimationTransitionGraph_0
  nodes [Result(bCanEnterTransition=Get IsStandToSit())]
 }
 AnimationTransitionGraph_0 {
  path=AnimGraph.AnimGraphNode_StateMachine_3.Locomotion.AnimStateNode_0.Idle.AnimGraphNode_StateMachine_2.DefaultIdleStateMachine.AnimStateNode_0.DefaultIdle.AnimGraphNode_StateMachine_1.Idle.AnimStateTransitionNode_2.AnimationTransitionGraph_0
  nodes [Result(bCanEnterTransition=true)]
 }
 Transition {
  path=AnimGraph.AnimGraphNode_StateMachine_3.Locomotion.AnimStateNode_0.Idle.AnimGraphNode_StateMachine_2.DefaultIdleStateMachine.AnimStateNode_0.DefaultIdle.AnimGraphNode_StateMachine_1.Idle.AnimStateTransitionNode_3.Transition
  nodes [Result(bCanEnterTransition=Get IsStandToSit())]
 }
 AnimationTransitionGraph_0 {
  path=AnimGraph.AnimGraphNode_StateMachine_3.Locomotion.AnimStateNode_0.Idle.AnimGraphNode_StateMachine_2.DefaultIdleStateMachine.AnimStateNode_0.DefaultIdle.AnimGraphNode_StateMachine_1.Idle.AnimStateTransitionNode_4.AnimationTransitionGraph_0
  nodes [Result(bCanEnterTransition=true)]
 }
 Transition {
  path=AnimGraph.AnimGraphNode_StateMachine_3.Locomotion.AnimStateNode_0.Idle.AnimGraphNode_StateMachine_2.DefaultIdleStateMachine.AnimStateNode_0.DefaultIdle.AnimGraphNode_StateMachine_1.Idle.AnimStateTransitionNode_5.Transition
  nodes [Result(bCanEnterTransition=Get IsSitToStand())]
 }
 AnimationTransitionGraph_0 {
  path=AnimGraph.AnimGraphNode_StateMachine_3.Locomotion.AnimStateNode_0.Idle.AnimGraphNode_StateMachine_2.DefaultIdleStateMachine.AnimStateNode_0.DefaultIdle.AnimGraphNode_StateMachine_1.Idle.AnimStateTransitionNode_6.AnimationTransitionGraph_0
  nodes [Result()]
 }
 AnimationTransitionGraph_0 {
  path=AnimGraph.AnimGraphNode_StateMachine_3.Locomotion.AnimStateNode_0.Idle.AnimGraphNode_StateMachine_2.DefaultIdleStateMachine.AnimStateNode_0.DefaultIdle.AnimGraphNode_StateMachine_1.Idle.AnimStateTransitionNode_8.AnimationTransitionGraph_0
  nodes [Result(bCanEnterTransition=Get IsSitToStand())]
 }
 SpecialIdle {
  path=AnimGraph.AnimGraphNode_StateMachine_3.Locomotion.AnimStateNode_0.Idle.AnimGraphNode_StateMachine_2.DefaultIdleStateMachine.AnimStateNode_1.SpecialIdle
  nodes [
   AnimGraphNode_ChooserPlayer_1 = Chooser Player: CHT_WideSnore_SpecialIdle!() -> (Pose:Pose Link Structure) {
    is=title AnimGraphNode_ChooserPlayer
    chooser {
     _structType=&EvaluateChooser
     chooser=/CosmeticCompanions/Assets/Quadruped/WideSnore/Animation/Choosers/CHT_WideSnore_SpecialIdle
    }
    chooserContextDefinition [
     {
      _structType=&ContextObjectTypeClass
      class=&SitPlant_AnimBP_C
      direction=ReadWrite
     }
     {struct=&ChooserPlayerSettings, _structType=&ContextObjectTypeStruct, direction=Read}
    ]
    defaultSettings {blendOption=Linear}
   }
   Output Animation Pose(Result=@AnimGraphNode_ChooserPlayer_1)
  ]
 }
 AnimationBlendStackGraph_0 {
  path=AnimGraph.AnimGraphNode_StateMachine_3.Locomotion.AnimStateNode_0.Idle.AnimGraphNode_StateMachine_2.DefaultIdleStateMachine.AnimStateNode_1.SpecialIdle.AnimGraphNode_ChooserPlayer_1.AnimationBlendStackGraph_0
  nodes [Output Pose(Result=Blend Stack Input())]
 }
 SpecialIdleToDefaultIdle {
  path=AnimGraph.AnimGraphNode_StateMachine_3.Locomotion.AnimStateNode_0.Idle.AnimGraphNode_StateMachine_2.DefaultIdleStateMachine.AnimStateNode_2.SpecialIdleToDefaultIdle
  nodes [
   AnimGraphNode_StateMachine_0 = TransitionToIdle() {bAllowConduitEntryStates=true}
   Output Animation Pose(Result=@AnimGraphNode_StateMachine_0)
  ]
 }
 TransitionToIdle {
  entry=Conduit
  path=AnimGraph.AnimGraphNode_StateMachine_3.Locomotion.AnimStateNode_0.Idle.AnimGraphNode_StateMachine_2.DefaultIdleStateMachine.AnimStateNode_2.SpecialIdleToDefaultIdle.AnimGraphNode_StateMachine_0.TransitionToIdle
  ids=id
  nodes [
   AnimStateConduitNode_0 = Conduit()
   AnimStateNode_0 = Combat_To_Idle()
   AnimStateNode_1 = Happy_01_To_Idle()
   AnimStateNode_2 = EmoteDance_To_Idle()
   AnimStateNode_3 = EmoteNonDance_To_Idle()
   AnimStateNode_4 = Happy02_To_Idle()
   AnimStateTransitionNode_0 = AnimStateTransitionNode()
   AnimStateTransitionNode_1 = AnimStateTransitionNode() {priorityOrder=4}
   AnimStateTransitionNode_2 = AnimStateTransitionNode() {logicType=TLT_Inertialization, priorityOrder=3}
   AnimStateTransitionNode_3 = AnimStateTransitionNode() {priorityOrder=2}
   AnimStateTransitionNode_4 = AnimStateTransitionNode() {priorityOrder=100}
  ]
  links [
   {
    dst=Combat_To_Idle
    rule=AnimGraph.AnimGraphNode_StateMachine_3.Locomotion.AnimStateNode_0.Idle.AnimGraphNode_StateMachine_2.DefaultIdleStateMachine.AnimStateNode_2.SpecialIdleToDefaultIdle.AnimGraphNode_StateMachine_0.TransitionToIdle.AnimStateTransitionNode_0.Transition
    src=Conduit
    via=AnimStateTransitionNode_0
   }
   {
    dst=EmoteDance_To_Idle
    rule=AnimGraph.AnimGraphNode_StateMachine_3.Locomotion.AnimStateNode_0.Idle.AnimGraphNode_StateMachine_2.DefaultIdleStateMachine.AnimStateNode_2.SpecialIdleToDefaultIdle.AnimGraphNode_StateMachine_0.TransitionToIdle.AnimStateTransitionNode_3.Transition
    src=Conduit
    via=AnimStateTransitionNode_3
   }
   {
    dst=EmoteNonDance_To_Idle
    rule=AnimGraph.AnimGraphNode_StateMachine_3.Locomotion.AnimStateNode_0.Idle.AnimGraphNode_StateMachine_2.DefaultIdleStateMachine.AnimStateNode_2.SpecialIdleToDefaultIdle.AnimGraphNode_StateMachine_0.TransitionToIdle.AnimStateTransitionNode_2.Transition
    src=Conduit
    via=AnimStateTransitionNode_2
   }
   {
    dst=Happy02_To_Idle
    rule=AnimGraph.AnimGraphNode_StateMachine_3.Locomotion.AnimStateNode_0.Idle.AnimGraphNode_StateMachine_2.DefaultIdleStateMachine.AnimStateNode_2.SpecialIdleToDefaultIdle.AnimGraphNode_StateMachine_0.TransitionToIdle.AnimStateTransitionNode_4.Transition
    src=Conduit
    via=AnimStateTransitionNode_4
   }
   {
    dst=Happy_01_To_Idle
    rule=AnimGraph.AnimGraphNode_StateMachine_3.Locomotion.AnimStateNode_0.Idle.AnimGraphNode_StateMachine_2.DefaultIdleStateMachine.AnimStateNode_2.SpecialIdleToDefaultIdle.AnimGraphNode_StateMachine_0.TransitionToIdle.AnimStateTransitionNode_1.Transition
    src=Conduit
    via=AnimStateTransitionNode_1
   }
  ]
 }
 Conduit {
  path=AnimGraph.AnimGraphNode_StateMachine_3.Locomotion.AnimStateNode_0.Idle.AnimGraphNode_StateMachine_2.DefaultIdleStateMachine.AnimStateNode_2.SpecialIdleToDefaultIdle.AnimGraphNode_StateMachine_0.TransitionToIdle.AnimStateConduitNode_0.Conduit
  nodes [Result(bCanEnterTransition=true)]
 }
 Combat_To_Idle {
  path=AnimGraph.AnimGraphNode_StateMachine_3.Locomotion.AnimStateNode_0.Idle.AnimGraphNode_StateMachine_2.DefaultIdleStateMachine.AnimStateNode_2.SpecialIdleToDefaultIdle.AnimGraphNode_StateMachine_0.TransitionToIdle.AnimStateNode_0.Combat_To_Idle
  nodes [Output Animation Pose(Result=Sequence Player 'SitPlant_COR_Angry_Idle_Out'() {bLoopAnimation=false, sequence=&SitPlant_COR_Angry_Idle_Out})]
 }
 Happy_01_To_Idle {
  path=AnimGraph.AnimGraphNode_StateMachine_3.Locomotion.AnimStateNode_0.Idle.AnimGraphNode_StateMachine_2.DefaultIdleStateMachine.AnimStateNode_2.SpecialIdleToDefaultIdle.AnimGraphNode_StateMachine_0.TransitionToIdle.AnimStateNode_1.Happy_01_To_Idle
  nodes [Output Animation Pose(Result=Sequence Player 'SitPlant_RCT_Happy_01_Out'() {bLoopAnimation=false, sequence=&SitPlant_RCT_Happy_01_Out})]
 }
 EmoteDance_To_Idle {
  path=AnimGraph.AnimGraphNode_StateMachine_3.Locomotion.AnimStateNode_0.Idle.AnimGraphNode_StateMachine_2.DefaultIdleStateMachine.AnimStateNode_2.SpecialIdleToDefaultIdle.AnimGraphNode_StateMachine_0.TransitionToIdle.AnimStateNode_2.EmoteDance_To_Idle
  nodes [Output Animation Pose(Result=Sequence Player 'SitPlant_RCT_Dance_Out'() {bLoopAnimation=false, sequence=&SitPlant_RCT_Dance_Out})]
 }
 EmoteNonDance_To_Idle {
  path=AnimGraph.AnimGraphNode_StateMachine_3.Locomotion.AnimStateNode_0.Idle.AnimGraphNode_StateMachine_2.DefaultIdleStateMachine.AnimStateNode_2.SpecialIdleToDefaultIdle.AnimGraphNode_StateMachine_0.TransitionToIdle.AnimStateNode_3.EmoteNonDance_To_Idle
  nodes [Output Animation Pose(Result=Sequence Player 'SitPlant_RCT_NonDance_Out'() {bLoopAnimation=false, sequence=&SitPlant_RCT_NonDance_Out})]
 }
 Happy02_To_Idle {
  path=AnimGraph.AnimGraphNode_StateMachine_3.Locomotion.AnimStateNode_0.Idle.AnimGraphNode_StateMachine_2.DefaultIdleStateMachine.AnimStateNode_2.SpecialIdleToDefaultIdle.AnimGraphNode_StateMachine_0.TransitionToIdle.AnimStateNode_4.Happy02_To_Idle
  nodes [Output Animation Pose(Result=Sequence Player 'SitPlant_RCT_Happy_02_Out'() {bLoopAnimation=false, sequence=&SitPlant_RCT_Happy_02_Out})]
 }
 Transition {
  path=AnimGraph.AnimGraphNode_StateMachine_3.Locomotion.AnimStateNode_0.Idle.AnimGraphNode_StateMachine_2.DefaultIdleStateMachine.AnimStateNode_2.SpecialIdleToDefaultIdle.AnimGraphNode_StateMachine_0.TransitionToIdle.AnimStateTransitionNode_0.Transition
  nodes [Result(bCanEnterTransition=Get IsCombatToIdle())]
 }
 Transition {
  path=AnimGraph.AnimGraphNode_StateMachine_3.Locomotion.AnimStateNode_0.Idle.AnimGraphNode_StateMachine_2.DefaultIdleStateMachine.AnimStateNode_2.SpecialIdleToDefaultIdle.AnimGraphNode_StateMachine_0.TransitionToIdle.AnimStateTransitionNode_1.Transition
  nodes [Result(bCanEnterTransition=Get IsReactionHappyCurve())]
 }
 Transition {
  path=AnimGraph.AnimGraphNode_StateMachine_3.Locomotion.AnimStateNode_0.Idle.AnimGraphNode_StateMachine_2.DefaultIdleStateMachine.AnimStateNode_2.SpecialIdleToDefaultIdle.AnimGraphNode_StateMachine_0.TransitionToIdle.AnimStateTransitionNode_2.Transition
  nodes [Result(bCanEnterTransition=Get IsEmoteNonDanceToIdle())]
 }
 Transition {
  path=AnimGraph.AnimGraphNode_StateMachine_3.Locomotion.AnimStateNode_0.Idle.AnimGraphNode_StateMachine_2.DefaultIdleStateMachine.AnimStateNode_2.SpecialIdleToDefaultIdle.AnimGraphNode_StateMachine_0.TransitionToIdle.AnimStateTransitionNode_3.Transition
  nodes [Result(bCanEnterTransition=Get IsEmoteDanceToIdle())]
 }
 Transition {
  path=AnimGraph.AnimGraphNode_StateMachine_3.Locomotion.AnimStateNode_0.Idle.AnimGraphNode_StateMachine_2.DefaultIdleStateMachine.AnimStateNode_2.SpecialIdleToDefaultIdle.AnimGraphNode_StateMachine_0.TransitionToIdle.AnimStateTransitionNode_4.Transition
  nodes [Result(bCanEnterTransition=true)]
 }
 Transition {
  path=AnimGraph.AnimGraphNode_StateMachine_3.Locomotion.AnimStateNode_0.Idle.AnimGraphNode_StateMachine_2.DefaultIdleStateMachine.AnimStateTransitionNode_0.Transition
  nodes [Result(bCanEnterTransition=true)]
 }
 Transition {
  path=AnimGraph.AnimGraphNode_StateMachine_3.Locomotion.AnimStateNode_0.Idle.AnimGraphNode_StateMachine_2.DefaultIdleStateMachine.AnimStateTransitionNode_1.Transition
  nodes [Result(bCanEnterTransition=Get IsSpecialIdleEnter())]
 }
 Transition {
  path=AnimGraph.AnimGraphNode_StateMachine_3.Locomotion.AnimStateNode_0.Idle.AnimGraphNode_StateMachine_2.DefaultIdleStateMachine.AnimStateTransitionNode_2.Transition
  nodes [Result(bCanEnterTransition=true)]
 }
 Transition {
  path=AnimGraph.AnimGraphNode_StateMachine_3.Locomotion.AnimStateNode_0.Idle.AnimGraphNode_StateMachine_2.DefaultIdleStateMachine.AnimStateTransitionNode_3.Transition
  nodes [Result(bCanEnterTransition=Get IsSpecialIdleExit())]
 }
 Transition {
  path=AnimGraph.AnimGraphNode_StateMachine_3.Locomotion.AnimStateNode_0.Idle.AnimGraphNode_StateMachine_2.DefaultIdleStateMachine.AnimStateTransitionNode_4.Transition
  nodes [Result()]
 }
 Transition {
  path=AnimGraph.AnimGraphNode_StateMachine_3.Locomotion.AnimStateNode_0.Idle.AnimGraphNode_StateMachine_2.DefaultIdleStateMachine.AnimStateTransitionNode_5.Transition
  nodes [Result(bCanEnterTransition=Get IsReactionHappyPlaying())]
 }
 LocomotionStart {
  path=AnimGraph.AnimGraphNode_StateMachine_3.Locomotion.AnimStateNode_2.LocomotionStart
  nodes [
   Blendspace Player 'SitPlant_xDEV_Start_BS'!(X=Get UpdatedGoalRotation() -> (UpdatedGoalRotation:float)) -> (Pose) {bLoop=false, blendSpace=/CosmeticCompanions/Assets/Quadruped/SitPlant/Animation/Developer/SitPlant_xDEV_Start_BS}
   n1 = Sequence Player 'SitPlant_COR_Loco_Start'() {bLoopAnimation=false, playRate=1.2, sequence=/CosmeticCompanions/Assets/Quadruped/SitPlant/Animation/Core/SitPlant_COR_Loco_Start}
   Output Animation Pose(Result=@n1) {
    node {stateEntryFunction {functionName=OnStateEntry_0}, stateExitFunction {functionName=OnStateExit_0}}
    stateEntryFunction {bSelfContext=true, memberGuid=#4, memberName=StartLocomotion_StateEntry}
    stateExitFunction {bSelfContext=true, memberGuid=#5, memberName=StartLocotion_StateExit}
   }
  ]
 }
 Locomotion {
  path=AnimGraph.AnimGraphNode_StateMachine_3.Locomotion.AnimStateNode_3.Locomotion
  nodes [
   Modify Curve(
    SourcePose=Sequence Player 'WideSnore_COR_Loco_Walk'() {sequence=/CosmeticCompanions/Assets/Quadruped/WideSnore/Animation/Core/WideSnore_COR_Loco_Walk}
    CurveValues_0=1
    Alpha=1
   ) {curveNames [Walk], curveValues [0]}
   Sequence Player 'SitPlant_COR_Loco_Trot'!(PlayRate:float=1) -> (Pose) {groupName=Locomotion, method=SyncGroup, sequence=&SitPlant_COR_Loco_Trot}
   Sequence Player 'SitPlant_xDEV_Trot_Slow'() {groupName=Locomotion, method=SyncGroup, sequence=/CosmeticCompanions/Assets/Quadruped/SitPlant/Animation/Developer/SitPlant_xDEV_Trot_Slow}
   n1 = Blend Poses by bool(
    BlendPose_0=Sequence Player 'WideSnore_COR_Loco_Run'() {playRate=1.3, sequence=/CosmeticCompanions/Assets/Quadruped/WideSnore/Animation/Core/WideSnore_COR_Loco_Run}
    BlendPose_1=Blend Poses by bool(
     BlendPose_0=Sequence Player 'WideSnore_COR_Loco_Run'() {sequence=&WideSnore_COR_Loco_Run}
     BlendPose_1=Sequence Player 'WideSnore_COR_Loco_Trot'() {sequence=/CosmeticCompanions/Assets/Quadruped/WideSnore/Animation/Core/WideSnore_COR_Loco_Trot}
     BlendTime_1=0.2
     bActiveValue=Get IsLocoTrot()
    ) {blendType=Sinusoidal}
    BlendTime_0=0.2
    BlendTime_1=0.4
    bActiveValue=Get IsLocoRun()
   ) {blendType=Sinusoidal}
   n2 = Sequence Player 'WideSnore_COR_Loco_Sprint'() {sequence=/CosmeticCompanions/Assets/Quadruped/WideSnore/Animation/Core/WideSnore_COR_Loco_Sprint}
   n3 = Get IsSprinting()
   Output Animation Pose(Result=Modify Curve(
     SourcePose=Blend Poses by bool(BlendPose_0=@n2, BlendPose_1=@n1, BlendTime_0=0.2, BlendTime_1=0.2, bActiveValue=@n3) {blendType=Sinusoidal}
     CurveValues_0=1
     Alpha=1
    ) {curveNames [Locomotion], curveValues [0]})
  ]
 }
 LocomotionStop {
  path=AnimGraph.AnimGraphNode_StateMachine_3.Locomotion.AnimStateNode_4.LocomotionStop
  nodes [
   AnimGraphNode_StateMachine_0 = LocomotionStopStateMachine() {is=title AnimGraphNode_StateMachine, bAllowConduitEntryStates=true}
   Output Animation Pose(Result=Modify Curve(SourcePose=@AnimGraphNode_StateMachine_0, CurveValues_0=1, Alpha=1) {curveNames [idle], curveValues [0]})
  ]
 }
 LocomotionStopStateMachine {
  entry=Conduit
  path=AnimGraph.AnimGraphNode_StateMachine_3.Locomotion.AnimStateNode_4.LocomotionStop.AnimGraphNode_StateMachine_0.LocomotionStopStateMachine
  ids=id
  nodes [
   AnimStateConduitNode_0 = Conduit()
   AnimStateNode_0 = RunStopFootL() {is=title AnimStateNode}
   AnimStateNode_1 = RunStopFootR() {is=title AnimStateNode}
   AnimStateNode_2 = WalkStopFootR() {is=title AnimStateNode}
   AnimStateNode_3 = WalkStopFootL() {is=title AnimStateNode}
   AnimStateTransitionNode_0 = AnimStateTransitionNode() {crossfadeDuration=0.25, logicType=TLT_Inertialization, priorityOrder=100}
   AnimStateTransitionNode_1 = AnimStateTransitionNode() {crossfadeDuration=0.25, logicType=TLT_Inertialization, priorityOrder=3}
   AnimStateTransitionNode_2 = AnimStateTransitionNode()
   AnimStateTransitionNode_3 = AnimStateTransitionNode() {priorityOrder=2}
  ]
  links [
   {
    dst=RunStopFootL
    rule=AnimGraph.AnimGraphNode_StateMachine_3.Locomotion.AnimStateNode_4.LocomotionStop.AnimGraphNode_StateMachine_0.LocomotionStopStateMachine.AnimStateTransitionNode_0.Transition
    src=Conduit
    via=AnimStateTransitionNode_0
   }
   {
    dst=RunStopFootR
    rule=AnimGraph.AnimGraphNode_StateMachine_3.Locomotion.AnimStateNode_4.LocomotionStop.AnimGraphNode_StateMachine_0.LocomotionStopStateMachine.AnimStateTransitionNode_1.Transition
    src=Conduit
    via=AnimStateTransitionNode_1
   }
   {
    dst=WalkStopFootL
    rule=AnimGraph.AnimGraphNode_StateMachine_3.Locomotion.AnimStateNode_4.LocomotionStop.AnimGraphNode_StateMachine_0.LocomotionStopStateMachine.AnimStateTransitionNode_3.Transition
    src=Conduit
    via=AnimStateTransitionNode_3
   }
   {
    dst=WalkStopFootR
    rule=AnimGraph.AnimGraphNode_StateMachine_3.Locomotion.AnimStateNode_4.LocomotionStop.AnimGraphNode_StateMachine_0.LocomotionStopStateMachine.AnimStateTransitionNode_2.Transition
    src=Conduit
    via=AnimStateTransitionNode_2
   }
  ]
 }
 Conduit {
  path=AnimGraph.AnimGraphNode_StateMachine_3.Locomotion.AnimStateNode_4.LocomotionStop.AnimGraphNode_StateMachine_0.LocomotionStopStateMachine.AnimStateConduitNode_0.Conduit
  nodes [Result(bCanEnterTransition=true)]
 }
 RunStopFootL {
  path=AnimGraph.AnimGraphNode_StateMachine_3.Locomotion.AnimStateNode_4.LocomotionStop.AnimGraphNode_StateMachine_0.LocomotionStopStateMachine.AnimStateNode_0.RunStopFootL
  nodes [
   n1 = Sequence Player 'SitPlant_xDEV_StopTrotLFootForward'!(StartPosition:float=Get Stop_StartPosition() -> (Stop_StartPosition:float)) -> (Pose) {
    bLoopAnimation=false
    playRate=0.7
    sequence=/CosmeticCompanions/Assets/Quadruped/SitPlant/Animation/Developer/SitPlant_xDEV_StopTrotLFootForward
   }
   Output Animation Pose(Result=@n1)
  ]
 }
 RunStopFootR {
  path=AnimGraph.AnimGraphNode_StateMachine_3.Locomotion.AnimStateNode_4.LocomotionStop.AnimGraphNode_StateMachine_0.LocomotionStopStateMachine.AnimStateNode_1.RunStopFootR
  nodes [
   n1 = Sequence Player 'SitPlant_xDEV_StopTrotRFootForward'!(StartPosition:float=Get Stop_StartPosition()) -> (Pose) {
    bLoopAnimation=false
    playRate=0.7
    sequence=/CosmeticCompanions/Assets/Quadruped/SitPlant/Animation/Developer/SitPlant_xDEV_StopTrotRFootForward
   }
   Output Animation Pose(Result=@n1)
  ]
 }
 WalkStopFootR {
  path=AnimGraph.AnimGraphNode_StateMachine_3.Locomotion.AnimStateNode_4.LocomotionStop.AnimGraphNode_StateMachine_0.LocomotionStopStateMachine.AnimStateNode_2.WalkStopFootR
  nodes [Output Animation Pose(Result=Sequence Player 'SitPlant_xDEV_StopWalkRFootForward'() {bLoopAnimation=false, sequence=/CosmeticCompanions/Assets/Quadruped/SitPlant/Animation/Developer/SitPlant_xDEV_StopWalkRFootForward})]
 }
 WalkStopFootL {
  path=AnimGraph.AnimGraphNode_StateMachine_3.Locomotion.AnimStateNode_4.LocomotionStop.AnimGraphNode_StateMachine_0.LocomotionStopStateMachine.AnimStateNode_3.WalkStopFootL
  nodes [Output Animation Pose(Result=Sequence Player 'SitPlant_xDEV_StopWalkLFootForward'() {bLoopAnimation=false, sequence=/CosmeticCompanions/Assets/Quadruped/SitPlant/Animation/Developer/SitPlant_xDEV_StopWalkLFootForward})]
 }
 Transition {
  path=AnimGraph.AnimGraphNode_StateMachine_3.Locomotion.AnimStateNode_4.LocomotionStop.AnimGraphNode_StateMachine_0.LocomotionStopStateMachine.AnimStateTransitionNode_0.Transition
  nodes [Result(bCanEnterTransition=true)]
 }
 Transition {
  path=AnimGraph.AnimGraphNode_StateMachine_3.Locomotion.AnimStateNode_4.LocomotionStop.AnimGraphNode_StateMachine_0.LocomotionStopStateMachine.AnimStateTransitionNode_1.Transition
  nodes [Result(bCanEnterTransition=Get IsFootPhase() -> (IsFootPhase:bool))]
 }
 Transition {
  path=AnimGraph.AnimGraphNode_StateMachine_3.Locomotion.AnimStateNode_4.LocomotionStop.AnimGraphNode_StateMachine_0.LocomotionStopStateMachine.AnimStateTransitionNode_2.Transition
  nodes [Result(bCanEnterTransition=Get IsWalkStopFootR() -> (IsWalkStopFootR:bool))]
 }
 Transition {
  path=AnimGraph.AnimGraphNode_StateMachine_3.Locomotion.AnimStateNode_4.LocomotionStop.AnimGraphNode_StateMachine_0.LocomotionStopStateMachine.AnimStateTransitionNode_3.Transition
  nodes [Result(bCanEnterTransition=Get IsWalkSpeed() -> (IsWalkSpeed:bool))]
 }
 Transition {path=AnimGraph.AnimGraphNode_StateMachine_3.Locomotion.AnimStateTransitionNode_0.Transition, nodes [Result(bCanEnterTransition=true)]}
 Transition {
  path=AnimGraph.AnimGraphNode_StateMachine_3.Locomotion.AnimStateTransitionNode_1.Transition
  nodes [Result(bCanEnterTransition=Get IsIdle() -> (IsIdle:bool))]
 }
 Transition {
  path=AnimGraph.AnimGraphNode_StateMachine_3.Locomotion.AnimStateTransitionNode_10.Transition
  nodes [Result(bCanEnterTransition=Get IsIdle())]
 }
 Transition {
  path=AnimGraph.AnimGraphNode_StateMachine_3.Locomotion.AnimStateTransitionNode_11.Transition
  nodes [Result(bCanEnterTransition=Get IsEarlyExitLocoStop() -> (IsEarlyExitLocoStop:bool))]
 }
 Transition {
  path=AnimGraph.AnimGraphNode_StateMachine_3.Locomotion.AnimStateTransitionNode_12.Transition
  nodes [Result(bCanEnterTransition=Get isLocoStartToStop() -> (isLocoStartToStop:bool))]
 }
 Transition {path=AnimGraph.AnimGraphNode_StateMachine_3.Locomotion.AnimStateTransitionNode_2.Transition, nodes [Result(bCanEnterTransition=true)]}
 Transition {path=AnimGraph.AnimGraphNode_StateMachine_3.Locomotion.AnimStateTransitionNode_3.Transition, nodes [Result()]}
 Transition {
  path=AnimGraph.AnimGraphNode_StateMachine_3.Locomotion.AnimStateTransitionNode_5.Transition
  nodes [
   Result()
   "float < float"(A="Get Relevant Anim Time Remaining (LocomotionStart)"!() -> (ReturnValue:float) {is=title K2Node_AnimGetter}, B=0.1)
  ]
 }
 Transition {
  path=AnimGraph.AnimGraphNode_StateMachine_3.Locomotion.AnimStateTransitionNode_8.Transition
  nodes [Result(bCanEnterTransition=Get IsMoving())]
 }
 Transition {
  path=AnimGraph.AnimGraphNode_StateMachine_3.Locomotion.AnimStateTransitionNode_9.Transition
  nodes [Result(bCanEnterTransition=Get IsMoving())]
 }
 Tail {
  entry=Conduit
  path=AnimGraph.AnimGraphNode_StateMachine_4.Tail
  ids=id
  nodes [
   AnimStateConduitNode_0 = Conduit()
   AnimStateNode_0 = Default()
   AnimStateNode_2 = Combat()
   AnimStateTransitionNode_0 = AnimStateTransitionNode() {priorityOrder=100}
   AnimStateTransitionNode_1 = AnimStateTransitionNode()
   AnimStateTransitionNode_4 = AnimStateTransitionNode() {priorityOrder=2}
   AnimStateTransitionNode_5 = AnimStateTransitionNode()
  ]
  links [
   {
    dst=Conduit
    rule=AnimGraph.AnimGraphNode_StateMachine_4.Tail.AnimStateTransitionNode_5.Transition
    src=Combat
    via=AnimStateTransitionNode_5
   }
   {
    dst=Combat
    rule=AnimGraph.AnimGraphNode_StateMachine_4.Tail.AnimStateTransitionNode_4.Transition
    src=Conduit
    via=AnimStateTransitionNode_4
   }
   {
    dst=Default
    rule=AnimGraph.AnimGraphNode_StateMachine_4.Tail.AnimStateTransitionNode_0.Transition
    src=Conduit
    via=AnimStateTransitionNode_0
   }
   {
    dst=Conduit
    rule=AnimGraph.AnimGraphNode_StateMachine_4.Tail.AnimStateTransitionNode_1.Transition
    src=Default
    via=AnimStateTransitionNode_1
   }
  ]
 }
 Conduit {path=AnimGraph.AnimGraphNode_StateMachine_4.Tail.AnimStateConduitNode_0.Conduit, nodes [Result(bCanEnterTransition=true)]}
 Default {
  path=AnimGraph.AnimGraphNode_StateMachine_4.Tail.AnimStateNode_0.Default
  nodes [Output Animation Pose(Result=Blendspace Player 'SitPlant_COR_Tail_BS'!(X=Get TailWagAmount() -> (TailWagAmount:float)) -> (Pose) {blendSpace=/CosmeticCompanions/Assets/Quadruped/SitPlant/Animation/Core/SitPlant_COR_Tail_BS})]
 }
 Combat {
  path=AnimGraph.AnimGraphNode_StateMachine_4.Tail.AnimStateNode_2.Combat
  nodes [Output Animation Pose(Result="Sequence Player (additive) 'SitPlant_COR_Tail_Angry_ADD'"() {bLoopAnimation=false, sequence=/CosmeticCompanions/Assets/Quadruped/SitPlant/Animation/Core/SitPlant_COR_Tail_Angry_ADD})]
 }
 Transition {path=AnimGraph.AnimGraphNode_StateMachine_4.Tail.AnimStateTransitionNode_0.Transition, nodes [Result(bCanEnterTransition=true)]}
 Transition {path=AnimGraph.AnimGraphNode_StateMachine_4.Tail.AnimStateTransitionNode_1.Transition, nodes [Result(bCanEnterTransition=true)]}
 Transition {
  path=AnimGraph.AnimGraphNode_StateMachine_4.Tail.AnimStateTransitionNode_4.Transition
  nodes [Result(bCanEnterTransition=Get IsMoodCombat())]
 }
 Transition {path=AnimGraph.AnimGraphNode_StateMachine_4.Tail.AnimStateTransitionNode_5.Transition, nodes [Result(bCanEnterTransition=true)]}
 EventGraph {nodes [
   n1 = Event BlueprintUpdateAnimation()
   > Set Delta Time X(Delta Time X=@n1.DeltaTimeX)
   > n2 = Sequence!(execute) -> (then_0, then_1, then_2:Exec, then_3:Exec)
   > n3 = HasMatchingGameplayTag!(
    execute:Exec
    self:Object Reference=Get CompanionLeader() -> (CompanionLeader:Actor Object Reference)
    TagToCheck:Gameplay Tag Structure=(TagName="Gameplay.Action.Player.Crouch")
   ) -> (then:Exec, ReturnValue:bool) {is=title K2Node_Message}
   > Set IsPlayerCrouching(IsPlayerCrouching:bool=@n3) -> (Output_Get:bool)
   > Set IsMoodBored(IsMoodBored:bool="Equal (Enum)"(A:EPlayerCompanionMood Enum=Get Mood() -> (Mood:EPlayerCompanionMood Enum=Default), B:EPlayerCompanionMood Enum=Bored)) -> (Output_Get:bool)
   > Set IsMoodWorried(IsMoodWorried:bool="Equal (Enum)"(A=Get Mood() -> (Mood=Default), B=Worried)) -> (Output_Get:bool)
   > Set IsMoodCombat(IsMoodCombat:bool="Equal (Enum)"(A=Get Mood() -> (Mood=Default), B=Combat)) -> (Output_Get:bool)
   n4 = GetOwningActor()
   > Set CompanionSpeed(CompanionSpeed:double=Vector Length XY!(A:Vector=GetVelocity!(self:Actor Object Reference=@n4) -> (ReturnValue:Vector)) -> (ReturnValue:double)) -> (Output_Get:double)
   > Set IsMontagePlaying(IsMontagePlaying:bool="float > float"(A=GetSlotLocalWeight!(self:Anim Instance Object Reference, SlotNodeName:Name=DefaultSlot) -> (ReturnValue:float))) -> (Output_Get:bool)
   > Set IsSprinting(IsSprinting:bool="float >= float"(A=Get CompanionSpeed() -> (CompanionSpeed:double), B=675)) -> (Output_Get:bool)
   > Set IsMoving(IsMoving:bool="float > float"(A=Get CompanionSpeed())) -> (Output_Get:bool)
   > Set IsIdle(IsIdle:bool="Equal (Float)"(A=Get CompanionSpeed())) -> (Output_Get:bool)
   > Set IsWalkSpeed(IsWalkSpeed:bool="float > float"(A=GetCurveValue(CurveName=walk), B=0.5)) -> (Output_Get:bool)
   > n5 = Branch(Condition=Get IsMoodCombat())
   > Set EarPosition(EarPosition:double=500) -> (Output_Get:double)
   Set EarPosition(execute=@n5.else, EarPosition=Get CompanionSpeed())
   n6 = Sequence(execute=@n2.then_1)
   > n7 = Sequence!(execute) -> (then_0, then_1, then_2)
   > n8 = Sequence!(execute) -> (then_0, then_1)
   > n9 = Branch(Condition="float >= float"(A=Get CompanionSpeed(), B=385))
   > Set IsLocoRun(IsLocoRun:bool=true) -> (Output_Get:bool)
   Branch(execute=@n9.else, Condition="float < float"(A=Get CompanionSpeed(), B=370))
   > Set IsLocoRun(IsLocoRun=false)
   n10 = Branch(execute=@n8.then_1, Condition="float >= float"(A=Get CompanionSpeed(), B=200))
   > Set IsLocoTrot(IsLocoTrot:bool=true) -> (Output_Get:bool)
   Branch(execute=@n10.else, Condition="float < float"(A=Get CompanionSpeed(), B=190))
   > Set IsLocoTrot(IsLocoTrot=false)
   n11 = Get!() -> (GoalRotation_Roll:float, GoalRotation_Pitch:float, GoalRotation_Yaw:float)
   Branch(execute=@n7.then_1, Condition="Not Equal (Float)"(A=Get OldGoalRotation() -> (OldGoalRotation:float), B=@n11.GoalRotation_Yaw))
   > Set OldGoalRotation(OldGoalRotation:float=@n11.GoalRotation_Yaw) -> (Output_Get:float)
   n12 = Get!() -> (CompanionRotation_Roll:float, CompanionRotation_Pitch:float, CompanionRotation_Yaw:float)
   > Set UpdatedGoalRotation(UpdatedGoalRotation:float=NormalizeAxis!(Angle:float=float - float(A=@n11.GoalRotation_Yaw, B=@n12.CompanionRotation_Yaw)) -> (ReturnValue:float)) -> (Output_Get:float)
   n13 = Branch(execute=@n6.then_2, Condition="float > float"(A=GetCurveValue(CurveName=AbilityStart), B=0.5))
   > Set Stop_StartPosition(Stop_StartPosition:float=0) -> (Output_Get:float)
   Set Stop_StartPosition(execute=@n13.else, Stop_StartPosition=0.1)
   n14 = Sequence!(execute=@n6.then_3) -> (then_0, then_1, then_2:Exec, then_3:Exec)
   n15 = Branch(execute=@n14.then_3, Condition=Get IsLocomotionCoolDown() -> (IsLocomotionCoolDown:bool))
   > Set Appendage01(Appendage01:float=0.8) -> (Output_Get:float)
   Set Appendage01(execute=@n15.else, Appendage01=MapRangeClamped(Value=Get CompanionSpeed(), InRangeA=0, InRangeB=500, OutRangeA=1, OutRangeB=0.01))
   n16 = Branch(execute=@n14.then_0, Condition=Get bIsPreviewActor() -> (bIsPreviewActor:bool))
   > n17 = Branch(Condition="Equal (Integer)"(A=Get PreviewActorIndex() -> (PreviewActorIndex:int=0)))
   > n18 = Branch(Condition=Get bIsPlayingEmote())
   > n19 = Branch(Condition="float > float"(A=GetCurveValue(CurveName=EmoteCurveB)))
   > n20 = Branch(Condition="float > float"(A=Get LookAtPitch() -> (LookAtPitch:float), B=55))
   > Set LookAtYawAdjusted(LookAtYawAdjusted:double=-45) -> (Output_Get:double)
   > Set LookAtPitchActual(LookAtPitchActual:float=25) -> (Output_Get:float)
   Set LookAtYawAdjusted(execute=@n20.else, LookAtYawAdjusted=-45)
   > Set LookAtPitchActual(LookAtPitchActual=0)
   Set LookAtYawAdjusted(execute=@n19.else, LookAtYawAdjusted=0)
   > Set LookAtPitchActual(LookAtPitchActual=0)
   Set LookAtYawAdjusted(execute=@n18.else, LookAtYawAdjusted=-27)
   > Set LookAtPitchActual(LookAtPitchActual=5)
   Set LookAtYawAdjusted(execute=@n17.else, LookAtYawAdjusted=0)
   > Set LookAtPitchActual(LookAtPitchActual=0)
   Set LookAtPitchActual(execute=@n16.else, LookAtPitchActual=Get LookAtPitch())
   > n21 = Branch(Condition=OR Boolean(A=Get IsMontagePlaying() -> (IsMontagePlaying:bool), B=Get IsHappyReactionCoolDown() -> (IsHappyReactionCoolDown:bool)))
   > n22 = Branch(Condition=Get IsEmoteCurveB() -> (IsEmoteCurveB:bool))
   > Set LookAtYawAdjusted(LookAtYawAdjusted=Get LookAtYaw() -> (LookAtYaw:float))
   Set LookAtYawAdjusted(execute=[@n22.else, @n23.then], LookAtYawAdjusted=0)
   n23 = Branch(execute=@n21.else, Condition=Get IsLocoStart() -> (IsLocoStart:bool))
   Set LookAtYawAdjusted(execute=@n23.else, LookAtYawAdjusted=Get LookAtYaw())
   Set BreathingAlpha(execute=@n14.then_1, BreathingAlpha:double=MapRangeClamped(Value=Get CompanionSpeed(), InRangeA=0, InRangeB=200, OutRangeA=0.8, OutRangeB=0)) -> (Output_Get:double)
   n24 = Branch(
    execute=@n14.then_2
    Condition=OR Boolean!(
     A=Get IsMontagePlaying()
     B=Get IsHappyReactionCoolDownReady() -> (IsHappyReactionCoolDownReady:bool)
     C:bool=Get IsEmoteCurveA() -> (IsEmoteCurveA:bool)
     D:bool=Get IsEmoteCurveB()
     E:bool=Get IsLocoStart()
     F:bool=Get IsMoodCombat()
    ) -> (ReturnValue)
   )
   > Set IdleAnticAlpha(IdleAnticAlpha:double=0) -> (Output_Get:double)
   Set IdleAnticAlpha(execute=@n24.else, IdleAnticAlpha=MapRangeClamped(Value=Get CompanionSpeed(), InRangeA=0, InRangeB=100, OutRangeA=1, OutRangeB=0))
   n25 = Branch(
    execute=@n6.then_1
    Condition=OR Boolean!(
     A=Get IsBoredCoolDown() -> (IsBoredCoolDown:bool)
     B=Get IsCombatReactionCoolDown() -> (IsCombatReactionCoolDown:bool)
     C=Get IsHappyReactionCoolDown()
     D=Get IsMoodWorried()
    ) -> (ReturnValue)
   )
   > n26 = Branch(Condition=Get IsMoodWorried())
   > Set TailWagAmount(TailWagAmount:float=0.1) -> (Output_Get:float)
   Set TailWagAmount(execute=@n26.else, TailWagAmount=0.8)
   n27 = Branch(execute=@n25.else, Condition="float > float"(A=GetCurveValue(CurveName=tail_curve)))
   > Set TailWagAmount(TailWagAmount=GetCurveValue(CurveName=tail_curve))
   Set TailWagAmount(
    execute=@n27.else
    TailWagAmount=Get!() -> (AuxiliaryTimer1_ValueInt:int, AuxiliaryTimer1_ValueFloat:float=~, AuxiliaryTimer1_bJustChanged:bool=~).AuxiliaryTimer1_ValueFloat
   )
   n28 = Sequence!(execute=@n2.then_2) -> (then_0, then_1)
   > n29 = Sequence!(execute) -> (then_0, then_1, then_2:Exec)
   > Branch(Condition=Get IsIdle())
   > Set IsLocomotionCoolDownReady(IsLocomotionCoolDownReady:bool=true) -> (Output_Get:bool)
   Branch(execute=@n29.then_1, Condition=Get IsLocomotionCoolDownReady() -> (IsLocomotionCoolDownReady:bool))
   > Branch(Condition="float > float"(A=Get CompanionSpeed(), B=20))
   > Set IsLocomotionCoolDown(IsLocomotionCoolDown:bool=true) -> (Output_Get:bool)
   > Delay(Duration=2)
   > Set IsLocomotionCoolDown(IsLocomotionCoolDown=false)
   > Set IsLocomotionCoolDownReady(IsLocomotionCoolDownReady=false)
   n30 = Sequence(execute=@n28.then_1)
   > Branch(Condition="Equal (Enum)"(A=Get Mood() -> (Mood=Default), B=Bored))
   > Set IsBoredCoolDownReady(IsBoredCoolDownReady:bool=true) -> (Output_Get:bool)
   Branch(execute=@n30.then_1, Condition=Get IsBoredCoolDownReady() -> (IsBoredCoolDownReady:bool))
   > Branch(Condition="Not Equal (Enum)"!(A:EPlayerCompanionMood Enum=Get Mood() -> (Mood=Default), B:EPlayerCompanionMood Enum=Bored) -> (ReturnValue:bool))
   > Set IsBoredCoolDown(IsBoredCoolDown:bool=true) -> (Output_Get:bool)
   > Delay(Duration=3)
   > Set IsBoredCoolDown(IsBoredCoolDown=false)
   > Set IsBoredCoolDownReady(IsBoredCoolDownReady=false)
   Set IsEmoteCurveA(execute=@n2.then_3, IsEmoteCurveA:bool="float > float"(A=GetCurveValue(CurveName=EmoteCurveA))) -> (Output_Get:bool)
   > Set IsEmoteCurveB(IsEmoteCurveB:bool="float > float"(A=GetCurveValue(CurveName=EmoteCurveB))) -> (Output_Get:bool)
   n31 = GetCurveValue(CurveName=Reaction_Happy_01)
   n32 = GetCurveValue(CurveName=Reaction_Happy)
   > Set IsReactionHappyPlaying(IsReactionHappyPlaying:bool=OR Boolean!(A="float > float"(A=@n32), B="float > float"(A=@n31)) -> (ReturnValue)) -> (Output_Get:bool)
   > Set IsReactionHappyCurve(IsReactionHappyCurve:bool="float > float"(A=GetCurveValue(CurveName=Reaction_Happy))) -> (Output_Get:bool)
   > Set BreathingPlayRate(BreathingPlayRate:float=Get!() -> (AuxiliaryTimer3_ValueInt:int, AuxiliaryTimer3_ValueFloat:float=~, AuxiliaryTimer3_bJustChanged:bool=~).AuxiliaryTimer3_ValueFloat) -> (Output_Get:float)
   n33 = Get IsMoodWorried()
   n34 = GetCurveValue(CurveName=Idle)
   > Set IsLocomotionFromIdle(IsLocomotionFromIdle:bool=AND Boolean!(A=Get IsMoving(), B="float > float"(A=@n34), C:bool=NOT Boolean(A=@n33)) -> (ReturnValue)) -> (Output_Get:bool)
   n35 = Get UpdatedGoalRotation()
   n36 = "float < float"(A=@n35, B=-45)
   n37 = "float > float"(A=@n35, B=45)
   n38 = Get CompanionSpeed()
   n39 = GetCurveValue(CurveName=Locomotion)
   > Set IsLocomotionToLocoStart(IsLocomotionToLocoStart:bool=AND Boolean(A=OR Boolean(A=@n37, B=@n36), B="float < float"(A=@n38, B=350), C="float > float"(A=@n39, B=0.8))) -> (Output_Get:bool)
   n40 = Get IsMoodCombat()
   n41 = Get IsMoodWorried()
   > Set IsLocomotionToIdle(IsLocomotionToIdle:bool=AND Boolean!(A=Get IsIdle(), B=OR Boolean(A=@n40, B=@n41)) -> (ReturnValue)) -> (Output_Get:bool)
   > Set IsSpecialIdleEnter(IsSpecialIdleEnter:bool=OR Boolean!(A=Get bIsPlayingEmote(), B=Get IsEmoteCurveA(), C:bool=Get IsEmoteCurveB()) -> (ReturnValue)) -> (Output_Get:bool)
   n42 = Get IsHappyReactionCoolDown()
   n43 = Get IsMoodCombat()
   n44 = Get bIsPlayingEmote()
   > Set IsSpecialIdleExit(IsSpecialIdleExit:bool=AND Boolean!(A=NOT Boolean(A=@n42), B=NOT Boolean(A=@n43), C:bool=NOT Boolean(A=@n44)) -> (ReturnValue)) -> (Output_Get:bool)
   > Set IsCombatToIdle(IsCombatToIdle:bool="float > float"(A=GetCurveValue(CurveName=UserCurve01))) -> (Output_Get:bool)
   > Set IsEmoteDanceToIdle(IsEmoteDanceToIdle:bool="float > float"(A=GetCurveValue(CurveName=EmoteCurveA))) -> (Output_Get:bool)
   > Set IsEmoteNonDanceToIdle(IsEmoteNonDanceToIdle:bool="float > float"(A=GetCurveValue(CurveName=EmoteCurveB))) -> (Output_Get:bool)
   > Set IsFrontEndSpawnAnticChooser(IsFrontEndSpawnAnticChooser:bool=AND Boolean!(A=Get bIsPreviewActor(), B=Get IsFrontEndSpawn() -> (IsFrontEndSpawn:bool)) -> (ReturnValue)) -> (Output_Get:bool)
   n45 = GetCurveValue(CurveName=AbilityEnd)
   > Set IsEarlyExitLocoStop(IsEarlyExitLocoStop:bool=AND Boolean(A=Get IsMoodBored() -> (IsMoodBored:bool), B="Equal (Float)"(A=@n45, B=1))) -> (Output_Get:bool)
   n46 = GetCurveValue(CurveName=AbilityEnd)
   > Set isLocoStartToStop(isLocoStartToStop:bool=AND Boolean(A=Get IsLocomotionStop() -> (IsLocomotionStop:bool), B="Equal (Float)"(A=@n46))) -> (Output_Get:bool)
   n47 = "Equal (Integer)"(A=Get!() -> (AuxiliaryTimer2_ValueInt:int, AuxiliaryTimer2_ValueFloat:float=~, AuxiliaryTimer2_bJustChanged:bool=~))
   n48 = Get bIsPreviewActor()
   > Set IsStandToSit(IsStandToSit:bool=OR Boolean!(A=Get IsMoodBored(), B=AND Boolean(A=@n48, B=@n47)) -> (ReturnValue)) -> (Output_Get:bool)
   n49 = Get IsMoodBored()
   n50 = Get() -> (AuxiliaryTimer2_ValueFloat=~, AuxiliaryTimer2_bJustChanged=~)
   > Set IsSitToStand(IsSitToStand:bool=AND Boolean(A=NOT Boolean(A=@n49), B="Equal (Integer)"(A=@n50, B=1))) -> (Output_Get:bool)
   n51 = Get CompanionSpeed()
   > Set IsEarChange(IsEarChange:bool=OR Boolean(A="float > float"(A=@n51, B=100), B=Get IsMoodCombat())) -> (Output_Get:bool)
   > Set WorriedAddPoseAlpha(WorriedAddPoseAlpha:double=MapRangeClamped(Value=Get CompanionSpeed(), InRangeA=100, InRangeB=700, OutRangeA=1, OutRangeB=0.3)) -> (Output_Get:double)
   n52 = NOT Boolean(A=Get IsMoodCombat())
   > Set CombatToBodyAdditive(CombatToBodyAdditive:bool=@n52) -> (Output_Get:bool)
   n53 = Get IsMoodCombat()
   > Set BodyAdditiveToCombat(BodyAdditiveToCombat:bool=@n53) -> (Output_Get:bool)
   > Set IsFootPhase(IsFootPhase:bool="float > float"(A=GetCurveValue(CurveName=FootPhase), B=0.5)) -> (Output_Get:bool)
   > Set IsWalkStopFootR(IsWalkStopFootR:bool=AND Boolean(A=Get IsWalkSpeed(), B=Get IsFootPhase())) -> (Output_Get:bool)
   Event BlueprintInitializeAnimation()
   > Set IsFrontEndSpawn(IsFrontEndSpawn:bool=true) -> (Output_Get:bool)
   n54 = Get IsIdle()
   AND Boolean(A=@n54, B=NOT Boolean(A=Get IsMoodCombat()))
   AND Boolean(A=@n53, B="float > float"(A=Get CompanionSpeed(), B=25))
   OR Boolean(A=@n52, B="float < float"(A=Get CompanionSpeed(), B=90))
   Set IsLocomotionStop(IsLocomotionStop:bool=@n54) -> (Output_Get:bool)
   Set FrontEndRandomIdlePose(FrontEndRandomIdlePose:int=RandomInteger!(Max:int=2) -> (ReturnValue:int=0)) -> (Output_Get:int=0)
  ]}
 IsFrontEndSpawn_False {nodes [
   IsFrontEndSpawn_False!() -> (then, Context:"Anim Update Context Structure (by ref)", Node:"Anim Node Reference Structure (by ref)") {is=title K2Node_FunctionEntry}
   > Set IsFrontEndSpawn(IsFrontEndSpawn=false)
   > Return Node()
  ]}
 OnEnter_SitState {nodes [
   OnEnter_SitState!() -> (then, Context:"Anim Update Context Structure (by ref)", Node:"Anim Node Reference Structure (by ref)") {is=title K2Node_FunctionEntry}
   > Set IsSitting(IsSitting:bool=true) -> (Output_Get:bool)
   > Return Node()
  ]}
 OnExit_SitState {nodes [
   OnExit_SitState!() -> (then, Context:"Anim Update Context Structure (by ref)", Node:"Anim Node Reference Structure (by ref)") {is=title K2Node_FunctionEntry}
   > Set IsSitting(IsSitting=false)
   > Return Node()
  ]}
 OnStateEntry_0 {nodes [
   OnStateEntry_0!() -> (then, Context:"Anim Update Context Structure (by ref)", Node:"Anim Node Reference Structure (by ref)") {is=title K2Node_FunctionEntry}
   Return Node()
  ]}
 StartLocomotion_StateEntry {nodes [
   StartLocomotion_StateEntry!() -> (then, Context:"Anim Update Context Structure (by ref)", Node:"Anim Node Reference Structure (by ref)") {is=title K2Node_FunctionEntry}
   > Set IsLocoStart(IsLocoStart:bool=true) -> (Output_Get:bool)
   > Return Node()
  ]}
 StartLocotion_StateExit {nodes [
   StartLocotion_StateExit!() -> (then, Context:"Anim Update Context Structure (by ref)", Node:"Anim Node Reference Structure (by ref)") {is=title K2Node_FunctionEntry}
   > Set IsLocoStart(IsLocoStart=false)
   > Return Node()
  ]}
]
variables [
 Appendage01 {type=number}
 BodyAdditiveToCombat {type=boolean}
 BreathingAlpha {type=number}
 BreathingPlayRate {type=number}
 CombatToBodyAdditive {type=boolean}
 Companion Random Number Timer {type=CompanionRandomNumberTimer, default {bTimerEnabled=false, timeMax=10, timeMin=5, valueMax=2, valueMin=1}}
 CompanionSpeed {type=number}
 Delta Time X {type=number}
 EarPosition {type=number}
 FrontEndRandomIdlePose {type=integer}
 IdleAnticAlpha {type=number}
 IsBoredCoolDown {type=boolean}
 IsBoredCoolDownReady {type=boolean}
 IsCombatCoolDownReady {type=boolean}
 IsCombatReactionCoolDown {type=boolean}
 IsCombatToIdle {type=boolean}
 IsEarChange {type=boolean}
 IsEarlyExitLocoStop {type=boolean}
 IsEmoteCurveA {type=boolean}
 IsEmoteCurveB {type=boolean}
 IsEmoteDance {type=boolean}
 IsEmoteDanceToIdle {type=boolean}
 IsEmoteNonDanceToIdle {type=boolean}
 IsEmoteTraversal {type=boolean}
 IsFootPhase {type=boolean}
 IsFrontEndSpawn {type=boolean}
 IsFrontEndSpawnAnticChooser {type=boolean}
 IsHappyReactionCoolDown {type=boolean}
 IsHappyReactionCoolDownReady {type=boolean}
 IsHappyToIdle {type=boolean}
 IsIdle {type=boolean}
 IsIdleStart {type=boolean}
 IsLocomotionCoolDown {type=boolean}
 IsLocomotionCoolDownReady {type=boolean}
 IsLocomotionFromIdle {type=boolean}
 IsLocomotionStop {type=boolean}
 IsLocomotionToIdle {type=boolean}
 IsLocomotionToLocoStart {type=boolean}
 IsLocoRun {type=boolean}
 IsLocoStart {type=boolean}
 isLocoStartToStop {type=boolean}
 IsLocoStop {type=boolean}
 IsLocoTrot {type=boolean}
 IsLowFidelity {type=boolean}
 IsMontagePlaying {type=boolean}
 IsMoodBored {type=boolean}
 IsMoodCombat {type=boolean}
 IsMoodWorried {type=boolean}
 IsMoving {type=boolean}
 IsPlayerCrouching {type=boolean}
 IsReactionHappyCurve {type=boolean}
 IsReactionHappyPlaying {type=boolean}
 IsSitting {type=boolean}
 IsSitToStand {type=boolean}
 IsSpecialIdleEnter {type=boolean}
 IsSpecialIdleExit {type=boolean}
 IsSprinting {type=boolean}
 IsStandToSit {type=boolean}
 IsWalkSpeed {type=boolean}
 IsWalkStopFootR {type=boolean}
 JawPosition {type=number}
 LookAtPitchActual {type=number}
 LookAtYawAdjusted {type=number}
 MovementDirection {type=number}
 OldGoalRotation {type=number}
 Stop_StartPosition {type=number}
 TailWagAmount {type=number}
 UpdatedGoalRotation {type=number}
 WorriedAddPoseAlpha {type=number}
]
