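# Animation Blueprint for the WideSnore quadruped cosmetic companion mesh, following its attached parent's pose and setting variant-dependent morphs.
# The AnimGraph only copies pose and curves from the attached parent mesh.
# When character customization loads, the EventGraph checks for Stage1 or Stage4 variant tags, zeroing the body and hoodie LOD0 morphs on a match and setting both to full otherwise.
# The Belly boolean is declared but unused.
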AnimBlueprint /CosmeticCompanions/Assets/Quadruped/WideSnore/Meshes/WideSnore_Quadruped_AnimBP
target_skeleton=&/WideSnore_Quadruped_Skeleton
graphs [
 AnimGraph {nodes [Output Pose(Result=Copy Pose From Mesh() {bCopyCurves=true, bUseAttachedParent=true})]}
 EventGraph {nodes [
   Event OnCharacterCustomizationLoaded()
   > n1 = Branch(Condition=CharacterItemDefinitionHasAnyMetaTag(Tags=Make Array!("[0]"=(TagName="Cosmetics.Variant.Property.Stage1"), "[1]":Gameplay Tag Structure=(TagName="Cosmetics.Variant.Property.Stage4")) -> (Array)))
   > SetMorphTarget!(execute, self:Anim Instance Object Reference, MorphTargetName:Name=BodyMorph_LOD0, Value:float=0) -> (then)
   > SetMorphTarget(MorphTargetName=HoodieMorph_LOD0, Value=0)
   SetMorphTarget(execute=@n1.else, MorphTargetName=BodyMorph_LOD0, Value=1)
   > SetMorphTarget(MorphTargetName=HoodieMorph_LOD0, Value=1)
  ]}
]
variables [Belly {type=boolean}]
